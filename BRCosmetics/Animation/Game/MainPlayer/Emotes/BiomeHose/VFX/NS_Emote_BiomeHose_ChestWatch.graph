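# Niagara effect for the BiomeHose emote's chest watch, a brief, mostly green burst of meshes, flashes and glows.
# One emitter fires six instant burst groups, one slightly delayed, and switches color, lifetime, size, offset and curves on the spawn group index.
# That index also sets the visibility tag for the two mesh and five sprite renderers.
# User.BurstIndex selects between alternate local offsets, so one system covers several burst placements.
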NiagaraSystem /BRCosmetics/Animation/Game/MainPlayer/Emotes/BiomeHose/VFX/NS_Emote_BiomeHose_ChestWatch
dependencies {
 usedDataInterfaces [&NiagaraDataInterfaceVectorCurve, &NiagaraDataInterfaceCurve, &NiagaraDataInterfaceColorCurve]
 usedDynamicInputs []
 usedModules [&SystemState]
}
emitters {ChestVFX {
  emitterUpdateScript [
   &EmitterState()
   &SpawnBurst_Instantaneous(Use Spawn Probability=false, UseLoopCountLimit=false)
   &SpawnBurst_Instantaneous(Spawn Group=1, Use Spawn Probability=false, UseLoopCountLimit=false) {module=SpawnBurst_Instantaneous001}
   &SpawnBurst_Instantaneous(Spawn Group=2, Use Spawn Probability=false, UseLoopCountLimit=false) {module=SpawnBurst_Instantaneous002}
   &SpawnBurst_Instantaneous(Spawn Group=3, Use Spawn Probability=false, UseLoopCountLimit=false) {module=SpawnBurst_Instantaneous003}
   &SpawnBurst_Instantaneous(Spawn Group=4, Use Spawn Probability=false, UseLoopCountLimit=false) {module=SpawnBurst_Instantaneous004}
   &SpawnBurst_Instantaneous(Spawn Group=5, Spawn Time=0.05, Use Spawn Probability=false, UseLoopCountLimit=false) {module=SpawnBurst_Instantaneous005}
  ]
  particleSpawnScript [
   /SetVariables(Particles.VisibilityTag=&SpawnGroup()) {isSetParameters=true}
   &InitializeParticle(
    Color=&MergeHelper_ColorSwitch(
     0=<rgba 0.5139 1.5396 0 0.1861>
     1=<rgba 0.0833 0.9843 0 0.6661>
     2=<rgba 0.6992 1.25 0 1>
     3=<rgba 3.0297 5 0 1>
     4=<rgba 1 1 1 1>
     5=<rgba 0.4681 1 0.2797 1>
     Index=&SpawnGroup()
     Num=::NewEnumerator4
    )
    Lifetime=&MergeHelper_FloatSwitch(0=0.9, 1=0.9, 2=0.1, 3=0.5, 4=0.2, 5=0.1, Index=&SpawnGroup(), Num=::NewEnumerator4)
    Mesh Scale=<0.7 0.4 0.7>
    Mesh Scale Mode=::NewEnumerator4
    Position Offset=&MergeHelper_VectorSwitch(
     0=&MakeCustomVectorFromBool(
      Bool=&SetBoolByFloatComparison(A=$User.BurstIndex, B=1, Comparison Type=::NewEnumerator4)
      True Vector=<10 0 -7>
      False Vector=<10 0 -4>
     )
     1=&MakeCustomVectorFromBool(
      Bool=&SetBoolByFloatComparison(A=$User.BurstIndex, B=1, Comparison Type=::NewEnumerator4)
      True Vector=<10 0 -7>
      False Vector=<10 0 -4>
     )
     2=&MakeCustomVectorFromBool(
      Bool=&SetBoolByFloatComparison(A=$User.BurstIndex, B=1.9, Comparison Type=::NewEnumerator6)
      True Vector=<10 0 -3>
      False Vector=<10 5 -3>
     )
     3=/Niagara/DynamicInputs/TypeConversions/MakeCustomPositionFromBool(
      Bool:NiagaraBool=&SetBoolByFloatComparison(A=$User.BurstIndex, B=1, Comparison Type=::NewEnumerator4)
      True Position:NiagaraPosition=<13 0.5 -5>
      False Position:NiagaraPosition=<14 0.5 -3.5>
     )
     4=&MakeCustomVectorFromBool(
      Bool=&SetBoolByFloatComparison(A=$User.BurstIndex, B=1.9, Comparison Type=::NewEnumerator6)
      True Vector=<10 0 -3>
      False Vector=<10 5 -3>
     )
     5=&MakeCustomVectorFromBool(
      Bool=&SetBoolByFloatComparison(A=$User.BurstIndex, B=1.9, Comparison Type=::NewEnumerator6)
      True Vector=<10 0 -3>
      False Vector=<10 5 -3>
     )
     Index=&SpawnGroup()
     Num=::NewEnumerator4
    )
    Sprite Size Mode=::NewEnumerator3
    Uniform Sprite Size=&MergeHelper_FloatSwitch(0=0, 1=0, 2=280, 3=20, 4=45, 5=300, Index=&SpawnGroup(), Num=::NewEnumerator4)
    UsePositionOffset=true
    Write Material Random=true
   )
   &InitialMeshOrientation(Mesh Axis To Orient=<0 -1 0>, Mesh Orientation Mode=::NewEnumerator2, Orientation Coordinate Space=::Local, Use Rotation Vector=false)
   &CameraOffset(Camera Offset Amount=&MergeHelper_FloatSwitch(0=3, 1=3, 2=10, 3=6, 4=10, 5=10, Index=&SpawnGroup(), Num=::NewEnumerator4))
  ]
  particleUpdateScript [
   &ParticleState()
   &ScaleMeshSize(Scale Factor=&VectorFromCurve(
     VectorCurve=DataInterface() {propertyValues {
       XCurve {keys [<at 0 1>]}
       YCurve {keys [
         {arriveTangent=0.3778, interpMode=RCIM_Cubic, leaveTangent=0.3778, tangentMode=RCTM_User, time=0, value=0.5}
         {arriveTangent=-5.8915, interpMode=RCIM_Cubic, leaveTangent=-5.8915, tangentMode=RCTM_User, time=0.0245, value=1.169}
         {
          arriveTangent=-30.1978
          arriveTangentWeight=0.0172
          interpMode=RCIM_Cubic
          leaveTangent=-30.1978
          leaveTangentWeight=0.3242
          tangentMode=RCTM_User
          time=0.0282
          value=1.0884
         }
         {interpMode=RCIM_Cubic, leaveTangent=-4, time=0.0405, value=1}
        ]}
       ZCurve {keys [<at 0 1>]}
      }}
     Scale Curve=&MakeVector(X=1, Y=1.2, Z=1)
    ))
   &ScaleSpriteSize(
    Scale Sprite Size Mode=::NewEnumerator0
    Uniform Scale Factor=&MergeHelper_FloatSwitch(
     0=0
     1=0
     2=&FloatFromCurve(FloatCurve=DataInterface() {propertyValues {Curve {
         defaultValue=1
         keys [
          {arriveTangent=4.1785, interpMode=RCIM_Cubic, leaveTangent=4.1785, tangentMode=RCTM_User, time=0, value=0.25}
          {arriveTangent=1.1179, interpMode=RCIM_Cubic, leaveTangent=1.1179, tangentMode=RCTM_User, time=0.2042, value=0.75}
          {arriveTangent=0.0224, interpMode=RCIM_Cubic, leaveTangent=0.0224, tangentMode=RCTM_User, time=1, value=1}
         ]
        }}})
     3=1
     4=&FloatFromCurve(FloatCurve=DataInterface() {propertyValues {Curve {keys [
          {arriveTangent=7.2928, interpMode=RCIM_Cubic, leaveTangent=7.2928, tangentMode=RCTM_User, time=0, value=0.5}
          {arriveTangent=4.5, interpMode=RCIM_Cubic, leaveTangent=4.5, time=0.5096, value=3.5141}
          <at 1 5 RCIM_Cubic>
         ]}}})
     5=&FloatFromCurve(FloatCurve=DataInterface() {propertyValues {Curve {
         defaultValue=0
         keys [
          {interpMode=RCIM_Cubic, tangentMode=RCTM_User, time=0, value=1}
          {arriveTangent=-0.7492, interpMode=RCIM_Cubic, leaveTangent=-0.7492, time=0.4743, value=0.5772}
          {arriveTangent=0.1433, interpMode=RCIM_Cubic, leaveTangent=0.1433, tangentMode=RCTM_User, time=0.7756, value=0.4189}
          {arriveTangent=2.8825, interpMode=RCIM_Cubic, leaveTangent=56.4021, tangentMode=RCTM_Break, time=0.889, value=0.5723}
          {interpMode=RCIM_Cubic, tangentMode=RCTM_User, time=1, value=0}
         ]
        }}})
     Index=&SpawnGroup()
     Num=::NewEnumerator4
    )
   )
   &ScaleColor(
    Scale Alpha=&MergeHelper_FloatSwitch(
     0=&FloatFromCurve(
      FloatCurve=DataInterface() {propertyValues {Curve {
         defaultValue=1
         keys [
          {interpMode=RCIM_Cubic, tangentMode=RCTM_User, time=0, value=0}
          {arriveTangentWeight=0.0226, interpMode=RCIM_Cubic, leaveTangentWeight=0.2337, tangentMode=RCTM_User, time=0.0679, value=1}
          {arriveTangentWeight=0.3427, interpMode=RCIM_Cubic, leaveTangentWeight=0.1129, tangentMode=RCTM_User, time=0.769, value=1}
          {tangentMode=RCTM_User, time=1, value=0}
         ]
        }}}
      Scale Curve=&FloatFromCurve(FloatCurve=DataInterface() {propertyValues {Curve {keys [<at 0 1>,<at 0.6959 1.0064>,<at 1 0>]}}}, CurveIndex=$Emitter.Age)
     )
     1=&FloatFromCurve(
      FloatCurve=DataInterface() {propertyValues {Curve {
         defaultValue=1
         keys [
          {interpMode=RCIM_Cubic, tangentMode=RCTM_User, time=0, value=0}
          {arriveTangentWeight=0.0226, interpMode=RCIM_Cubic, leaveTangentWeight=0.2337, tangentMode=RCTM_User, time=0.0679, value=1}
          {arriveTangentWeight=0.3427, interpMode=RCIM_Cubic, leaveTangentWeight=0.1129, tangentMode=RCTM_User, time=0.769, value=1}
          {tangentMode=RCTM_User, time=1, value=0}
         ]
        }}}
      Scale Curve=&FloatFromCurve(FloatCurve=DataInterface() {propertyValues {Curve {keys [<at 0 1>,<at 0.6959 1.0064>,<at 1 0>]}}}, CurveIndex=$Emitter.Age)
     )
     2=&FloatFromCurve(FloatCurve=DataInterface() {propertyValues {Curve {keys [<at -0.0046 1.0113 RCIM_Cubic>,<at 1.0011 0.002 RCIM_Cubic>]}}}, Scale Curve=5)
     3=&FloatFromCurve(FloatCurve=DataInterface() {propertyValues {Curve {
         defaultValue=1
         keys [
          <at -0.0008 1 RCIM_Cubic>
          {arriveTangent=-1.4708, interpMode=RCIM_Cubic, leaveTangent=-1.4708, time=0.1107, value=0.4537}
          {interpMode=RCIM_Cubic, tangentMode=RCTM_User, time=0.498, value=0.2664}
          {arriveTangent=-0.0656, interpMode=RCIM_Cubic, leaveTangent=-0.0656, tangentMode=RCTM_User, time=1, value=0}
         ]
        }}})
     4=&FloatFromCurve(
      FloatCurve=DataInterface() {propertyValues {Curve {keys [
          <at -0.0046 1.0113 RCIM_Cubic>
          {arriveTangent=-1.0349, interpMode=RCIM_Cubic, leaveTangent=-1.0349, time=0.435, value=0.834}
          {arriveTangent=-1.2597, interpMode=RCIM_Cubic, leaveTangent=-1.2597, tangentMode=RCTM_User, time=0.6814, value=0.3013}
          <at 1.0011 0.002 RCIM_Cubic>
         ]}}}
      Scale Curve=4
     )
     5=&FloatFromCurve(FloatCurve=DataInterface() {propertyValues {Curve {keys [<at -0.0046 1.0113 RCIM_Cubic>,<at 1.0011 0.002 RCIM_Cubic>]}}})
     Index=&SpawnGroup()
     Num=::NewEnumerator4
    )
    Scale RGB=&MergeHelper_VectorSwitch(
     0=<1 1 1>
     1=<1 1 1>
     2=&MakeVectorFromLinearColorRGB(Color=&ColorFromCurve(ColorCurve=DataInterface() {propertyValues {
         AlphaCurve {keys []}
         BlueCurve {keys [<at 0.0949 5 RCIM_Cubic>,<at 0.2972 1 RCIM_Cubic>]}
         GreenCurve {keys [<at 0.0949 5 RCIM_Cubic>,<at 0.2972 1 RCIM_Cubic>]}
         RedCurve {keys [<at 0.0949 5 RCIM_Cubic>,<at 0.2972 1 RCIM_Cubic>]}
        }}))
     3=<1 1 1>
     4=&MakeVectorFromLinearColorRGB(Color=&ColorFromCurve(ColorCurve=DataInterface() {propertyValues {
         AlphaCurve {keys []}
         BlueCurve {keys [<at 0.178 0>,<at 0.3081 0>,<at 0.6065 0 RCIM_Cubic>,<at 0.7747 0 RCIM_Cubic>,<at 1 0>]}
         GreenCurve {keys [
           <at 0.178 10>
           <at 0.3081 1>
           {arriveTangent=-1.4398, interpMode=RCIM_Cubic, leaveTangent=-1.4398, time=0.6065, value=0.8906}
           {arriveTangent=-1.2481, interpMode=RCIM_Cubic, leaveTangent=-1.2481, time=0.7747, value=0.3281}
           <at 1 0.1365>
          ]}
         RedCurve {keys [
           <at 0.178 0.3544>
           <at 0.3081 0.2038>
           {arriveTangent=-0.4182, interpMode=RCIM_Cubic, leaveTangent=-0.4182, time=0.6065, value=0.1405}
           {arriveTangent=-0.2893, interpMode=RCIM_Cubic, leaveTangent=-0.2893, time=0.7747, value=0.0086}
           <at 1 0.0036>
          ]}
        }}))
     5=<1 1 1>
     Index=&SpawnGroup()
     Num=::NewEnumerator4
    )
    ScaleA=true
    ScaleRGB=true
   )
   /SetVariables(Particles.Star Size:Vector2f=&Multiply_Vector2D_ByFloat(Float=0.4, Vector2D=$Particles.SpriteSize)) {isSetParameters=true}
   /SetVariables(Particles.ButtonAlignment:Vector3f=<0 1 0>) {isSetParameters=true}
  ]
  rendererClasses [&NiagaraMeshRendererProperties, &NiagaraSpriteRendererProperties]
  renderers [
   {rendererClass=&NiagaraMeshRendererProperties, rendererIndex=0}
   {rendererClass=&NiagaraMeshRendererProperties, rendererIndex=1}
   {rendererClass=&NiagaraSpriteRendererProperties, rendererIndex=2}
   {rendererClass=&NiagaraSpriteRendererProperties, rendererIndex=3}
   {rendererClass=&NiagaraSpriteRendererProperties, rendererIndex=4}
   {rendererClass=&NiagaraSpriteRendererProperties, rendererIndex=5}
   {rendererClass=&NiagaraSpriteRendererProperties, rendererIndex=6}
  ]
 }}
summary {
 systemName=NS_Emote_BiomeHose_ChestWatch
 emitters [{emitterName=ChestVFX, rendererClasses [&NiagaraMeshRendererProperties, &NiagaraSpriteRendererProperties]}]
 userVariables [User.BurstIndex {type=NiagaraFloat, defaultValue=0, description=""}]
}
