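# Niagara system for the BiomeHose emote's transform effect, building up condensed energy around the player.
# System first, second and third transition flags select spawn counts, timings and curves for different buildup lengths.
# The BuildUp_Condensed emitter layers spawn groups of sprites and meshes, using visibility tags and per-group switches.
# Masked sprites pulse, stretch with speed and get pulled toward a raised attractor, while particle variables store animated lightning and dot parameters.
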NiagaraSystem /BRCosmetics/Animation/Game/MainPlayer/Emotes/BiomeHose/VFX/NS_Emote_BiomeHose_Transform
dependencies {
 usedDataInterfaces [
  &NiagaraDataInterfaceCurve
  &NiagaraDataInterfaceColorCurve
  &NiagaraDataInterfaceSpriteRendererInfo
  &NiagaraDataInterfaceSkeletalMesh
 ]
 usedDynamicInputs [
  &SetBoolByFloatComparison
  &Boolean_AndOperation
  &Multiply_VectorByFloat
 ]
 usedModules [
  &self:SystemSpawnScript.SystemScriptSource.SystemScriptGraph.NiagaraNodeAssignment_1.SetVariables_#1
  &SystemState
  &self:SystemSpawnScript.SystemScriptSource.SystemScriptGraph.NiagaraNodeAssignment_0.SetVariables_#2
  &self:SystemSpawnScript.SystemScriptSource.SystemScriptGraph.NiagaraNodeAssignment_2.SetVariables_#3
  &self:SystemSpawnScript.SystemScriptSource.SystemScriptGraph.NiagaraNodeAssignment_3.SetVariables_#4
  &self:SystemSpawnScript.SystemScriptSource.SystemScriptGraph.NiagaraNodeAssignment_4.SetVariables_#5
 ]
}
emitters {
 BuildUp_Condensed {
  emitterUpdateScript [
   &EmitterState()
   &SpawnBurst_Instantaneous(
    Spawn Count=&MakeCustomIntFromBool(Bool=$System.IsFirstTransition, True INT=10, False INT=5)
    Spawn Time=0.01
    Use Spawn Probability=false
    UseLoopCountLimit=false
   )
   &SpawnRate(
    SpawnRate=&MakeCustomFloatFromBool(
     Bool=$System.isFirstTransition
     True Float=&Multiply_Float(
      A=60
      B=&FloatFromCurve(
       FloatCurve=DataInterface() {propertyValues {Curve {
          defaultValue=1
          keys [{tangentMode=RCTM_User, time=0, value=0}, {tangentMode=RCTM_User, time=0.15, value=1}, <at 0.6365 1>, <at 1.1433 1.3953>, <at 1.37 0>]
         }}}
       CurveIndex=$System.Age
      )
     )
     False Float=&MakeCustomFloatFromBool(
      Bool=$System.isSecondTransition
      True Float=&Multiply_Float(
       A=37
       B=&FloatFromCurve(
        FloatCurve=DataInterface() {propertyValues {Curve {defaultValue=1, keys [{tangentMode=RCTM_User, time=0, value=0}, {tangentMode=RCTM_User, time=0.07, value=1}, <at 0.9 1>, <at 0.97 0>]}}}
        CurveIndex=$System.Age
       )
      )
      False Float=&Multiply_Float(
       A=40
       B=&FloatFromCurve(
        FloatCurve=DataInterface() {propertyValues {Curve {defaultValue=1, keys [{tangentMode=RCTM_User, time=0, value=1}, <at 0.35 1>, <at 0.37 0>]}}}
        CurveIndex=$System.Age
       )
      )
     )
    )
    Use Spawn Probability=false
   )
   &SpawnBurst_Instantaneous(
    Spawn Count=&MakeCustomIntFromBool(Bool=$System.isFirstTransition, True INT=30, False INT=40)
    Spawn Group=1
    Spawn Time=&MakeCustomFloatFromBool(
     Bool=$System.isFirstTransition
     True Float=1.45
     False Float=&MakeCustomFloatFromBool(Bool=$System.isSecondTransition, True Float=1.1, False Float=0.4)
    )
    Use Spawn Probability=false
    UseLoopCountLimit=false
   ) {module=SpawnBurst_Instantaneous001}
   &SpawnRate(
    SpawnGroup=3
    SpawnRate=&Multiply_Float(
     A=&MakeCustomFloatFromBool(
      Bool=$System.isFirstTransition
      True Float=&Multiply_Float(
       A=65
       B=&FloatFromCurve(
        FloatCurve=DataInterface() {propertyValues {Curve {
           defaultValue=1
           keys [
            <at 0.05 0 RCIM_Cubic>
            {
             arriveTangent=3.9415
             arriveTangentWeight=0.222
             interpMode=RCIM_Cubic
             leaveTangentWeight=0.2858
             tangentMode=RCTM_Break
             time=0.56
             value=1
            }
            <at 1.31 0.9906>
            <at 1.425 0>
           ]
          }}}
        CurveIndex=$System.Age
       )
      )
      False Float=&MakeCustomFloatFromBool(
       Bool=$System.isSecondTransition
       True Float=&Multiply_Float(
        A=65
        B=&FloatFromCurve(
         FloatCurve=DataInterface() {propertyValues {Curve {
            defaultValue=1
            keys [
             <at 0 0 RCIM_Cubic>
             {arriveTangent=3.9415, arriveTangentWeight=0.222, interpMode=RCIM_Cubic, leaveTangentWeight=0.2858, tangentMode=RCTM_Break, time=0.1, value=1}
             <at 0.9 0.9906>
             <at 0.97 0>
            ]
           }}}
         CurveIndex=$System.Age
        )
       )
       False Float=&Multiply_Float(
        A=120
        B=&FloatFromCurve(FloatCurve=DataInterface() {propertyValues {Curve {defaultValue=1, keys [<at 0.37 0.9906>,<at 0.4 0>]}}}, CurveIndex=$System.Age)
       )
      )
     )
     B=0.33
    )
    Use Spawn Probability=false
   ) {module=SpawnRate001}
   &SpawnRate(
    SpawnGroup=4
    SpawnRate=&MakeCustomFloatFromBool(
     Bool=$System.isFirstTransition
     True Float=&Multiply_Float(
      A=80
      B=&FloatFromCurve(
       FloatCurve=DataInterface() {propertyValues {Curve {
          defaultValue=1
          keys [
           {arriveTangent=3.0718, interpMode=RCIM_Cubic, leaveTangent=3.0718, tangentMode=RCTM_User, time=0.05, value=0}
           {
            arriveTangent=1.0221
            arriveTangentWeight=0.222
            interpMode=RCIM_Cubic
            leaveTangent=1.0221
            leaveTangentWeight=0.2858
            time=0.47
            value=1.055
           }
           {arriveTangent=-1.6104, tangentMode=RCTM_User, time=1.08, value=1.0528}
           <at 1.0806 0>
          ]
         }}}
       CurveIndex=$System.Age
      )
     )
     False Float=&MakeCustomFloatFromBool(
      Bool=$System.isSecondTransition
      True Float=&Multiply_Float(
       A=90
       B=&FloatFromCurve(
        FloatCurve=DataInterface() {propertyValues {Curve {
           defaultValue=1
           keys [
            {arriveTangent=3.0718, interpMode=RCIM_Cubic, leaveTangent=3.0718, tangentMode=RCTM_User, time=0.0062, value=0.4141}
            {arriveTangent=0.601, arriveTangentWeight=0.222, interpMode=RCIM_Cubic, leaveTangent=0.601, leaveTangentWeight=0.2858, time=0.2, value=1.055}
            {arriveTangent=-1.6104, tangentMode=RCTM_User, time=1.069, value=1.0528}
            <at 1.07 0>
           ]
          }}}
        CurveIndex=$System.Age
       )
      )
      False Float=&Multiply_Float(
       A=150
       B=&FloatFromCurve(
        FloatCurve=DataInterface() {propertyValues {Curve {defaultValue=1, keys [{arriveTangent=-1.6104, tangentMode=RCTM_User, time=0.38, value=1.0528}, <at 0.3801 0>]}}}
        CurveIndex=$System.Age
       )
      )
     )
    )
    Use Spawn Probability=false
   ) {module=SpawnRate002}
   &SpawnBurst_Instantaneous(Spawn Group=5, Use Spawn Probability=false, UseLoopCountLimit=false) {module=SpawnBurst_Instantaneous002}
   &SpawnBurst_Instantaneous(Spawn Group=6, Use Spawn Probability=false, UseLoopCountLimit=false) {module=SpawnBurst_Instantaneous003}
  ]
  particleSpawnScript [
   /SetVariables(Particles.VisibilityTag=&MakeCustomIntFromBool(
     Bool=&SetBoolByIntComparison(A=&SpawnGroup(), B=2, Comparison Type=::NewEnumerator7)
     True INT=&RandomRangeInt(Minimum=0, Maximum=2)
     False INT=&SpawnGroup()
    )) {isSetParameters=true}
   /SetVariables(Particles.Initial Lightning Color:LinearColor=&MultiplyLinearColor(
     Linear Color=&RandomRangeLinearColor(Minimum=<rgba 1 1 1 1>, Maximum=<rgba 1.2 1.2 1.2 2>)
     Scale Factor=&MakeVector4_FromFloat(Float=0.8)
    )) {isSetParameters=true}
   &InitializeParticle(
    Color=&MergeHelper_ColorSwitch(
     0=&MultiplyLinearColor(
      Linear Color=&RandomRangeLinearColor(Minimum=<rgba 1 1 1 1>, Maximum=<rgba 1.2 1.2 1.2 2>)
      Scale Factor=&MakeVector4_FromFloat(Float=0.8)
     )
     1=&MultiplyLinearColor(
      Linear Color=&RandomRangeLinearColor(Minimum=<rgba 1 1 1 1>, Maximum=<rgba 1.2 1.2 1.2 2>)
      Scale Factor=&MakeVector4_FromFloat(Float=3)
     )
     2=&MultiplyLinearColor(
      Linear Color=&RandomRangeLinearColor(Minimum=<rgba 1 1 1 1>, Maximum=<rgba 1.2 1.2 1.2 2>)
      Scale Factor=&MakeVector4_FromFloat(Float=3)
     )
     3=&MultiplyLinearColor(
      Linear Color=&RandomRangeLinearColor(Minimum=<rgba 1 1 1 1>, Maximum=<rgba 1.2 1.2 1.2 2>)
      Scale Factor=&MakeVector4_FromFloat(Float=0.8)
     )
     4=<rgba 1 1 1 1>
     5=<rgba 1 1 1 1>
     6=<rgba 1 1 1 1>
     Index=&SpawnGroup()
     Num=::NewEnumerator5
    )
    Lifetime=&MergeHelper_FloatSwitch(
     0=&RandomRangeFloat(Minimum=0.27, Maximum=0.5)
     1=&RandomRangeFloat(Minimum=0.21, Maximum=0.35)
     2=&RandomRangeFloat(Minimum=0.27, Maximum=0.5)
     3=&RandomRangeFloat(Minimum=0.27, Maximum=0.5)
     4=&RandomRangeFloat(Minimum=0.5, Maximum=1)
     5=&MakeCustomFloatFromBool(
      Bool=$System.isFirstTransition
      True Float=1.45
      False Float=&MakeCustomFloatFromBool(Bool=$System.isSecondTransition, True Float=1.1, False Float=0.4)
     )
     6=&MakeCustomFloatFromBool(
      Bool=$System.isFirstTransition
      True Float=1.4
      False Float=&MakeCustomFloatFromBool(Bool=$System.isSecondTransition, True Float=1.1, False Float=0.4)
     )
     Index=&SpawnGroup()
     Num=::NewEnumerator5
    )
    Mesh Scale Mode=::NewEnumerator3
    Mesh Uniform Scale=&Multiply_Float(A=&RandomRangeFloat(Minimum=0.5, Maximum=0.55), B=0.3)
    Position Offset=&MaskVector(Mask Source=0, Mask Parameter=?, Mask=5, Masked Vector=<0 0 125>, Passthrough Vector=<0 0 0>)
    Sprite Size Mode=::NewEnumerator3
    Uniform Sprite Size=&MaskFloat(
     Mask Source=0
     Mask Parameter=?
     Mask=4
     Masked Float=&RandomRangeFloat(Minimum=15, Maximum=100)
     Passthrough Float=370
    )
    UsePositionOffset=true
    Write Material Random=true
   )
   &InitialMeshOrientation(Use Rotation Vector=false) {module=InitialMeshOrientation002}
   &ShapeLocation(
    MaskSpawnGroup=false
    Offset=&MergeHelper_VectorSwitch(
     0=&RandomRangeVector(Minimum=<0 0 150>, Maximum=<0 0 50>)
     1=&RandomRangeVector(Minimum=<0 0 175>, Maximum=<0 0 75>)
     2=&RandomRangeVector(Minimum=<0 0 175>, Maximum=<0 0 75>)
     3=<0 0 100>
     4=<0 0 125>
     5=<0 0 0>
     6=<0 0 0>
     Index=&SpawnGroup()
     Num=::NewEnumerator5
    )
    Offset Mode=::NewEnumerator0
    Override Seed=false
    Sphere Radius=&MergeHelper_FloatSwitch(0=25, 1=35, 2=35, 3=17, 4=85, 5=0, 6=0, Index=&SpawnGroup(), Num=::NewEnumerator5)
    UseOwnerScale=true
   )
   &CameraOffset(Camera Offset Amount=&MergeHelper_FloatSwitch(0=15, 1=15, 2=15, 3=-50, 4=-50, 5=-50, 6=0, Index=&SpawnGroup(), Num=::NewEnumerator5))
   &SubUVAnimation(
    Output Alternate Renderer Binding=true
    "Sprite Renderer "=DataInterface() {propertyValues {SpriteRenderer=&self:BuildUp_Condensed_0.NiagaraSpriteRendererProperties_0}}
    SubUV Animation Mode:ENiagara_SubUVLookupModeV2=::NewEnumerator2
    UseEndFrame=false
    UseStartFrame=false
   )
  ]
  particleUpdateScript [
   &ParticleState()
   /SetVariables(Particles.undulate:NiagaraFloat=&Waveform(
     "Waveform [1]"=::NewEnumerator0
     "[1] Frequency"=&MakeCustomFloatFromBool(
      Bool=$System.isFirstTransition
      True Float=&FloatFromCurve(
       FloatCurve=DataInterface() {propertyValues {Curve {defaultValue=1, keys [{tangentMode=RCTM_User, time=0.05, value=0}, <at 0.904 5.1353>, {tangentMode=RCTM_User, time=0.9822, value=13.6191}]}}}
       CurveIndex=$Emitter.Age
      )
      False Float=&MakeCustomFloatFromBool(
       Bool=$System.isSecondTransition
       True Float=&FloatFromCurve(
        FloatCurve=DataInterface() {propertyValues {Curve {defaultValue=1, keys [{tangentMode=RCTM_User, time=0.05, value=0}, <at 0.73 5.1353>, {tangentMode=RCTM_User, time=0.98, value=13.6191}]}}}
        CurveIndex=$Emitter.Age
       )
       False Float=&FloatFromCurve(
        FloatCurve=DataInterface() {propertyValues {Curve {defaultValue=1, keys [{tangentMode=RCTM_User, time=0, value=0}, <at 0.26 5.1353>, {tangentMode=RCTM_User, time=0.35, value=13.6191}]}}}
        CurveIndex=$Emitter.Age
       )
      )
     )
     "[1] Amplitude Min/Max"=<0.5 1>
    )) {isSetParameters=true}
   &ScaleSpriteSize(
    Scale Factor=/Niagara/DynamicInputs/Helpers/MaskVector2d(
     Mask Source:NiagaraInt32=0
     Mask Parameter:NiagaraInt32=?
     Mask:NiagaraInt32=4
     Masked Vector2D:Vector2f=&Multiply_Vector2D(
      A=&MakeVector2D(X=$Particles.undulate, Y=$Particles.undulate)
      B=&Vector2DFromFloat(Value=&MakeCustomFloatFromBool(
        Bool=$System.isFirstTransition
        True Float=&FloatFromCurve(FloatCurve=DataInterface() {propertyValues {Curve {
            defaultValue=1
            keys [
             {interpMode=RCIM_Cubic, tangentMode=RCTM_User, time=0, value=0}
             {interpMode=RCIM_Cubic, tangentMode=RCTM_User, time=0.5, value=1}
             {tangentMode=RCTM_User, time=1, value=0}
            ]
           }}})
        False Float=&MakeCustomFloatFromBool(
         Bool=$System.isSecondTransition
         True Float=&FloatFromCurve(FloatCurve=DataInterface() {propertyValues {Curve {
             defaultValue=1
             keys [
              {interpMode=RCIM_Cubic, tangentMode=RCTM_User, time=0, value=0}
              {interpMode=RCIM_Cubic, tangentMode=RCTM_User, time=0.3626, value=1}
              {tangentMode=RCTM_User, time=1, value=0}
             ]
            }}})
         False Float=&FloatFromCurve(FloatCurve=DataInterface() {propertyValues {Curve {
             defaultValue=1
             keys [
              {arriveTangent=23.3017, interpMode=RCIM_Cubic, leaveTangent=23.3017, tangentMode=RCTM_User, time=0.0016, value=0}
              {arriveTangent=-0.089, interpMode=RCIM_Cubic, leaveTangent=-0.089, tangentMode=RCTM_User, time=0.0919, value=1}
              {arriveTangent=-2.3568, interpMode=RCIM_Cubic, leaveTangent=-2.3568, tangentMode=RCTM_User, time=0.1841, value=0.6426}
              {arriveTangent=-0.7463, interpMode=RCIM_Cubic, leaveTangent=-0.7463, tangentMode=RCTM_User, time=0.3362, value=0.3694}
              {arriveTangent=-0.1702, tangentMode=RCTM_User, time=1, value=0}
             ]
            }}})
        )
       ))
     )
     Passthrough Vector2D:Vector2f=&Vector2DFromFloat(Value=&FloatFromCurve(FloatCurve=DataInterface() {propertyValues {Curve {
          defaultValue=1
          keys [
           <at 0 1 RCIM_Cubic>
           {arriveTangent=-0.8928, interpMode=RCIM_Cubic, leaveTangent=-0.8928, tangentMode=RCTM_User, time=0.549, value=0.7971}
           {arriveTangent=-0.8989, interpMode=RCIM_Cubic, leaveTangent=-0.8989, time=0.598, value=0.6785}
           {arriveTangent=-0.5718, interpMode=RCIM_Cubic, leaveTangent=-14.2537, tangentMode=RCTM_Break, time=0.8716, value=0.5071}
           {arriveTangent=12.3329, interpMode=RCIM_Cubic, leaveTangent=12.3329, tangentMode=RCTM_User, time=0.9011, value=0.4391}
           {arriveTangent=32.4571, interpMode=RCIM_Cubic, leaveTangent=32.4571, time=0.9016, value=0.8684}
           {arriveTangent=-20.0455, interpMode=RCIM_Cubic, leaveTangent=-20.0455, time=0.9151, value=0.8908}
           {arriveTangent=-9.6165, interpMode=RCIM_Cubic, leaveTangent=-9.6165, time=0.9199, value=0.5004}
           {arriveTangent=0.6083, interpMode=RCIM_Cubic, leaveTangent=0.6083, time=0.9361, value=0.6885}
           {arriveTangent=-11.0982, interpMode=RCIM_Cubic, leaveTangent=-11.0982, time=0.9697, value=0.5307}
           <at 0.9981 0 RCIM_Cubic>
          ]
         }}}))
    )
    Scale Sprite Size Mode=::NewEnumerator2
   )
   &PointAttractionForce(
    Attraction Radius=200
    Attraction Strength=&MaskFloat(
     Mask Source=0
     Mask Parameter=?
     Mask=4
     Masked Float=&MakeCustomFloatFromBool(
      Bool=$System.isFirstTransition
      True Float=&FloatFromCurve(
       FloatCurve=DataInterface() {propertyValues {Curve {
          defaultValue=1
          keys [
           <at 0.6065 0>
           {tangentMode=RCTM_User, time=0.7552, value=18.0418}
           <at 1.0157 48.9073>
           <at 1.1847 133.2891>
           {tangentMode=RCTM_User, time=1.2017, value=256.446}
           <at 1.2455 325.7762>
          ]
         }}}
       CurveIndex=$Emitter.Age
      )
      False Float=&MakeCustomFloatFromBool(
       Bool=$System.isSecondTransition
       True Float=&FloatFromCurve(
        FloatCurve=DataInterface() {propertyValues {Curve {
           defaultValue=1
           keys [
            <at 0.3365 0>
            <at 0.8717 56.2049>
            <at 0.9063 159.2881>
            <at 0.9522 201.7817>
            {tangentMode=RCTM_User, time=0.9595, value=311.5236}
            <at 0.994 374.9999>
           ]
          }}}
        CurveIndex=$Emitter.Age
       )
       False Float=&FloatFromCurve(
        FloatCurve=DataInterface() {propertyValues {Curve {defaultValue=1, keys [<at 0.224 0>,<at 0.3021 77.4608>,<at 0.3023 235.5688>,<at 0.3302 330.6783>]}}}
        CurveIndex=$Emitter.Age
       )
      )
     )
     Passthrough Float=0
    )
    Attractor Position Offset:Vector3f=<0 0 135>
    Kill Within Radius=false
    Module.Use Falloff=true
    UseAttractorOffset=true
   )
   &ScaleMeshSize(Scale Factor=&VectorFromFloat(Value=&FloatFromCurve(FloatCurve=DataInterface() {propertyValues {Curve {keys [
          {arriveTangent=35.758, interpMode=RCIM_Cubic, leaveTangent=35.758, tangentMode=RCTM_User, time=0, value=0}
          {arriveTangent=-6.3979, interpMode=RCIM_Cubic, leaveTangent=4.4257, tangentMode=RCTM_Break, time=0.1425, value=0.8597}
          {arriveTangent=0.0224, interpMode=RCIM_Cubic, leaveTangent=0.0224, tangentMode=RCTM_User, time=0.2523, value=1.1167}
          <at 1.0411 0.2664 RCIM_Cubic>
         ]}}})))
   &ScaleMeshSize(Scale Factor=&RandomRangeVector(Minimum=<1 1 1.5>, Maximum=<1 1 2>)) {module=ScaleMeshSize001}
   &SolveForcesAndVelocity(Clamp Velocity=false, Limit Acceleration=false)
   &ScaleSpriteSizeBySpeed(
    Max Scale Factor=&MakeVector2D(
     X=1
     Y=&MaskFloat(
      Mask Source=0
      Mask Parameter=?
      Mask=4
      Masked Float=&MakeCustomFloatFromBool(
       Bool=$System.isFirstTransition
       True Float=&FloatFromCurve(
        FloatCurve=DataInterface() {propertyValues {Curve {
           defaultValue=1
           keys [
            {tangentMode=RCTM_User, time=0.75, value=1}
            <at 0.9167 1.9003>
            {tangentMode=RCTM_User, time=0.98, value=3.3482}
            <at 1.1005 1.5783>
            <at 1.2608 1.4905>
            <at 1.2988 3.1836>
            <at 1.4293 0.24>
            <at 1.4716 0.24>
           ]
          }}}
        CurveIndex=$Emitter.Age
       )
       False Float=&MakeCustomFloatFromBool(
        Bool=$System.isSecondTransition
        True Float=&FloatFromCurve(
         FloatCurve=DataInterface() {propertyValues {Curve {
            defaultValue=1
            keys [
             {tangentMode=RCTM_User, time=0.0583, value=1}
             <at 0.2769 1.6258>
             {tangentMode=RCTM_User, time=0.4158, value=3.2217}
             <at 0.5 1.4025>
             <at 1.065 0.8331>
             <at 1.0987 3.5495>
            ]
           }}}
         CurveIndex=$Emitter.Age
        )
        False Float=&FloatFromCurve(
         FloatCurve=DataInterface() {propertyValues {Curve {
            defaultValue=1
            keys [
             {tangentMode=RCTM_User, time=0.0433, value=0.9752}
             <at 0.135 1.819>
             {tangentMode=RCTM_User, time=0.2103, value=3.2669}
             <at 0.267 1.497>
             <at 0.378 1.389>
             <at 0.4207 3.2373>
            ]
           }}}
         CurveIndex=$Emitter.Age
        )
       )
      )
      Passthrough Float=1
     )
    )
    Min Scale Factor=&MakeVector2D(
     X=&MaskFloat(
      Mask Source=0
      Mask Parameter=?
      Mask=4
      Masked Float=&MakeCustomFloatFromBool(
       Bool=$System.isFirstTransition
       True Float=&FloatFromCurve(
        FloatCurve=DataInterface() {propertyValues {Curve {
           defaultValue=1
           keys [
            {tangentMode=RCTM_User, time=0.65, value=1}
            <at 0.85 0.7796>
            {tangentMode=RCTM_User, time=0.92, value=0.4295}
            <at 0.9555 0.2>
            <at 1.2963 0.2>
            <at 1.4162 0.1>
            <at 1.4736 0.1>
           ]
          }}}
        CurveIndex=$Emitter.Age
       )
       False Float=&MakeCustomFloatFromBool(
        Bool=$System.isSecondTransition
        True Float=&FloatFromCurve(
         FloatCurve=DataInterface() {propertyValues {Curve {
            defaultValue=1
            keys [
             {tangentMode=RCTM_User, time=0.3, value=1}
             <at 0.789 0.7299>
             {tangentMode=RCTM_User, time=0.8696, value=0.4898}
             <at 0.9268 0.1744>
             <at 0.9959 0.0461>
            ]
           }}}
         CurveIndex=$Emitter.Age
        )
        False Float=&FloatFromCurve(
         FloatCurve=DataInterface() {propertyValues {Curve {
            defaultValue=1
            keys [{tangentMode=RCTM_User, time=0.1557, value=1}, <at 0.2253 0.777>, {tangentMode=RCTM_User, time=0.3027, value=0.4496}, <at 0.3429 0.2>]
           }}}
         CurveIndex=$Emitter.Age
         Scale Curve=&FloatFromCurve(FloatCurve=DataInterface() {propertyValues {Curve {defaultValue=1, keys [{tangentMode=RCTM_User, time=0.0009, value=0}, {tangentMode=RCTM_User, time=0.0044, value=1}]}}})
        )
       )
      )
      Passthrough Float=1
     )
     Y=1
    )
   )
   &ScaleColor(
    Scale Alpha=&FloatFromCurve(FloatCurve=DataInterface() {propertyValues {Curve {keys [
         {tangentMode=RCTM_User, time=0, value=0.876}
         {arriveTangent=-4.3986, interpMode=RCIM_Cubic, leaveTangent=-4.3986, tangentMode=RCTM_User, time=0.1692, value=1}
         {arriveTangent=0.1205, interpMode=RCIM_Cubic, leaveTangent=0.1205, tangentMode=RCTM_User, time=1, value=0}
        ]}}})
    Scale RGB=&Lerp_Vector(
     A=&VectorFromFloat(Value=20)
     B=&VectorFromFloat(Value=&RandomRangeFloat(Minimum=0, Maximum=2, Evaluation Type=::NewEnumerator1))
     Alpha=&FloatFromCurve(FloatCurve=DataInterface() {propertyValues {Curve {keys [
          {interpMode=RCIM_Cubic, leaveTangent=2.5294, tangentMode=RCTM_Break, time=0, value=0}
          {arriveTangent=-0.2166, interpMode=RCIM_Cubic, tangentMode=RCTM_Break, time=0.4, value=1}
          <at 0.9989 0 RCIM_Cubic>
         ]}}})
    )
    ScaleA=true
    ScaleRGB=true
   )
   &DynamicMaterialParameters(
    Index 0 Param 1=10
    Index 0 Param 2=&RandomRangeFloat(Minimum=0, Maximum=0)
    Index 0 Param 4=13
    Index 1 Param 1=&FloatFromCurve(FloatCurve=DataInterface() {propertyValues {Curve {keys [<at 0 0.5>,<at 1 1>]}}})
    Index 1 Param 2=2
    Index 1 Param 3=&FloatFromCurve(FloatCurve=DataInterface() {propertyValues {Curve {keys [
         {interpMode=RCIM_Cubic, tangentMode=RCTM_Break, time=0.5, value=-0.5}
         {arriveTangent=1.2, interpMode=RCIM_Cubic, leaveTangent=1.2, time=0.7034, value=-0.0784}
         <at 1 0.1>
        ]}}})
    Index 2 Param 1=&Multiply_Float(
     A=&FloatFromCurve(
      FloatCurve=DataInterface() {propertyValues {Curve {keys [
          {interpMode=RCIM_Cubic, leaveTangent=8.1085, tangentMode=RCTM_Break, time=0.7695, value=1}
          {arriveTangent=-9.494, interpMode=RCIM_Cubic, leaveTangent=-9.494, tangentMode=RCTM_User, time=0.9048, value=0.8212}
          {tangentMode=RCTM_User, time=1, value=0}
         ]}}}
      Scale Curve=0.8
     )
     B=&Divide_Float(
      A=&MakeFloatFromVector(Vector=$Particles.Scale, Channel=::NewEnumerator0)
      B=&RandomRangeFloat(Minimum=10, Maximum=15)
     )
    )
    Index 2 Param 3=&RandomRangeFloat(Minimum=0, Maximum=5)
    Index 2 Param 4=25
    Index 3 Param 1=&FloatFromCurve(
     FloatCurve=DataInterface() {propertyValues {Curve {keys [
         {interpMode=RCIM_Cubic, leaveTangent=-0.0025, tangentMode=RCTM_Break, time=0, value=0}
         {arriveTangent=1.8528, interpMode=RCIM_Cubic, tangentMode=RCTM_Break, time=1, value=1}
        ]}}}
     Scale Curve=&RandomRangeFloat(Minimum=0, Maximum=0)
    )
    Index 3 Param 2=&FloatFromCurve(
     FloatCurve=DataInterface() {propertyValues {Curve {keys [
         {interpMode=RCIM_Cubic, leaveTangent=-0.0025, tangentMode=RCTM_Break, time=0, value=0.5}
         {arriveTangent=-0.0316, interpMode=RCIM_Cubic, tangentMode=RCTM_Break, time=1, value=-0.2}
        ]}}}
     Scale Curve=0
    )
    Index 3 Param 3=&FloatFromCurve(FloatCurve=DataInterface() {propertyValues {Curve {keys [
         {interpMode=RCIM_Cubic, leaveTangent=1.1034, tangentMode=RCTM_Break, time=0, value=-1}
         {arriveTangent=2.2399, tangentMode=RCTM_User, time=0.8976, value=1.2895}
         <at 0.9997 -1.8498>
        ]}}})
    Index 3 Param 4=&FloatFromCurve(FloatCurve=DataInterface() {propertyValues {Curve {keys [
         {interpMode=RCIM_Cubic, leaveTangent=1.8089, tangentMode=RCTM_Break, time=0.1837, value=-0.5}
         {arriveTangent=2.0941, interpMode=RCIM_Cubic, leaveTangent=2.0941, time=0.5787, value=0.4567}
         {arriveTangent=2.0757, interpMode=RCIM_Cubic, tangentMode=RCTM_Break, time=0.9, value=1}
        ]}}})
    Module.Index0Param0Write=true
    Module.Index0Param1Write=true
    Module.Index0Param2Write=true
    Module.Index0Param3Write=true
    Module.Index1Param0Write=true
    Module.Index1Param1Write=true
    Module.Index1Param2Write=true
    Module.Index1Param3Write=true
    Module.Index2Param0Write=true
    Module.Index2Param1Write=true
    Module.Index2Param2Write=true
    Module.Index2Param3Write=true
    Module.Index3Param0Write=true
    Module.Index3Param1Write=true
    Module.Index3Param2Write=true
    Module.Index3Param3Write=true
    Write Parameter Index 1=true
    Write Parameter Index 2=true
    Write Parameter Index 3=true
   ) {enabled=false}
   /SetVariables(
    Particles.Lightning Param 0:Vector4f=&MakeVector4(X=10, Y=0, Z=0, W=13)
    Particles.Lightning Param 1:Vector4f=&MakeVector4(
     X=&FloatFromCurve(FloatCurve=DataInterface() {propertyValues {Curve {keys [<at 0 0.5>,<at 1 1>]}}})
     Y=2
     Z=&FloatFromCurve(FloatCurve=DataInterface() {propertyValues {Curve {keys [
          {interpMode=RCIM_Cubic, tangentMode=RCTM_Break, time=0.5, value=-0.5}
          {arriveTangent=1.2, interpMode=RCIM_Cubic, leaveTangent=1.2, time=0.7034, value=-0.0784}
          <at 1 0.1>
         ]}}})
     W=0
    )
    Particles.Lightning Param 2:Vector4f=&MakeVector4(
     X=&Multiply_Float(
      A=&FloatFromCurve(
       FloatCurve=DataInterface() {propertyValues {Curve {keys [
           {interpMode=RCIM_Cubic, leaveTangent=8.1085, tangentMode=RCTM_Break, time=0.7695, value=1}
           {arriveTangent=-9.494, interpMode=RCIM_Cubic, leaveTangent=-9.494, tangentMode=RCTM_User, time=0.9048, value=0.8212}
           {tangentMode=RCTM_User, time=1, value=0}
          ]}}}
       Scale Curve=0.8
      )
      B=&Divide_Float(
       A=&MakeFloatFromVector(Vector=$Particles.Scale, Channel=::NewEnumerator0)
       B=&RandomRangeFloat(Minimum=10, Maximum=15)
      )
     )
     Y=0
     Z=&RandomRangeFloat(Minimum=0, Maximum=5)
     W=25
    )
    Particles.Lightning Param 3:Vector4f=&MakeVector4(
     X=&FloatFromCurve(
      FloatCurve=DataInterface() {propertyValues {Curve {keys [
          {interpMode=RCIM_Cubic, leaveTangent=-0.0025, tangentMode=RCTM_Break, time=0, value=0}
          {arriveTangent=1.8528, interpMode=RCIM_Cubic, tangentMode=RCTM_Break, time=1, value=1}
         ]}}}
      Scale Curve=&RandomRangeFloat(Minimum=0, Maximum=0)
     )
     Y=&FloatFromCurve(
      FloatCurve=DataInterface() {propertyValues {Curve {keys [
          {interpMode=RCIM_Cubic, leaveTangent=-0.0025, tangentMode=RCTM_Break, time=0, value=0.5}
          {arriveTangent=-0.0316, interpMode=RCIM_Cubic, tangentMode=RCTM_Break, time=1, value=-0.2}
         ]}}}
      Scale Curve=0
     )
     Z=&FloatFromCurve(FloatCurve=DataInterface() {propertyValues {Curve {keys [
          {interpMode=RCIM_Cubic, leaveTangent=1.1034, tangentMode=RCTM_Break, time=0, value=-1}
          {arriveTangent=2.2399, tangentMode=RCTM_User, time=0.8976, value=1.2895}
          <at 0.9997 -1.8499>
         ]}}})
     W=&FloatFromCurve(FloatCurve=DataInterface() {propertyValues {Curve {keys [
          {interpMode=RCIM_Cubic, leaveTangent=1.8089, tangentMode=RCTM_Break, time=0.1837, value=-0.5}
          {arriveTangent=2.0941, interpMode=RCIM_Cubic, leaveTangent=2.0941, time=0.5787, value=0.4567}
          {arriveTangent=2.0757, interpMode=RCIM_Cubic, tangentMode=RCTM_Break, time=0.9, value=1}
         ]}}})
    )
   ) {isSetParameters=true}
   &DynamicMaterialParameters(
    Index 0 Param 2=&RandomRangeFloat(Minimum=0, Maximum=0)
    Index 0 Param 4=13
    Index 1 Param 1=&FloatFromCurve(FloatCurve=DataInterface() {propertyValues {Curve {keys [<at 0 0.5>,<at 1 1>]}}})
    Index 1 Param 2=2
    Index 1 Param 3=&FloatFromCurve(
     FloatCurve=DataInterface() {propertyValues {Curve {keys [
         {interpMode=RCIM_Cubic, tangentMode=RCTM_Break, time=0.5, value=-0.5}
         {arriveTangent=1.2, interpMode=RCIM_Cubic, leaveTangent=1.2, time=0.7031, value=-0.1575}
         <at 1 0.1>
        ]}}}
     Scale Curve=&RandomRangeFloat(Minimum=1, Maximum=1.2)
    )
    Index 2 Param 1=&Multiply_Float(
     A=&FloatFromCurve(
      FloatCurve=DataInterface() {propertyValues {Curve {keys [
          {arriveTangent=13.4781, interpMode=RCIM_Cubic, leaveTangent=13.4781, tangentMode=RCTM_User, time=-0.0213, value=1.2493}
          {arriveTangent=-5.8724, interpMode=RCIM_Cubic, leaveTangent=-5.8724, tangentMode=RCTM_User, time=0.8725, value=0.6938}
          {tangentMode=RCTM_User, time=1, value=0}
         ]}}}
      Scale Curve=&RandomRangeFloat(Minimum=0.6, Maximum=1.2)
     )
     B=&Divide_Float(
      A=&MakeFloatFromVector(Vector=$Particles.Scale, Channel=::NewEnumerator0)
      B=&RandomRangeFloat(Minimum=10, Maximum=15)
     )
    )
    Index 2 Param 2=&FloatFromCurve(
     FloatCurve=DataInterface() {propertyValues {Curve {keys [
         {interpMode=RCIM_Cubic, tangentMode=RCTM_Break, time=0.7, value=0}
         {arriveTangent=-0.0859, interpMode=RCIM_Cubic, tangentMode=RCTM_Break, time=1, value=-1}
        ]}}}
     Scale Curve=50
    )
    Index 2 Param 3=&Waveform(
     Global Input Position=&RandomRangeFloat(Minimum=0, Maximum=15)
     Global Phase=$Particles.Age
     "Waveform [1]"=::NewEnumerator12
     "[1] Phase Scale"=2
     "[1] Frequency"=1
     "[1] Amplitude Min/Max"=<0 1>
    )
    Index 2 Param 4=25
    Index 3 Param 1=&FloatFromCurve(
     FloatCurve=DataInterface() {propertyValues {Curve {keys [
         {interpMode=RCIM_Cubic, leaveTangent=-0.0025, tangentMode=RCTM_Break, time=0, value=0}
         {arriveTangent=1.8528, interpMode=RCIM_Cubic, tangentMode=RCTM_Break, time=1, value=1}
        ]}}}
     Scale Curve=&RandomRangeFloat(Minimum=0, Maximum=0)
    )
    Index 3 Param 2=&FloatFromCurve(
     FloatCurve=DataInterface() {propertyValues {Curve {keys [
         {interpMode=RCIM_Cubic, leaveTangent=-0.0025, tangentMode=RCTM_Break, time=0, value=0.5}
         {arriveTangent=-0.0316, interpMode=RCIM_Cubic, tangentMode=RCTM_Break, time=1, value=-0.2}
        ]}}}
     Scale Curve=0
    )
    Index 3 Param 3=&FloatFromCurve(FloatCurve=DataInterface() {propertyValues {Curve {keys [
         {interpMode=RCIM_Cubic, leaveTangent=1.1034, tangentMode=RCTM_Break, time=0, value=-1}
         {arriveTangent=2.2399, tangentMode=RCTM_User, time=0.5, value=0.125}
        ]}}})
    Index 3 Param 4=&FloatFromCurve(FloatCurve=DataInterface() {propertyValues {Curve {keys [
         {interpMode=RCIM_Cubic, leaveTangent=1.8089, tangentMode=RCTM_Break, time=0.1837, value=-0.5}
         {arriveTangent=2.0941, interpMode=RCIM_Cubic, leaveTangent=2.0941, time=0.5787, value=0.4567}
         {arriveTangent=2.0757, interpMode=RCIM_Cubic, tangentMode=RCTM_Break, time=0.9, value=1}
        ]}}})
    Module.Index0Param0Write=true
    Module.Index0Param1Write=true
    Module.Index0Param2Write=true
    Module.Index0Param3Write=true
    Module.Index1Param0Write=true
    Module.Index1Param1Write=true
    Module.Index1Param2Write=true
    Module.Index1Param3Write=true
    Module.Index2Param0Write=true
    Module.Index2Param1Write=true
    Module.Index2Param2Write=true
    Module.Index2Param3Write=true
    Module.Index3Param0Write=true
    Module.Index3Param1Write=true
    Module.Index3Param2Write=true
    Module.Index3Param3Write=true
    Write Parameter Index 1=true
    Write Parameter Index 2=true
    Write Parameter Index 3=true
   ) {enabled=false, module=DynamicMaterialParameters001}
   /SetVariables(
    Particles.BG Lightning Param 0:Vector4f=&MakeVector4(X=0, Y=0, Z=0, W=13)
    Particles.BG Lightning Param 1:Vector4f=&MakeVector4(
     X=&FloatFromCurve(FloatCurve=DataInterface() {propertyValues {Curve {keys [<at 0 0.5>,<at 1 1>]}}})
     Y=2
     Z=&FloatFromCurve(
      FloatCurve=DataInterface() {propertyValues {Curve {keys [
          {interpMode=RCIM_Cubic, tangentMode=RCTM_Break, time=0.5, value=-0.5}
          {arriveTangent=1.2, interpMode=RCIM_Cubic, leaveTangent=1.2, time=0.7031, value=-0.1575}
          <at 1 0.1>
         ]}}}
      Scale Curve=&RandomRangeFloat(Minimum=1, Maximum=1.2)
     )
     W=0
    )
    Particles.BG Lightning Param 2:Vector4f=&MakeVector4(
     X=&Multiply_Float(
      A=&FloatFromCurve(
       FloatCurve=DataInterface() {propertyValues {Curve {keys [
           {arriveTangent=13.4781, interpMode=RCIM_Cubic, leaveTangent=13.4781, tangentMode=RCTM_User, time=-0.0213, value=1.2493}
           {arriveTangent=-5.8724, interpMode=RCIM_Cubic, leaveTangent=-5.8724, tangentMode=RCTM_User, time=0.8725, value=0.6938}
           {tangentMode=RCTM_User, time=1, value=0}
          ]}}}
       Scale Curve=&RandomRangeFloat(Minimum=0.6, Maximum=1.2)
      )
      B=&Divide_Float(
       A=&MakeFloatFromVector(Vector=$Particles.Scale, Channel=::NewEnumerator0)
       B=&RandomRangeFloat(Minimum=10, Maximum=15)
      )
     )
     Y=&FloatFromCurve(
      FloatCurve=DataInterface() {propertyValues {Curve {keys [
          {interpMode=RCIM_Cubic, tangentMode=RCTM_Break, time=0.7, value=0}
          {arriveTangent=-0.0859, interpMode=RCIM_Cubic, tangentMode=RCTM_Break, time=1, value=-1}
         ]}}}
      Scale Curve=50
     )
     Z=&Waveform(
      Global Input Position=&RandomRangeFloat(Minimum=0, Maximum=15)
      Global Phase=$Particles.Age
      "Waveform [1]"=::NewEnumerator12
      "[1] Phase Scale"=2
      "[1] Frequency"=1
      "[1] Amplitude Min/Max"=<0 1>
     )
     W=25
    )
    Particles.BG Lightning Param 3:Vector4f=&MakeVector4(
     X=&FloatFromCurve(
      FloatCurve=DataInterface() {propertyValues {Curve {keys [
          {interpMode=RCIM_Cubic, leaveTangent=-0.0025, tangentMode=RCTM_Break, time=0, value=0}
          {arriveTangent=1.8528, interpMode=RCIM_Cubic, tangentMode=RCTM_Break, time=1, value=1}
         ]}}}
      Scale Curve=&RandomRangeFloat(Minimum=0, Maximum=0)
     )
     Y=&FloatFromCurve(
      FloatCurve=DataInterface() {propertyValues {Curve {keys [
          {interpMode=RCIM_Cubic, leaveTangent=-0.0025, tangentMode=RCTM_Break, time=0, value=0.5}
          {arriveTangent=-0.0316, interpMode=RCIM_Cubic, tangentMode=RCTM_Break, time=1, value=-0.2}
         ]}}}
      Scale Curve=0
     )
     Z=&FloatFromCurve(FloatCurve=DataInterface() {propertyValues {Curve {keys [
          {interpMode=RCIM_Cubic, leaveTangent=1.1034, tangentMode=RCTM_Break, time=0, value=-1}
          {arriveTangent=2.2399, tangentMode=RCTM_User, time=0.5, value=0.125}
         ]}}})
     W=&FloatFromCurve(FloatCurve=DataInterface() {propertyValues {Curve {keys [
          {interpMode=RCIM_Cubic, leaveTangent=1.8089, tangentMode=RCTM_Break, time=0.1837, value=-0.5}
          {arriveTangent=2.0941, interpMode=RCIM_Cubic, leaveTangent=2.0941, time=0.5787, value=0.4567}
          {arriveTangent=2.0757, interpMode=RCIM_Cubic, tangentMode=RCTM_Break, time=0.9, value=1}
         ]}}})
    )
   ) {isSetParameters=true}
   /SetVariables(
    Particles.ColorWave:NiagaraFloat=&Waveform(
     "Waveform [1]"=::NewEnumerator0
     "[1] Frequency"=&RandomRangeFloat(
      Minimum=&MakeCustomFloatFromBool(Bool=$System.isThirdTransition, True Float=1.25, False Float=1.5)
      Maximum=3.5
     )
     "[1] Amplitude Min/Max"=<2 3>
    )
    Particles.LinearColorWave:NiagaraFloat=&Waveform(
     "Waveform [1]"=::NewEnumerator0
     "[1] Frequency"=&RandomRangeFloat(Minimum=1.5, Maximum=3.5)
     "[1] Amplitude Min/Max"=&MakeCustomVector2DFromBool(Bool=$System.isThirdTransition, True Vector2D=<1 1.25>, False Vector2D=<0.85 1>)
    )
   ) {isSetParameters=true}
   /SetVariables(Particles.Inner Dots Color:LinearColor=&MakeLinearColorFromVectorAndFloat(
     "Vector (RGB)"=<0 0 0>
     "Float (Alpha)"=&MakeCustomFloatFromBool(
      Bool=$System.isFirstTransition
      True Float=&FloatFromCurve(
       FloatCurve=DataInterface() {propertyValues {Curve {keys [
           {interpMode=RCIM_Cubic, tangentMode=RCTM_Break, time=0.9, value=1}
           {arriveTangent=0.0285, interpMode=RCIM_Cubic, leaveTangent=0.0285, tangentMode=RCTM_User, time=1.4099, value=0.9}
           {interpMode=RCIM_Cubic, leaveTangent=-0.1133, time=1.4528, value=0.0282}
          ]}}}
       CurveIndex=$Emitter.Age
       Scale Curve=$Particles.ColorWave
      )
      False Float=&MakeCustomFloatFromBool(
       Bool=$System.isSecondTransition
       True Float=&FloatFromCurve(
        FloatCurve=DataInterface() {propertyValues {Curve {keys [
            {interpMode=RCIM_Cubic, tangentMode=RCTM_Break, time=0.5, value=1}
            {arriveTangent=0.0285, interpMode=RCIM_Cubic, leaveTangent=0.0285, tangentMode=RCTM_User, time=1, value=0.9}
            {interpMode=RCIM_Cubic, leaveTangent=0.5263, time=1.1, value=0}
           ]}}}
        CurveIndex=$Emitter.Age
        Scale Curve=$Particles.ColorWave
       )
       False Float=&FloatFromCurve(
        FloatCurve=DataInterface() {propertyValues {Curve {keys [
            {interpMode=RCIM_Cubic, tangentMode=RCTM_Break, time=0.24, value=1}
            {arriveTangent=0.0285, interpMode=RCIM_Cubic, leaveTangent=0.0285, tangentMode=RCTM_User, time=0.39, value=0.9}
            {interpMode=RCIM_Cubic, leaveTangent=0.1923, time=0.4, value=0}
           ]}}}
        CurveIndex=$Emitter.Age
        Scale Curve=$Particles.ColorWave
       )
      )
     )
    )) {isSetParameters=true}
   /SetVariables(
    Particles.Outline Dots Color:LinearColor=&MakeLinearColorFromVectorAndFloat(
     "Vector (RGB)"=&MakeVectorFromLinearColorRGB(Color=<rgba 0 0.3365 0.0261 1>)
     "Float (Alpha)"=&MakeCustomFloatFromBool(
      Bool=$System.isFirstTransition
      True Float=&FloatFromCurve(
       FloatCurve=DataInterface() {propertyValues {Curve {keys [
           {interpMode=RCIM_Cubic, tangentMode=RCTM_Break, time=0.9, value=1}
           {arriveTangent=0.0285, interpMode=RCIM_Cubic, leaveTangent=0.0285, tangentMode=RCTM_User, time=1.4099, value=0.9}
           {interpMode=RCIM_Cubic, leaveTangent=-0.1133, time=1.4528, value=0.0282}
          ]}}}
       CurveIndex=$Emitter.Age
       Scale Curve=$Particles.LinearColorWave
      )
      False Float=&MakeCustomFloatFromBool(
       Bool=$System.isSecondTransition
       True Float=&FloatFromCurve(
        FloatCurve=DataInterface() {propertyValues {Curve {keys [
            {interpMode=RCIM_Cubic, tangentMode=RCTM_Break, time=0.5, value=1}
            {arriveTangent=0.0285, interpMode=RCIM_Cubic, leaveTangent=0.0285, tangentMode=RCTM_User, time=1, value=0.9}
            {interpMode=RCIM_Cubic, leaveTangent=0.5263, time=1.1, value=0}
           ]}}}
        CurveIndex=$Emitter.Age
        Scale Curve=$Particles.LinearColorWave
       )
       False Float=&FloatFromCurve(
        FloatCurve=DataInterface() {propertyValues {Curve {keys [
            {interpMode=RCIM_Cubic, tangentMode=RCTM_Break, time=0.24, value=1}
            {arriveTangent=0.0285, interpMode=RCIM_Cubic, leaveTangent=0.0285, tangentMode=RCTM_User, time=0.39, value=0.9}
            {interpMode=RCIM_Cubic, leaveTangent=0.1923, time=0.4, value=0}
           ]}}}
        CurveIndex=$Emitter.Age
        Scale Curve=$Particles.LinearColorWave
       )
      )
     )
    )
    Particles.Outline Dots Size:Vector2f=&Multiply_Vector2D_ByFloat(Float=1.1, Vector2D=$Particles.SpriteSize)
   ) {isSetParameters=true}
   /SetVariables(Particles.Glow Color=&MakeLinearColorFromVectorAndFloat(
     "Vector (RGB)"=&MakeVectorFromLinearColorRGB(Color=&MakeCustomLinearColorFromBool(
       Bool=$System.isFirstTransition
       True Color=&ColorFromCurve(ColorCurve=DataInterface() {propertyValues {
          AlphaCurve {keys []}
          BlueCurve {keys [<at 0 0.0037 RCIM_Cubic>,<at 0.304 0>,<at 0.5555 0>,<at 0.6876 0>,<at 0.8089 0>,<at 0.8861 0>,<at 0.967 0.0037>]}
          GreenCurve {keys [<at 0 0.0406 RCIM_Cubic>,<at 0.304 1>,<at 0.5555 1.047>,<at 0.6876 2>,<at 0.8089 1.4462>,<at 0.8861 1>,<at 0.967 0.0406>]}
          RedCurve {keys [<at 0 0 RCIM_Cubic>,<at 0.304 0.2123>,<at 0.5555 0.3256>,<at 0.6876 0.7105>,<at 0.8089 0.533>,<at 0.8861 0.2123>,<at 0.967 0>]}
         }})
       False Color=&MakeCustomLinearColorFromBool(
        Bool=$System.isSecondTransition
        True Color=&ColorFromCurve(ColorCurve=DataInterface() {propertyValues {
           AlphaCurve {keys []}
           BlueCurve {keys [<at 0 0.0037 RCIM_Cubic>,<at 0.304 0>,<at 0.5289 0>,<at 0.6826 0.0027>,<at 0.7405 0>,<at 0.8861 0>,<at 0.967 0.0037>]}
           GreenCurve {keys [<at 0 0.0406 RCIM_Cubic>,<at 0.304 1>,<at 0.5289 1.047>,<at 0.6826 2>,<at 0.7405 1.4462>,<at 0.8861 1>,<at 0.967 0.0406>]}
           RedCurve {keys [<at 0 0 RCIM_Cubic>,<at 0.304 0.2123>,<at 0.5289 0.3256>,<at 0.6826 0.7651>,<at 0.7405 0.533>,<at 0.8861 0.2123>,<at 0.967 0>]}
          }})
        False Color=&ColorFromCurve(ColorCurve=DataInterface() {propertyValues {
           AlphaCurve {keys []}
           BlueCurve {keys [<at 0 0.0037 RCIM_Cubic>,<at 0.304 0>,<at 0.5291 0>,<at 0.624 0>,<at 0.7404 0>,<at 0.905 0>,<at 1 0.0037>]}
           GreenCurve {keys [<at 0 0.0406 RCIM_Cubic>,<at 0.304 1>,<at 0.5291 1.047>,<at 0.624 2>,<at 0.7404 1.4462>,<at 0.905 1>,<at 1 0.0406>]}
           RedCurve {keys [<at 0 0 RCIM_Cubic>,<at 0.304 0.2123>,<at 0.5291 0.3256>,<at 0.624 0.6978>,<at 0.7404 0.533>,<at 0.905 0.2123>,<at 1 0>]}
          }})
       )
      ))
     "Float (Alpha)"=&FloatFromCurve(FloatCurve=DataInterface() {propertyValues {Curve {
         defaultValue=1
         keys [
          {interpMode=RCIM_Cubic, tangentMode=RCTM_User, time=0, value=0}
          {arriveTangent=1.8535, interpMode=RCIM_Cubic, leaveTangent=1.8535, time=0.2789, value=0.2889}
          {arriveTangentWeight=0.0592, interpMode=RCIM_Cubic, leaveTangentWeight=0.2557, tangentMode=RCTM_User, time=0.5395, value=1}
          {
           arriveTangentWeight=0.1829
           interpMode=RCIM_Cubic
           leaveTangent=3.6217
           leaveTangentWeight=0.0185
           tangentMode=RCTM_Break
           time=0.9447
           value=1
          }
          {arriveTangent=-58.9457, tangentMode=RCTM_User, time=1, value=0}
         ]
        }}})
    )) {isSetParameters=true}
   &Light_Attributes(
    Color=1
    Diffuse Scale=1
    Falloff Exponent=1
    Light Position Offset:Vector3f=<0 0 50>
    Light Radius:NiagaraFloat=&MakeCustomFloatFromBool(Bool=&Boolean_OrOperation(Bool1=$User.bIsFrontEnd, Bool2=$User.bIsFrontEndPreview), True Float=0, False Float=150)
    Position=1
    Radius=1
    Scale Alpha:NiagaraFloat=&FloatFromCurve(FloatCurve=DataInterface() {propertyValues {Curve {keys [
         {arriveTangent=0.0635, interpMode=RCIM_Cubic, leaveTangent=0.0635, tangentMode=RCTM_User, time=0.0246, value=0}
         {arriveTangent=1.0981, interpMode=RCIM_Cubic, leaveTangent=1.0981, time=0.4851, value=0.9001}
         {interpMode=RCIM_Cubic, leaveTangent=-1.9475, time=0.9353, value=1}
        ]}}})
    Scale RGB:Vector3f=&MakeVectorFromLinearColorRGB(Color=&ColorFromCurve(ColorCurve=DataInterface() {propertyValues {
        AlphaCurve {keys []}
        BlueCurve {keys [<at 0.5299 0 RCIM_Cubic>,<at 1 0 RCIM_Cubic>]}
        GreenCurve {keys [<at 0.5299 2 RCIM_Cubic>,<at 1 2 RCIM_Cubic>]}
        RedCurve {keys [<at 0.5299 0.1209 RCIM_Cubic>,<at 1 0.7878 RCIM_Cubic>]}
       }}))
    ScaleA=true
    ScaleRGB=true
    Specular Scale=1
    UsePositionOffset=true
   )
  ]
  rendererClasses [
   &NiagaraMeshRendererProperties
   &NiagaraSpriteRendererProperties
   &NiagaraLightRendererProperties
  ]
  renderers [
   {rendererClass=&NiagaraMeshRendererProperties, rendererIndex=0}
   {rendererClass=&NiagaraMeshRendererProperties, rendererIndex=1}
   {rendererClass=&NiagaraMeshRendererProperties, rendererIndex=2}
   {rendererClass=&NiagaraMeshRendererProperties, rendererIndex=3}
   {rendererClass=&NiagaraMeshRendererProperties, rendererIndex=4}
   {rendererClass=&NiagaraMeshRendererProperties, rendererIndex=5}
   {rendererClass=&NiagaraMeshRendererProperties, rendererIndex=6}
   {rendererClass=&NiagaraMeshRendererProperties, rendererIndex=7}
   {rendererClass=&NiagaraSpriteRendererProperties, rendererIndex=8}
   {rendererClass=&NiagaraSpriteRendererProperties, rendererIndex=9}
   {rendererClass=&NiagaraSpriteRendererProperties, rendererIndex=10}
   {rendererClass=&NiagaraLightRendererProperties, rendererIndex=11}
  ]
 }
 FE_Ground {rendererClasses [], renderers []}
 Watch_BurstLightning {
  emitterUpdateScript [
   &EmitterState()
   &SpawnBurst_Instantaneous(
    Spawn Count=&MakeCustomIntFromBool(Bool=$System.IsFirstTransition, True INT=1, False INT=0)
    Use Spawn Probability=false
    UseLoopCountLimit=false
   )
   &SpawnBurst_Instantaneous(
    Spawn Count=&MakeCustomIntFromBool(Bool=$System.IsFirstTransition, True INT=1, False INT=0)
    Spawn Group=1
    Use Spawn Probability=false
    UseLoopCountLimit=false
   ) {module=SpawnBurst_Instantaneous001}
  ]
  particleSpawnScript [
   /SetVariables(Particles.VisibilityTag=&SpawnGroup()) {isSetParameters=true}
   &InitializeParticle(
    Color=&MaskLinearColor(
     Mask Source=0
     Mask Parameter=?
     Mask=0
     Masked Linear Color=<rgba 0.6992 1.25 0 1>
     Passthrough Linear Color=&MultiplyLinearColor(
      Linear Color=&RandomRangeLinearColor(Minimum=<rgba 1 1 1 1>, Maximum=<rgba 1.2 1.2 1.2 2>)
      Scale Factor=&MakeVector4_FromFloat(Float=0.8)
     )
    )
    Lifetime=&MaskFloat(
     Mask Source=0
     Mask Parameter=?
     Mask=0
     Masked Float=0.1
     Passthrough Float=&RandomRangeFloat(Minimum=0.27, Maximum=0.5)
    )
    Mesh Scale Mode=::NewEnumerator3
    Mesh Uniform Scale=&Multiply_Float(A=&RandomRangeFloat(Minimum=0.5, Maximum=0.55), B=0.3)
    Sprite Size Mode=::NewEnumerator3
    Uniform Sprite Size=280
    UsePositionOffset=false
    Write Material Random=true
   )
   &SkeletalMeshLocation(
    Bone Sampling Mode=::NewEnumerator4
    MaskSpawnGroup=false
    Sampled Position Offset=&MaskVector(Mask Source=0, Mask Parameter=?, Mask=0, Masked Vector=<-2 0 8>, Passthrough Vector=<-15 -3 -5>)
    Sampled Position Offset Coordinate Space:ENiagaraSkelMeshTransforms=::NewEnumerator2
    Skeletal Mesh=$System.Skeletal Mesh Reader
   )
   &CameraOffset(Camera Offset Amount=7)
   &InitialMeshOrientation(
    Mesh Orientation Mode=::NewEnumerator2
    Orientation Vector=&MaskVector(Mask Source=0, Mask Parameter=?, Mask=0, Masked Vector=<1 0 0>, Passthrough Vector=<0 1 0>)
    Rotation=&MaskVector(Mask Source=0, Mask Parameter=?, Mask=0, Masked Vector=<0 0 0>, Passthrough Vector=<0 0.25 0>)
    Use Rotation Vector=true
   )
  ]
  particleUpdateScript [
   &ParticleState()
   &ScaleSpriteSize(Uniform Curve Sprite Scale=DataInterface() {propertyValues {Curve {
       defaultValue=1
       keys [
        {arriveTangent=4.1785, interpMode=RCIM_Cubic, leaveTangent=4.1785, tangentMode=RCTM_User, time=0, value=0.25}
        {arriveTangent=1.1179, interpMode=RCIM_Cubic, leaveTangent=1.1179, tangentMode=RCTM_User, time=0.2042, value=0.75}
        {arriveTangent=0.0224, interpMode=RCIM_Cubic, leaveTangent=0.0224, tangentMode=RCTM_User, time=1, value=1}
       ]
      }}})
   &ScaleColor(
    Scale Alpha=&MaskFloat(
     Mask Source=0
     Mask Parameter=?
     Mask=0
     Masked Float=&FloatFromCurve(FloatCurve=DataInterface() {propertyValues {Curve {keys [<at -0.0046 1.0113 RCIM_Cubic>,<at 1.0011 0.002 RCIM_Cubic>]}}})
     Passthrough Float=&FloatFromCurve(FloatCurve=DataInterface() {propertyValues {Curve {keys [
          {tangentMode=RCTM_User, time=0, value=0.876}
          {arriveTangent=-4.3986, interpMode=RCIM_Cubic, leaveTangent=-4.3986, tangentMode=RCTM_User, time=0.1692, value=1}
          {arriveTangent=0.1205, interpMode=RCIM_Cubic, leaveTangent=0.1205, tangentMode=RCTM_User, time=1, value=0}
         ]}}})
    )
    Scale RGB=&MaskVector(
     Mask Source=0
     Mask Parameter=?
     Mask=0
     Masked Vector=&MakeVectorFromLinearColorRGB(Color=&ColorFromCurve(ColorCurve=DataInterface() {propertyValues {
         AlphaCurve {keys []}
         BlueCurve {keys [<at 0.0949 5 RCIM_Cubic>,<at 0.2972 1 RCIM_Cubic>]}
         GreenCurve {keys [<at 0.0949 5 RCIM_Cubic>,<at 0.2972 1 RCIM_Cubic>]}
         RedCurve {keys [<at 0.0949 5 RCIM_Cubic>,<at 0.2972 1 RCIM_Cubic>]}
        }}))
     Passthrough Vector=&Lerp_Vector(
      A=&VectorFromFloat(Value=&FloatFromCurve(FloatCurve=DataInterface() {propertyValues {Curve {keys [<at 0.2771 1>,<at 0.3786 1>]}}}, Scale Curve=20))
      B=&VectorFromFloat(Value=&RandomRangeFloat(Minimum=0, Maximum=2, Evaluation Type=::NewEnumerator1))
      Alpha=&FloatFromCurve(FloatCurve=DataInterface() {propertyValues {Curve {keys [
           {interpMode=RCIM_Cubic, leaveTangent=2.5294, tangentMode=RCTM_Break, time=0, value=0}
           {arriveTangent=-0.2166, interpMode=RCIM_Cubic, tangentMode=RCTM_Break, time=0.4, value=1}
           <at 0.9989 0 RCIM_Cubic>
          ]}}})
     )
    )
    ScaleA=true
    ScaleRGB=true
   )
   &SolveForcesAndVelocity(Clamp Velocity=false, Limit Acceleration=false)
   &DynamicMaterialParameters(
    Index 0 Param 2=&RandomRangeFloat(Minimum=0, Maximum=0)
    Index 0 Param 4=13
    Index 1 Param 1=&FloatFromCurve(FloatCurve=DataInterface() {propertyValues {Curve {keys [<at 0 0.5>,<at 1 1>]}}})
    Index 1 Param 2=2
    Index 1 Param 3=&FloatFromCurve(FloatCurve=DataInterface() {propertyValues {Curve {keys [{interpMode=RCIM_Cubic, tangentMode=RCTM_Break, time=0.5, value=-0.5}, <at 1 0.1>]}}})
    Index 2 Param 1=&Multiply_Float(
     A=&FloatFromCurve(
      FloatCurve=DataInterface() {propertyValues {Curve {keys [{interpMode=RCIM_Cubic, leaveTangent=8.1085, tangentMode=RCTM_Break, time=0.8568, value=1}, {tangentMode=RCTM_User, time=1, value=0}]}}}
      Scale Curve=0.8
     )
     B=&Divide_Float(
      A=&MakeFloatFromVector(Vector=$Particles.Scale, Channel=::NewEnumerator0)
      B=&RandomRangeFloat(Minimum=10, Maximum=15)
     )
    )
    Index 2 Param 2=&FloatFromCurve(
     FloatCurve=DataInterface() {propertyValues {Curve {keys [
         {interpMode=RCIM_Cubic, tangentMode=RCTM_Break, time=0.7, value=0}
         {arriveTangent=-0.0859, interpMode=RCIM_Cubic, tangentMode=RCTM_Break, time=1, value=-1}
        ]}}}
     Scale Curve=50
    )
    Index 2 Param 3=&Waveform(
     Global Input Position=&RandomRangeFloat(Minimum=0, Maximum=15)
     Global Phase=$Particles.Age
     "Waveform [1]"=::NewEnumerator12
     "[1] Phase Scale"=2
     "[1] Frequency"=1
     "[1] Amplitude Min/Max"=<0 1>
    )
    Index 2 Param 4=25
    Index 3 Param 1=&FloatFromCurve(
     FloatCurve=DataInterface() {propertyValues {Curve {keys [
         {interpMode=RCIM_Cubic, leaveTangent=-0.0025, tangentMode=RCTM_Break, time=0, value=0}
         {arriveTangent=1.8528, interpMode=RCIM_Cubic, tangentMode=RCTM_Break, time=1, value=1}
        ]}}}
     Scale Curve=&RandomRangeFloat(Minimum=0, Maximum=0)
    )
    Index 3 Param 2=&FloatFromCurve(
     FloatCurve=DataInterface() {propertyValues {Curve {keys [
         {interpMode=RCIM_Cubic, leaveTangent=-0.0025, tangentMode=RCTM_Break, time=0, value=0.5}
         {arriveTangent=-0.0316, interpMode=RCIM_Cubic, tangentMode=RCTM_Break, time=1, value=-0.2}
        ]}}}
     Scale Curve=0
    )
    Index 3 Param 3=&FloatFromCurve(FloatCurve=DataInterface() {propertyValues {Curve {keys [
         {interpMode=RCIM_Cubic, leaveTangent=1.1034, tangentMode=RCTM_Break, time=0, value=-1}
         {arriveTangent=2.2399, tangentMode=RCTM_User, time=0.5, value=0.125}
        ]}}})
    Index 3 Param 4=&FloatFromCurve(FloatCurve=DataInterface() {propertyValues {Curve {keys [
         {interpMode=RCIM_Cubic, leaveTangent=1.8089, tangentMode=RCTM_Break, time=0.3, value=-0.5}
         {arriveTangent=2.0757, interpMode=RCIM_Cubic, tangentMode=RCTM_Break, time=0.9, value=1}
        ]}}})
    Module.Index0Param0Write=true
    Module.Index0Param1Write=true
    Module.Index0Param2Write=true
    Module.Index0Param3Write=true
    Module.Index1Param0Write=true
    Module.Index1Param1Write=true
    Module.Index1Param2Write=true
    Module.Index1Param3Write=true
    Module.Index2Param0Write=true
    Module.Index2Param1Write=true
    Module.Index2Param2Write=true
    Module.Index2Param3Write=true
    Module.Index3Param0Write=true
    Module.Index3Param1Write=true
    Module.Index3Param2Write=true
    Module.Index3Param3Write=true
    Write Parameter Index 1=true
    Write Parameter Index 2=true
    Write Parameter Index 3=true
   )
   /SetVariables(Particles.Star Size:Vector2f=&Multiply_Vector2D_ByFloat(Float=0.4, Vector2D=$Particles.SpriteSize)) {isSetParameters=true}
   &ScaleMeshSize(Scale Factor=&VectorFromFloat(Value=&FloatFromCurve(FloatCurve=DataInterface() {propertyValues {Curve {keys [
          {arriveTangent=35.758, interpMode=RCIM_Cubic, leaveTangent=35.758, tangentMode=RCTM_User, time=0, value=0}
          {arriveTangent=-6.3979, interpMode=RCIM_Cubic, leaveTangent=4.4257, tangentMode=RCTM_Break, time=0.1425, value=0.8597}
          {arriveTangent=0.0224, interpMode=RCIM_Cubic, leaveTangent=0.0224, tangentMode=RCTM_User, time=0.2523, value=1.1167}
          <at 1.0411 0.2664 RCIM_Cubic>
         ]}}})))
   &ScaleMeshSize(Scale Factor=&RandomRangeVector(Minimum=<1 1 1.5>, Maximum=<1 1 2>)) {module=ScaleMeshSize001}
   /SetVariables(Particles.Corona Dynamic Param:Vector4f=&MakeVector4(X=1, Y=1, Z=1, W=1)) {isSetParameters=true}
  ]
  rendererClasses [&NiagaraSpriteRendererProperties, &NiagaraMeshRendererProperties]
  renderers [
   {rendererClass=&NiagaraSpriteRendererProperties}
   {rendererClass=&NiagaraSpriteRendererProperties, rendererIndex=1}
   {rendererClass=&NiagaraMeshRendererProperties, rendererIndex=2}
   {rendererClass=&NiagaraMeshRendererProperties, rendererIndex=3}
  ]
 }
}
summary {
 systemName=NS_Emote_BiomeHose_Transform
 emitters [
  {
   emitterName=Watch_BurstLightning
   rendererClasses [&NiagaraSpriteRendererProperties, &NiagaraMeshRendererProperties]
  }
  {
   emitterName=BuildUp_Condensed
   rendererClasses [
    &NiagaraMeshRendererProperties
    &NiagaraSpriteRendererProperties
    &NiagaraLightRendererProperties
   ]
  }
  {emitterName=FE_Ground, rendererClasses []}
 ]
 userVariables [
  User.bIsFrontEnd {type=NiagaraBool}
  User.bIsFrontEndPreview {type=NiagaraBool}
  User.BurstIndex {type=NiagaraFloat, defaultValue=0, description=""}
 ]
}
