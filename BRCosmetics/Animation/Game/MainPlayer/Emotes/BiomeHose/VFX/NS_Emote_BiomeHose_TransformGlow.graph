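# Niagara system for the BiomeHose emote's transform glow: a green energy explosion burst.
# EnergyExplosion spawns four partitioned particles for expanding meshes and a sprite, with curve-driven color parameters.
# Explosion packs flashes, sparks and drifting curl-noise motes into one emitter, switching settings per spawn group.
# User.BurstIndex sets a first-transition flag gating one burst and the light, whose radius is zero in frontend, while quality level adjusts scale and color.
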NiagaraSystem /BRCosmetics/Animation/Game/MainPlayer/Emotes/BiomeHose/VFX/NS_Emote_BiomeHose_TransformGlow
dependencies {
 usedDataInterfaces [
  &NiagaraDataInterfaceColorCurve
  &NiagaraDataInterfaceCurve
  &NiagaraDataInterfaceVectorCurve
  &NiagaraDataInterfacePlatformSet
  &NiagaraDataInterfaceSpriteRendererInfo
  &NiagaraDataInterfaceVector2DCurve
  &NiagaraDataInterfaceSocketReader
 ]
 usedDynamicInputs []
 usedModules [
  &self:SystemSpawnScript.SystemScriptSource.SystemScriptGraph.NiagaraNodeAssignment_1.SetVariables_#1
  &SystemState
  &self:SystemSpawnScript.SystemScriptSource.SystemScriptGraph.NiagaraNodeAssignment_2.SetVariables_#2
  &self:SystemSpawnScript.SystemScriptSource.SystemScriptGraph.NiagaraNodeAssignment_4.SetVariables_#3
 ]
}
emitters {
 EnergyExplosion {
  emitterUpdateScript [
   &EmitterState(Inactive Response=::NewEnumerator1, Life Cycle Mode=::NewEnumerator1, Loop Behavior=::NewEnumerator1, Loop Duration=1, UseLoopDelay=false)
   &SpawnBurst_Instantaneous(Spawn Count=4, Use Spawn Probability=false, UseLoopCountLimit=false)
   /SetVariables(Emitter.Emissive UV:LinearColor=&ColorFromCurve(
     ColorCurve=DataInterface() {propertyValues {
       AlphaCurve {keys [<at 0.0395 1>,<at 0.3632 0.5902>,<at 0.8331 0>]}
       BlueCurve {keys [<at 0 0 RCIM_Cubic>,<at 0.9918 0>]}
       GreenCurve {keys [<at 0 1 RCIM_Cubic>,<at 0.9918 1>]}
       RedCurve {keys [<at 0 1 RCIM_Cubic>,<at 0.9918 0.9421>]}
      }}
     CurveIndex=$Emitter.NormalizedLoopAge
    )) {isSetParameters=true}
  ]
  particleSpawnScript [
   &PartitionParticles(Number Of Partitions=4, Particles Per Partition=1, Partition Mode:ENiagara_PartitionMode_v2=::NewEnumerator1)
   /SetVariables(Particles.VisibilityTag=$Particles.PartitionParticles.Partition) {isSetParameters=true}
   &InitializeParticle(
    Lifetime=&MergeHelper_FloatSwitch(0=0.5, 1=0.5, 2=0.6, 3=0.85, Index=$Particles.PartitionParticles.Partition, Num=::NewEnumerator2)
    Mesh Scale=&Multiply_Vector(
     A=&VectorFromFloat(Value=1.1)
     B=&MergeHelper_VectorSwitch(0=<-2.4 -1.6 -1.6>, 1=<1.4 1.4 1.4>, 2=<0 0 0>, 3=<-2.5 -1.7 -1.7>, Index=$Particles.PartitionParticles.Partition, Num=::NewEnumerator2)
    )
    Mesh Scale Mode=::NewEnumerator4
    Position Offset=<0 0 118>
    Sprite Rotation Mode=::NewEnumerator2
    Sprite Size Mode=::NewEnumerator3
    Uniform Sprite Size=300
    UsePositionOffset=true
    Write Material Random=true
   )
   &InitialMeshOrientation(
    Mesh Orientation Mode=::NewEnumerator2
    Rotation=&RandomRangeVector(Minimum=<0.4 0 0>, Maximum=<0.6 0 0>)
    Use Rotation Vector=true
   )
  ]
  particleUpdateScript [
   &ParticleState()
   &ScaleSpriteSize(
    Scale Sprite Size Mode=::NewEnumerator0
    Uniform Scale Factor=&FloatFromCurve(
     FloatCurve=DataInterface() {propertyValues {Curve {keys [
         {arriveTangent=2.0821, interpMode=RCIM_Cubic, leaveTangent=2.0821, tangentMode=RCTM_User, time=0, value=0.4}
         {arriveTangent=1.0135, interpMode=RCIM_Cubic, leaveTangent=1.0135, tangentMode=RCTM_User, time=0.269, value=0.8186}
         {arriveTangent=0.1533, interpMode=RCIM_Cubic, leaveTangent=0.1533, tangentMode=RCTM_User, time=1, value=1}
        ]}}}
     Scale Curve=1.1
    )
   )
   &ScaleMeshSize(Scale Factor=&Multiply_VectorByFloat(
     Vector=<1 1 1>
     Float=&FloatFromCurve(FloatCurve=DataInterface() {propertyValues {Curve {keys [
          {arriveTangent=5.164, interpMode=RCIM_Cubic, leaveTangent=5.164, tangentMode=RCTM_User, time=0, value=0.4}
          {arriveTangent=1.0135, interpMode=RCIM_Cubic, leaveTangent=1.0135, tangentMode=RCTM_User, time=0.2915, value=1.1296}
          {arriveTangent=0.1533, interpMode=RCIM_Cubic, leaveTangent=0.1533, tangentMode=RCTM_User, time=1, value=1.5}
         ]}}})
    ))
   &SolveForcesAndVelocity(Clamp Velocity=false, Limit Acceleration=false)
   &ScaleColor(
    Scale Alpha=&FloatFromCurve(FloatCurve=DataInterface() {propertyValues {Curve {keys [
         {arriveTangent=-0.0222, interpMode=RCIM_Cubic, leaveTangent=-0.0222, tangentMode=RCTM_User, time=0.2, value=0}
         {arriveTangent=7.4942, interpMode=RCIM_Cubic, leaveTangent=-1.7347, tangentMode=RCTM_Break, time=0.4, value=1}
         {arriveTangent=-0.6673, interpMode=RCIM_Cubic, leaveTangent=-0.6673, tangentMode=RCTM_User, time=0.6, value=0.375}
         {arriveTangent=-0.8204, interpMode=RCIM_Cubic, leaveTangent=-0.8204, tangentMode=RCTM_User, time=0.8, value=0.23}
         {arriveTangent=-0.2805, interpMode=RCIM_Cubic, leaveTangent=-0.2805, tangentMode=RCTM_User, time=1, value=0}
        ]}}})
    Scale RGB=&MakeVectorFromLinearColorRGB(Color=&ColorFromCurve(ColorCurve=DataInterface() {propertyValues {
        AlphaCurve {keys []}
        BlueCurve {keys [<at 0.237 0>,<at 0.7831 0>,<at 0.9357 0>]}
        GreenCurve {keys [<at 0.237 20>,<at 0.7831 1>,<at 0.9357 0.0781>]}
        RedCurve {keys [<at 0.237 3.7827>,<at 0.7831 0.0054>,<at 0.9357 0.0004>]}
       }}))
    ScaleA=true
    ScaleRGB=true
   )
   /SetVariables(Particles.Soft Energy Color:LinearColor=&ColorFromCurve(ColorCurve=DataInterface() {propertyValues {
       AlphaCurve {keys [<at 0.0602 2>, {arriveTangent=-2.0632, interpMode=RCIM_Cubic, time=0.1359, value=0.5}, <at 1 0 RCIM_Cubic>]}
       BlueCurve {keys [
         <at 0 0.0621 RCIM_Cubic>
         {arriveTangent=-0.0657, interpMode=RCIM_Cubic, leaveTangent=-0.0657, time=0.5616, value=0.0147}
         <at 0.8455 0.0065>
        ]}
       GreenCurve {keys [
         <at 0 0.7323 RCIM_Cubic>
         {arriveTangent=-0.7441, interpMode=RCIM_Cubic, leaveTangent=-0.7441, time=0.5616, value=0.2323}
         <at 0.8455 0.1031>
        ]}
       RedCurve {keys [<at 0 0 RCIM_Cubic>,<at 0.5616 0 RCIM_Cubic>,<at 0.8455 0>]}
      }})) {isSetParameters=true}
   /SetVariables(Particles.Dark Color:LinearColor=&MakeLinearColorFromVectorAndFloat(
     "Vector (RGB)"=&MakeVectorFromLinearColorRGB(Color=&ColorFromCurve(ColorCurve=DataInterface() {propertyValues {
         AlphaCurve {keys []}
         BlueCurve {keys [<at 0.1742 0.0015>,<at 0.9007 0>]}
         GreenCurve {keys [<at 0.1742 0.0865>,<at 0.9007 0>]}
         RedCurve {keys [<at 0.1742 0.0006>,<at 0.9007 0>]}
        }}))
     "Float (Alpha)"=&FloatFromCurve(
      FloatCurve=DataInterface() {propertyValues {Curve {keys [
          {arriveTangent=2.0596, interpMode=RCIM_Cubic, leaveTangent=2.0596, tangentMode=RCTM_User, time=0.2, value=0}
          {arriveTangent=7.4942, interpMode=RCIM_Cubic, leaveTangent=-1.7347, tangentMode=RCTM_Break, time=0.4, value=1}
          {arriveTangent=-0.6673, interpMode=RCIM_Cubic, leaveTangent=-0.6673, tangentMode=RCTM_User, time=0.5976, value=0.4597}
          {arriveTangent=-0.8204, interpMode=RCIM_Cubic, leaveTangent=-0.8204, tangentMode=RCTM_User, time=0.7976, value=0.3147}
          {arriveTangent=-0.2805, interpMode=RCIM_Cubic, leaveTangent=-0.2805, tangentMode=RCTM_User, time=1, value=0}
         ]}}}
      Scale Curve=1.5
     )
    )) {isSetParameters=true}
   /SetVariables(Particles.Bright Sprite Color:LinearColor=&MakeLinearColorFromVectorAndFloat(
     "Vector (RGB)"=&MakeVectorFromLinearColorRGB(Color=&ColorFromCurve(ColorCurve=DataInterface() {propertyValues {
         AlphaCurve {keys []}
         BlueCurve {keys [<at 0.0781 0>,<at 0.1264 0>,<at 0.8244 0>,<at 1 0>]}
         GreenCurve {keys [<at 0.0781 20>,<at 0.1264 4>,<at 0.8244 1>,<at 1 0.199>]}
         RedCurve {keys [<at 0.0781 6.2633>,<at 0.1264 1.1679>,<at 0.8244 0.219>,<at 1 0.0385>]}
        }}))
     "Float (Alpha)"=&FloatFromCurve(
      FloatCurve=DataInterface() {propertyValues {Curve {keys [
          {arriveTangent=5.4704, interpMode=RCIM_Cubic, leaveTangent=5.4704, tangentMode=RCTM_User, time=0.2, value=0.9812}
          {arriveTangent=-2.3514, interpMode=RCIM_Cubic, leaveTangent=-2.3514, tangentMode=RCTM_User, time=0.5976, value=0.4597}
          {arriveTangent=-0.8204, interpMode=RCIM_Cubic, leaveTangent=-0.8204, tangentMode=RCTM_User, time=0.7976, value=0.3147}
          {arriveTangent=-0.2805, interpMode=RCIM_Cubic, leaveTangent=-0.2805, tangentMode=RCTM_User, time=1, value=0}
         ]}}}
      Scale Curve=0.7
     )
    )) {isSetParameters=true}
   /SetVariables(Particles.Outline Color:LinearColor=&MakeLinearColorFromVectorAndFloat(
     "Vector (RGB)"=&MakeVectorFromLinearColorRGB(Color=&ColorFromCurve(ColorCurve=DataInterface() {propertyValues {
         AlphaCurve {keys []}
         BlueCurve {keys [<at 0.0042 0.1672>,<at 0.0791 0.064>,<at 0.1757 0>,<at 0.4209 0.0279>,<at 0.8333 0>,<at 0.9808 0>]}
         GreenCurve {keys [<at 0.0042 5>,<at 0.0791 5>,<at 0.1757 20>,<at 0.4209 4>,<at 0.8333 1>,<at 0.9808 0.199>]}
         RedCurve {keys [<at 0.0042 0.4678>,<at 0.0791 0.4566>,<at 0.1757 5.1729>,<at 0.4209 0.5865>,<at 0.8333 0.219>,<at 0.9808 0.0385>]}
        }}))
     "Float (Alpha)"=&FloatFromCurve(
      FloatCurve=DataInterface() {propertyValues {Curve {keys [
          {arriveTangent=1.3091, interpMode=RCIM_Cubic, leaveTangent=1.3091, tangentMode=RCTM_User, time=0.2, value=0.9812}
          {arriveTangent=-2.3514, interpMode=RCIM_Cubic, leaveTangent=-2.3514, tangentMode=RCTM_User, time=0.5488, value=0.4362}
          {arriveTangent=-0.2805, interpMode=RCIM_Cubic, leaveTangent=-0.2805, tangentMode=RCTM_User, time=1, value=0}
         ]}}}
      Scale Curve=0.25
     )
    )) {isSetParameters=true}
  ]
  rendererClasses [&NiagaraMeshRendererProperties, &NiagaraSpriteRendererProperties]
  renderers [
   {rendererClass=&NiagaraMeshRendererProperties, rendererIndex=0}
   {rendererClass=&NiagaraMeshRendererProperties, rendererIndex=1}
   {rendererClass=&NiagaraSpriteRendererProperties, rendererIndex=2}
   {rendererClass=&NiagaraMeshRendererProperties, rendererIndex=3}
   {rendererClass=&NiagaraMeshRendererProperties, rendererIndex=4}
  ]
 }
 Explosion {
  emitterUpdateScript [
   &EmitterState()
   /SetVariables(Emitter.isFirstTransition:NiagaraBool=&SetBoolByFloatComparison(A=$User.BurstIndex, B=0.9, Comparison Type=::NewEnumerator6)) {isSetParameters=true}
   &SpawnBurst_Instantaneous(Use Spawn Probability=false, UseLoopCountLimit=false)
   &SpawnBurst_Instantaneous(Spawn Group=1, Use Spawn Probability=false, UseLoopCountLimit=false) {module=SpawnBurst_Instantaneous001}
   &SpawnBurst_Instantaneous(
    Spawn Count=&MakeCustomIntFromBool(Bool=$Emitter.isFirstTransition, True INT=1, False INT=0)
    Spawn Group=2
    Use Spawn Probability=false
    UseLoopCountLimit=false
   ) {module=SpawnBurst_Instantaneous002}
   &SpawnBurst_Instantaneous(Spawn Group=3, Use Spawn Probability=false, UseLoopCountLimit=false) {module=SpawnBurst_Instantaneous010}
   &SpawnBurst_Instantaneous(Spawn Group=4, Use Spawn Probability=false, UseLoopCountLimit=false) {module=SpawnBurst_Instantaneous004}
   &SpawnBurst_Instantaneous(Spawn Group=5, Spawn Time=0.35, Use Spawn Probability=false, UseLoopCountLimit=false) {module=SpawnBurst_Instantaneous005}
   &SpawnBurst_Instantaneous(Spawn Count=40, Spawn Group=6, Use Spawn Probability=false, UseLoopCountLimit=false) {module=SpawnBurst_Instantaneous006}
   &SpawnBurst_Instantaneous(Spawn Count=30, Spawn Group=7, Use Spawn Probability=false, UseLoopCountLimit=false) {module=SpawnBurst_Instantaneous007}
   &SpawnRate(
    SpawnGroup=8
    SpawnRate=&FloatFromCurve(FloatCurve=DataInterface() {propertyValues {Curve {keys [<at 0 1>,<at 0.7081 0>]}}}, CurveIndex=$Emitter.Age, Scale Curve=100)
    Use Spawn Probability=false
   )
   &SpawnBurst_Instantaneous(Spawn Count=20, Spawn Group=8, Use Spawn Probability=false, UseLoopCountLimit=false) {module=SpawnBurst_Instantaneous008}
   &SpawnRate(
    SpawnGroup=9
    SpawnRate=&FloatFromCurve(
     FloatCurve=DataInterface() {propertyValues {Curve {keys [
         {arriveTangent=8.2259, interpMode=RCIM_Cubic, leaveTangent=8.2259, tangentMode=RCTM_User, time=0.0001, value=0.1}
         {arriveTangent=1.5378, interpMode=RCIM_Cubic, leaveTangent=1.5378, time=0.3, value=1}
         {arriveTangent=-0.8075, interpMode=RCIM_Cubic, leaveTangent=-0.8075, tangentMode=RCTM_User, time=0.5349, value=0.9224}
         {arriveTangent=-2.8828, interpMode=RCIM_Cubic, leaveTangent=-2.8828, tangentMode=RCTM_User, time=0.9, value=0}
        ]}}}
     CurveIndex=$Emitter.Age
     Scale Curve=7
    )
    Use Spawn Probability=false
   ) {module=SpawnRate001}
   &SpawnBurst_Instantaneous(Spawn Count=7, Spawn Group=9, Use Spawn Probability=false, UseLoopCountLimit=false) {module=SpawnBurst_Instantaneous009}
   /SetVariables(Emitter.isNotSwitch:NiagaraBool=&SetBoolByFloatComparison(
     A=&FN_PickFloatDependOnPlatform(
      TargetPlatform=DataInterface() {propertyValues {Platforms {
         qualityLevelMask=-30
         deviceProfileStates [
          {profileName=Linux, qualityLevelMask=0, setQualityLevelMask=3}
          {profileName=Mac, qualityLevelMask=0, setQualityLevelMask=3}
          {profileName=Windows, qualityLevelMask=0, setQualityLevelMask=3}
         ]
        }}}
      TargetPlatformActive=1
      TargetPlatformNoActive=0
     )
     B=0
     Comparison Type=::NewEnumerator4
    )) {isSetParameters=true}
  ]
  particleSpawnScript [
   /SetVariables(Particles.RandomFloat=&UniformRangedFloat(Minimum=0, Maximum=3)) {isSetParameters=true}
   /SetVariables(Particles.VisibilityTag=&SpawnGroup()) {isSetParameters=true}
   &PartitionParticles(Number Of Partitions=/Niagara/DynamicInputs/Helpers/MaskInt(Mask Source:NiagaraInt32=0, Mask Parameter:NiagaraInt32=?, Mask:NiagaraInt32=8, Masked Int:NiagaraInt32=5, Passthrough Int:NiagaraInt32=0))
   &InitializeParticle(
    Color=&MergeHelper_ColorSwitch(
     0=<rgba 1 1 1 1>
     1=<rgba 1 1 1 1>
     2=<rgba 1 1 1 1>
     3=<rgba 1 1 1 1>
     4=<rgba 1 1 1 1>
     5=&FNTimeOfDayColor(
      Night Color=<rgba 0.0029 0.1156 0 0.6>
      Morning Color=<rgba 0.0042 0.1656 0 0.7>
      Day Color=<rgba 0.0021 0.0823 0 0.9>
      Evening Color=<rgba 0.0043 0.1698 0 0.66>
     )
     6=&MultiplyLinearColor(
      Linear Color=&MultiplyLinearColor(
       Linear Color=&RandomRangeLinearColor(Minimum=<rgba 1 1 1 1>, Maximum=<rgba 0.776 0.776 0.776 1>)
       Scale Factor=&MakeVector4_FromLinearColor(LinearColor=<rgba 0 1 0.4471 1>)
      )
      Scale Factor=<1 1 1 1>
     )
     7=<rgba 0.8927 1 0.8427 1>
     8=<rgba 0.4798 1 0.1168 5>
     9=<rgba 0.4798 1 0.1168 5>
     Index=&SpawnGroup()
     Num=::NewEnumerator8
    )
    Lifetime=&MergeHelper_FloatSwitch(
     0=0.5
     1=0.8
     2=0.7
     3=0.5
     4=0.2
     5=0.9
     6=&RandomRangeFloat(Minimum=0.25, Maximum=0.5)
     7=&RandomRangeFloat(Minimum=0.2, Maximum=1)
     8=&RandomRangeFloat(Minimum=0.15, Maximum=1.1)
     9=&RandomRangeFloat(Minimum=0.5, Maximum=1.7)
     Index=&SpawnGroup()
     Num=::NewEnumerator8
    )
    Mesh Scale Mode=::NewEnumerator3
    Mesh Uniform Scale=&Multiply_Float(
     A=&MakeCustomFloatFromBool(
      Bool=&SetBoolByIntComparison(A=$Engine.QualityLevel, B=0, Comparison Type=::NewEnumerator4)
      True Float=1.1
      False Float=1
     )
     B=&Multiply_Float(A=1.1, B=1.8)
    )
    Position Offset=&MergeHelper_VectorSwitch(
     0=<0 0 118>
     1=<0 0 140>
     2=<0 0 0>
     3=<0 0 0>
     4=<0 0 120>
     5=<0 0 125>
     6=<0 0 0>
     7=<0 0 0>
     8=<0 0 0>
     9=<0 0 0>
     Index=&SpawnGroup()
     Num=::NewEnumerator8
    )
    Sprite Rotation Angle=&MaskFloat(Mask Source=0, Mask Parameter=?, Mask=1, Masked Float=0.25, Passthrough Float=0)
    Sprite Rotation Mode=::NewEnumerator0
    Sprite Size=&MergeHelper_Vector2Switch(
     0=<0 0>
     1=&Vector2DFromFloat(Value=1)
     2=<0 0>
     3=<0 0>
     4=&Vector2DFromFloat(Value=100)
     5=<420 420>
     6=&Multiply_Vector2D_ByFloat(Float=0.3, Vector2D=&RandomRangeVector2D(Minimum=<15 100>, Maximum=<20 150>))
     7=&Vector2DFromFloat(Value=&RandomRangeFloat(Minimum=5, Maximum=70))
     8=&Vector2DFromFloat(Value=&UniformAOrBFloat(
       A=&RandomRangeFloat(Minimum=3, Maximum=5)
       B=&RandomRangeFloat(Minimum=6, Maximum=9)
       Distribution Weight=0.25
       Evaluation Type=::NewEnumerator0
       Recalculate Random Each Loop=true
       Override Seed=false
      ))
     9=&Vector2DFromFloat(Value=&UniformAOrBFloat(
       A=&RandomRangeFloat(Minimum=3, Maximum=5)
       B=&RandomRangeFloat(Minimum=6, Maximum=9)
       Distribution Weight=0.25
       Evaluation Type=::NewEnumerator0
       Recalculate Random Each Loop=true
       Override Seed=false
      ))
     Index=&SpawnGroup()
     Num=::NewEnumerator8
    )
    Sprite Size Mode=::NewEnumerator4
    UsePositionOffset=true
    Write Material Random=true
   )
   &ShapeLocation(
    MaskSpawnGroup=false
    Non Uniform Scale=&MaskVector(Mask Source=0, Mask Parameter=?, Mask=6, Masked Vector=<1 1 1>, Passthrough Vector=<0 0 0>)
    Offset=&MaskVector(Mask Source=0, Mask Parameter=?, Mask=6, Masked Vector=<0 0 110>, Passthrough Vector=<0 0 0>)
    Offset Mode=::NewEnumerator0
    Sphere Distribution=::NewEnumerator1
    Sphere Radius=&MaskFloat(
     Mask Source=0
     Mask Parameter=?
     Mask=6
     Masked Float=&RandomRangeFloat(Minimum=5, Maximum=15)
     Passthrough Float=0
    )
    U Position=&Lerp_Float(A=0, B=1, Alpha=&ReturnNormalizedExecIndex(Normalized Index Scale=2))
    UseOwnerScale=true
    V Position=&RandomRangeFloat(Minimum=0, Maximum=1)
   )
   &ShapeLocation(
    MaskSpawnGroup=false
    Non Uniform Scale=&MaskVector(
     Mask Source=0
     Mask Parameter=?
     Mask=7
     Masked Vector=<1 1 1>
     Passthrough Vector=&MaskVector(
      Mask Source=0
      Mask Parameter=?
      Mask=8
      Masked Vector=<1 1 1>
      Passthrough Vector=&MaskVector(Mask Source=0, Mask Parameter=?, Mask=9, Masked Vector=<0.8 0.8 1>, Passthrough Vector=<0 0 0>)
     )
    )
    Offset=&MaskVector(
     Mask Source=0
     Mask Parameter=?
     Mask=7
     Masked Vector=<0 0 125>
     Passthrough Vector=&MaskVector(
      Mask Source=0
      Mask Parameter=?
      Mask=8
      Masked Vector=&MergeHelper_VectorSwitch(0=<-30 0 50>, 1=<50 0 150>, 2=<-20 0 150>, 3=<-60 0 80>, 4=<50 0 70>, Index=$Particles.PartitionParticles.Partition, Num=::NewEnumerator3)
      Passthrough Vector=&MaskVector(Mask Source=0, Mask Parameter=?, Mask=9, Masked Vector=<0 0 80>, Passthrough Vector=<0 0 0>)
     )
    )
    Offset Mode=::NewEnumerator0
    Override Seed=false
    Sphere Radius=&MaskFloat(
     Mask Source=0
     Mask Parameter=?
     Mask=7
     Masked Float=5
     Passthrough Float=&MaskFloat(
      Mask Source=0
      Mask Parameter=?
      Mask=8
      Masked Float=15
      Passthrough Float=&MaskFloat(Mask Source=0, Mask Parameter=?, Mask=9, Masked Float=80, Passthrough Float=0)
     )
    )
    Sphere Surface Distribution=1
    UseOwnerScale=true
   ) {module=ShapeLocation001}
   &AddVelocity(
    Module.Use Falloff Distance=false
    Origin Offset=&MaskVector(
     Mask Source=0
     Mask Parameter=?
     Mask=6
     Masked Vector=<0 0 110>
     Passthrough Vector=&MaskVector(Mask Source=0, Mask Parameter=?, Mask=7, Masked Vector=<0 0 125>, Passthrough Vector=<0 0 0>)
    )
    Override Seed=false
    Velocity Mode=::NewEnumerator1
    Velocity Speed=&MaskFloat(
     Mask Source=0
     Mask Parameter=?
     Mask=6
     Masked Float=&RandomRangeFloat(Minimum=100, Maximum=300)
     Passthrough Float=&MaskFloat(
      Mask Source=0
      Mask Parameter=?
      Mask=7
      Masked Float=&RandomRangeFloat(Minimum=450, Maximum=600)
      Passthrough Float=0
     )
    )
   )
   &AddVelocity(Velocity=&MaskVector(
     Mask Source=0
     Mask Parameter=?
     Mask=8
     Masked Vector=&Multiply_VectorByFloat(
      Vector=&MergeHelper_VectorSwitch(
       0=<400 300 -200>
       1=<200 -300 300>
       2=<0 400 400>
       3=<-200 -200 -200>
       4=<-300 -100 400>
       Index=$Particles.PartitionParticles.Partition
       Num=::NewEnumerator3
      )
      Float=1.5
     )
     Passthrough Vector=&MaskVector(
      Mask Source=0
      Mask Parameter=?
      Mask=9
      Masked Vector=&RandomRangeVector(Minimum=<-50 -50 50>, Maximum=<50 50 80>)
      Passthrough Vector=<0 0 0>
     )
    )) {module=AddVelocity001}
   &ApplyInitialForces()
   &CameraOffset(Camera Offset Amount=&MergeHelper_FloatSwitch(
     0=&MakeCustomFloatFromBool(
      Bool=&SetBoolByIntComparison(A=$Engine.QualityLevel, B=0, Comparison Type=::NewEnumerator4)
      True Float=5
      False Float=0
     )
     1=50
     2=0
     3=0
     4=0
     5=-50
     6=-15
     7=-30
     8=0
     9=0
     Index=&SpawnGroup()
     Num=::NewEnumerator8
    ))
   &InitialMeshOrientation(Mesh Axis To Orient=<0 0 1>, Mesh Orientation Mode=::NewEnumerator2, Use Rotation Vector=false)
   &SubUVAnimation(
    Output Alternate Renderer Binding=true
    "Sprite Renderer "=DataInterface() {propertyValues {SpriteRenderer=&self:Explosion_0.NiagaraSpriteRendererProperties_2}}
    SubUV Animation Mode:ENiagara_SubUVLookupModeV2=::NewEnumerator2
    UseEndFrame=false
    UseStartFrame=false
   )
  ]
  particleUpdateScript [
   &ParticleState()
   /SetVariables(Particles.undulate:NiagaraFloat=&Waveform(
     "Waveform [1]"=::NewEnumerator0
     "[1] Frequency"=&FloatFromCurve(FloatCurve=DataInterface() {propertyValues {Curve {keys [<at 0.0298 15.2341>,<at 0.1918 -0.2341>]}}}, CurveIndex=$Emitter.Age)
     "[1] Amplitude Min/Max"=<0.4 1>
    )) {isSetParameters=true}
   /SetVariables(Particles.wiggle:NiagaraFloat=&Waveform(
     "Waveform [1]"=::NewEnumerator0
     "[1] Frequency"=&RandomRangeFloat(Minimum=1.5, Maximum=3)
     "[1] Amplitude Min/Max"=<0.75 1.25>
    )) {isSetParameters=true}
   &ScaleSpriteSize(
    Scale Factor=&MergeHelper_Vector2Switch(
     0=<0 0>
     1=&Vector2DFromCurve(
      Vector2Curve=DataInterface() {propertyValues {
        XCurve {
         defaultValue=0
         keys [
          {interpMode=RCIM_Cubic, tangentMode=RCTM_User, time=0, value=0}
          {
           arriveTangent=0.8936
           arriveTangentWeight=0.2024
           interpMode=RCIM_Cubic
           leaveTangent=0.8936
           tangentMode=RCTM_User
           time=0.6071
           value=0.9716
          }
          {
           arriveTangent=0.4878
           arriveTangentWeight=0.0303
           interpMode=RCIM_Cubic
           leaveTangent=0.4878
           tangentMode=RCTM_User
           time=0.6981
           value=0.9847
          }
          <at 1 0>
         ]
        }
        YCurve {
         defaultValue=0
         keys [
          {interpMode=RCIM_Cubic, tangentMode=RCTM_User, time=0, value=0}
          {arriveTangentWeight=0.1333, interpMode=RCIM_Cubic, tangentMode=RCTM_User, time=0.4, value=1}
          <at 1 0>
         ]
        }
       }}
      ScaleCurve=<125 225>
     )
     2=<0 0>
     3=<0 0>
     4=&Vector2DFromFloat(Value=&FloatFromCurve(FloatCurve=DataInterface() {propertyValues {Curve {keys [
           {arriveTangent=7.2928, interpMode=RCIM_Cubic, leaveTangent=7.2928, tangentMode=RCTM_User, time=0, value=0.5}
           {arriveTangent=4.5, interpMode=RCIM_Cubic, leaveTangent=4.5, time=0.5096, value=3.5141}
           <at 1 5 RCIM_Cubic>
          ]}}}))
     5=&Vector2DFromFloat(Value=&FloatFromCurve(FloatCurve=DataInterface() {propertyValues {Curve {
          defaultValue=1
          keys [
           <at 0 1 RCIM_Cubic>
           {arriveTangent=-0.8928, interpMode=RCIM_Cubic, leaveTangent=-0.8928, tangentMode=RCTM_User, time=0.549, value=0.7971}
           {arriveTangent=-0.8989, interpMode=RCIM_Cubic, leaveTangent=-0.8989, time=0.598, value=0.6785}
           {arriveTangent=-0.5718, interpMode=RCIM_Cubic, leaveTangent=-14.2537, tangentMode=RCTM_Break, time=0.8716, value=0.5071}
           {arriveTangent=12.3329, interpMode=RCIM_Cubic, leaveTangent=12.3329, tangentMode=RCTM_User, time=0.9011, value=0.4391}
           {arriveTangent=32.4571, interpMode=RCIM_Cubic, leaveTangent=32.4571, time=0.9016, value=0.8684}
           {arriveTangent=-20.0455, interpMode=RCIM_Cubic, leaveTangent=-20.0455, time=0.9151, value=0.8908}
           {arriveTangent=-9.6165, interpMode=RCIM_Cubic, leaveTangent=-9.6165, time=0.9199, value=0.5004}
           {arriveTangent=0.6083, interpMode=RCIM_Cubic, leaveTangent=0.6083, time=0.9361, value=0.6885}
           {arriveTangent=-11.0982, interpMode=RCIM_Cubic, leaveTangent=-11.0982, time=0.9697, value=0.5307}
           <at 0.9981 0 RCIM_Cubic>
          ]
         }}}))
     6=<1 1>
     7=&Vector2DFromFloat(Value=&FloatFromCurve(FloatCurve=DataInterface() {propertyValues {Curve {
          defaultValue=1
          keys [
           <at 0 1.5576 RCIM_Cubic>
           {arriveTangent=-12.1624, interpMode=RCIM_Cubic, leaveTangent=0.4684, tangentMode=RCTM_Break, time=0.0801, value=1}
           {arriveTangent=-1.6002, interpMode=RCIM_Cubic, leaveTangent=-1.6002, tangentMode=RCTM_User, time=0.5897, value=0.5058}
           {tangentMode=RCTM_User, time=1, value=0}
          ]
         }}}))
     8=&Multiply_Vector2D(
      A=&Vector2DFromCurve(
       Vector2Curve=DataInterface() {propertyValues {
         XCurve {keys [
           <at 0 0 RCIM_Cubic>
           {arriveTangent=2.216, interpMode=RCIM_Cubic, leaveTangent=2.216, tangentMode=RCTM_User, time=0.3351, value=0.4289}
           {arriveTangent=1.0786, interpMode=RCIM_Cubic, leaveTangent=1.0786, tangentMode=RCTM_User, time=0.6725, value=0.9604}
           {arriveTangent=-0.7784, interpMode=RCIM_Cubic, leaveTangent=-0.7784, tangentMode=RCTM_User, time=0.8783, value=1}
           <at 1 0 RCIM_Cubic>
          ]}
         YCurve {keys [
           <at 0 1 RCIM_Cubic>
           {arriveTangent=-0.4162, interpMode=RCIM_Cubic, leaveTangent=-0.4162, time=0.3268, value=0.5904}
           {arriveTangent=-0.877, interpMode=RCIM_Cubic, leaveTangent=-0.877, time=0.6767, value=0.7184}
           <at 1 0 RCIM_Cubic>
          ]}
        }}
       ScaleCurve=<1 1>
      )
      B=&MakeVector2D(X=$Particles.wiggle, Y=$Particles.wiggle)
     )
     9=&Multiply_Vector2D(
      A=&Vector2DFromCurve(
       Vector2Curve=DataInterface() {propertyValues {
         XCurve {keys [
           <at 0 0 RCIM_Cubic>
           {arriveTangent=2.216, interpMode=RCIM_Cubic, leaveTangent=2.216, tangentMode=RCTM_User, time=0.3351, value=0.4289}
           {arriveTangent=1.0786, interpMode=RCIM_Cubic, leaveTangent=1.0786, tangentMode=RCTM_User, time=0.6725, value=0.9604}
           {arriveTangent=-0.7784, interpMode=RCIM_Cubic, leaveTangent=-0.7784, tangentMode=RCTM_User, time=0.8783, value=1}
           <at 1 0 RCIM_Cubic>
          ]}
         YCurve {keys [
           <at 0 1 RCIM_Cubic>
           {arriveTangent=-0.4162, interpMode=RCIM_Cubic, leaveTangent=-0.4162, time=0.3268, value=0.5904}
           {arriveTangent=-0.877, interpMode=RCIM_Cubic, leaveTangent=-0.877, time=0.6767, value=0.7184}
           <at 1 0 RCIM_Cubic>
          ]}
        }}
       ScaleCurve=<1 1>
      )
      B=&MakeVector2D(X=$Particles.wiggle, Y=$Particles.wiggle)
     )
     Index=&SpawnGroup()
     Num=::NewEnumerator8
    )
    Scale Sprite Size Mode=::NewEnumerator2
   )
   &ScaleMeshSize(Scale Factor=&Multiply_VectorByFloat(
     Vector=<1 1 1>
     Float=&MakeCustomFloatFromBool(
      Bool=&SetBoolByIntComparison(A=$Engine.QualityLevel, B=0, Comparison Type=::NewEnumerator4)
      True Float=&FloatFromCurve(FloatCurve=DataInterface() {propertyValues {Curve {keys [
           {arriveTangent=3.9768, interpMode=RCIM_Cubic, leaveTangent=3.9768, tangentMode=RCTM_User, time=0, value=0.75}
           {arriveTangent=1.0135, interpMode=RCIM_Cubic, leaveTangent=1.0135, tangentMode=RCTM_User, time=0.2914, value=1.3}
           {arriveTangent=0.1533, interpMode=RCIM_Cubic, leaveTangent=0.1533, tangentMode=RCTM_User, time=1, value=1.5}
          ]}}})
      False Float=&FloatFromCurve(FloatCurve=DataInterface() {propertyValues {Curve {keys [
           {arriveTangent=5.164, interpMode=RCIM_Cubic, leaveTangent=5.164, tangentMode=RCTM_User, time=0, value=0.4}
           {arriveTangent=1.0135, interpMode=RCIM_Cubic, leaveTangent=1.0135, tangentMode=RCTM_User, time=0.2914, value=1.1296}
           {arriveTangent=0.1533, interpMode=RCIM_Cubic, leaveTangent=0.1533, tangentMode=RCTM_User, time=1, value=1.5}
          ]}}})
     )
    )) {module=ScaleMeshSize001}
   &ScaleColor(
    Scale Alpha=&MergeHelper_FloatSwitch(
     0=&FloatFromCurve(FloatCurve=DataInterface() {propertyValues {Curve {
         defaultValue=1
         keys [
          {interpMode=RCIM_Cubic, tangentMode=RCTM_User, time=0.3287, value=1}
          {arriveTangent=-1.4897, interpMode=RCIM_Cubic, leaveTangent=-1.4897, time=0.6332, value=0.1917}
          {interpMode=RCIM_Cubic, tangentMode=RCTM_User, time=1, value=0}
         ]
        }}})
     1=&FloatFromCurve(FloatCurve=DataInterface() {propertyValues {Curve {keys [
          {interpMode=RCIM_Cubic, tangentMode=RCTM_User, time=0, value=0}
          {interpMode=RCIM_Cubic, tangentMode=RCTM_User, time=0.196, value=1.0282}
          {arriveTangent=-1.2789, interpMode=RCIM_Cubic, leaveTangent=-1.2789, time=0.5097, value=0.3036}
          {tangentMode=RCTM_User, time=1, value=0}
         ]}}})
     2=0
     3=0
     4=&FloatFromCurve(
      FloatCurve=DataInterface() {propertyValues {Curve {keys [
          <at -0.0046 1.0113 RCIM_Cubic>
          {arriveTangent=-1.0349, interpMode=RCIM_Cubic, leaveTangent=-1.0349, time=0.435, value=0.834}
          {arriveTangent=-1.2597, interpMode=RCIM_Cubic, leaveTangent=-1.2597, tangentMode=RCTM_User, time=0.6814, value=0.3013}
          <at 1.0011 0.002 RCIM_Cubic>
         ]}}}
      Scale Curve=5
     )
     5=&FloatFromCurve(FloatCurve=DataInterface() {propertyValues {Curve {
         defaultValue=1
         keys [
          <at -0.0104 0 RCIM_Cubic>
          {interpMode=RCIM_Cubic, tangentMode=RCTM_User, time=0.1089, value=1}
          {interpMode=RCIM_Cubic, tangentMode=RCTM_User, time=1, value=0}
         ]
        }}})
     6=&FloatFromCurve(
      FloatCurve=DataInterface() {propertyValues {Curve {
         defaultValue=1
         keys [{interpMode=RCIM_Cubic, tangentMode=RCTM_User, time=0.3074, value=1}, {interpMode=RCIM_Cubic, tangentMode=RCTM_User, time=1, value=0}]
        }}}
      Scale Curve=10
     )
     7=&FloatFromCurve(
      FloatCurve=DataInterface() {propertyValues {Curve {
         defaultValue=1
         keys [
          {interpMode=RCIM_Cubic, tangentMode=RCTM_User, time=0.4453, value=1}
          {arriveTangent=-1.8986, interpMode=RCIM_Cubic, leaveTangent=-1.8986, tangentMode=RCTM_User, time=1, value=0}
         ]
        }}}
      Scale Curve=&Waveform(
       "Waveform [1]"=::NewEnumerator0
       "[1] Frequency"=&RandomRangeFloat(Minimum=1.5, Maximum=3.5)
       "[1] Amplitude Min/Max"=<0.75 1>
      )
     )
     8=&FloatFromCurve(
      FloatCurve=DataInterface() {propertyValues {Curve {
         defaultValue=1
         keys [
          {interpMode=RCIM_Cubic, tangentMode=RCTM_User, time=0.0357, value=1}
          {arriveTangent=-2.3528, interpMode=RCIM_Cubic, leaveTangent=-2.3528, tangentMode=RCTM_User, time=0.3134, value=0.5602}
          {interpMode=RCIM_Cubic, tangentMode=RCTM_User, time=0.8365, value=0}
         ]
        }}}
      Scale Curve=0.7
     )
     9=&FloatFromCurve(FloatCurve=DataInterface() {propertyValues {Curve {
         defaultValue=1
         keys [
          {interpMode=RCIM_Cubic, tangentMode=RCTM_User, time=0.0357, value=1}
          {arriveTangent=-2.3528, interpMode=RCIM_Cubic, leaveTangent=-2.3528, tangentMode=RCTM_User, time=0.3134, value=0.5602}
          {interpMode=RCIM_Cubic, tangentMode=RCTM_User, time=0.8365, value=0}
         ]
        }}})
     Index=&SpawnGroup()
     Num=::NewEnumerator8
    )
    Scale RGB=&MergeHelper_VectorSwitch(
     0=&MakeVectorFromLinearColorRGB(Color=&ColorFromCurve(ColorCurve=DataInterface() {propertyValues {
         AlphaCurve {keys []}
         BlueCurve {keys [
           <at 0.0215 10>
           <at 0.0614 0>
           {arriveTangent=0.008, interpMode=RCIM_Cubic, leaveTangent=0.008, time=0.2059, value=0}
           <at 0.5299 0.0064>
           <at 0.7805 0.0058>
           <at 0.9813 0.0071 RCIM_Cubic>
          ]}
         GreenCurve {keys [
           <at 0.0215 10>
           <at 0.0614 3>
           {arriveTangent=-2.7811, interpMode=RCIM_Cubic, leaveTangent=-2.7811, time=0.2059, value=2}
           <at 0.5299 2>
           <at 0.7805 1>
           <at 0.9813 0.1156 RCIM_Cubic>
          ]}
         RedCurve {keys [
           <at 0.0215 10>
           <at 0.0614 2.0325>
           {arriveTangent=-2.8263, interpMode=RCIM_Cubic, leaveTangent=-2.8263, time=0.2059, value=0.7526}
           <at 0.5299 0.3282>
           <at 0.7805 0>
           <at 0.9813 0.0193 RCIM_Cubic>
          ]}
        }}))
     1=&MakeVectorFromLinearColorRGB(Color=<rgba 0 1 0.1199 1>)
     2=<0 0 0>
     3=<0 0 0>
     4=&MakeVectorFromLinearColorRGB(Color=&ColorFromCurve(ColorCurve=DataInterface() {propertyValues {
         AlphaCurve {keys []}
         BlueCurve {keys [<at 0.078 0>,<at 0.2082 0 RCIM_Cubic>,<at 0.2095 0>,<at 0.3594 0 RCIM_Cubic>,<at 1 0>]}
         GreenCurve {keys [
           <at 0.078 5>
           {arriveTangent=-30.4294, interpMode=RCIM_Cubic, leaveTangent=-30.4294, time=0.2082, value=1}
           <at 0.2095 1>
           {arriveTangent=-1.0924, interpMode=RCIM_Cubic, leaveTangent=-1.0924, time=0.3594, value=0.3281}
           <at 1 0.1365>
          ]}
         RedCurve {keys [
           <at 0.078 0.1772>
           {arriveTangent=-1.0785, interpMode=RCIM_Cubic, leaveTangent=-1.0785, time=0.2082, value=0.0354}
           <at 0.2095 0.0354>
           {arriveTangent=-0.0403, interpMode=RCIM_Cubic, leaveTangent=-0.0403, time=0.3594, value=0.0086}
           <at 1 0.0036>
          ]}
        }}))
     5=<1 1 1>
     6=&Multiply_VectorByFloat(Vector=<1 1 1>, Float=20)
     7=&Lerp_Vector(
      A=<0 0 0>
      B=<1 1 1>
      Alpha=&FloatFromCurve(
       FloatCurve=DataInterface() {propertyValues {Curve {keys [<at 0.0108 1.0089 RCIM_Constant>,<at 0.0501 0 RCIM_Constant>,<at 0.1124 1 RCIM_Constant>,<at 0.1651 0 RCIM_Constant>]}}}
       CurveIndex=$Emitter.Age
      )
     )
     8=&VectorFromFloat(Value=&FloatFromCurve(FloatCurve=DataInterface() {propertyValues {Curve {keys [<at 0 1 RCIM_Constant>,<at 0.0702 0 RCIM_Constant>,<at 0.3689 1 RCIM_Constant>,<at 0.4893 0 RCIM_Constant>]}}}))
     9=<1 1 1>
     Index=&SpawnGroup()
     Num=::NewEnumerator8
    )
    ScaleA=true
    ScaleRGB=true
   )
   &Drag(
    Drag=&MaskFloat(
     Mask Source=0
     Mask Parameter=?
     Mask=7
     Masked Float=&FloatFromCurve(FloatCurve=DataInterface() {propertyValues {Curve {
         defaultValue=1
         keys [
          <at 0.0064 5.4535>
          <at 0.0217 0.5985>
          {tangentMode=RCTM_User, time=0.1187, value=0.3398}
          <at 0.1467 2.2952>
          {tangentMode=RCTM_User, time=0.1628, value=6.0933}
          <at 0.2026 7.5401>
         ]
        }}})
     Passthrough Float=&MaskFloat(
      Mask Source=0
      Mask Parameter=?
      Mask=8
      Masked Float=&FloatFromCurve(
       FloatCurve=DataInterface() {propertyValues {Curve {keys [<at 0 0.5>,<at 0.1886 1>,<at 1 2>]}}}
       Scale Curve=&RandomRangeFloat(Minimum=5, Maximum=7)
      )
      Passthrough Float=&MaskFloat(Mask Source=0, Mask Parameter=?, Mask=9, Masked Float=1, Passthrough Float=0)
     )
    )
    Use Linear Drag=true
    Use Rotational Drag=true
   )
   &GravityForce(Gravity=&MaskVector(
     Mask Source=0
     Mask Parameter=?
     Mask=8
     Masked Vector=&MakeVector(
      X=0
      Y=0
      Z=&Multiply_Float(
       A=&FloatFromCurve(FloatCurve=DataInterface() {propertyValues {Curve {
           defaultValue=1
           keys [
            {interpMode=RCIM_Cubic, tangentMode=RCTM_User, time=0, value=0}
            {interpMode=RCIM_Cubic, tangentMode=RCTM_User, time=118.803, value=-2083.5833}
            <at 500 500 RCIM_Cubic>
           ]
          }}})
       B=&FloatFromCurve(FloatCurve=DataInterface() {propertyValues {Curve {keys [<at 0.7 1>,<at 0.9004 -1>]}}}, Scale Curve=$Emitter.Age)
      )
     )
     Passthrough Vector=<0 0 0>
    ))
   &CurlNoiseForce(
    Bias Noise=false
    Noise Frequency=&MaskFloat(
     Mask Source=0
     Mask Parameter=?
     Mask=8
     Masked Float=25
     Passthrough Float=&MaskFloat(
      Mask Source=0
      Mask Parameter=?
      Mask=9
      Masked Float=&RandomRangeFloat(Minimum=20, Maximum=35)
      Passthrough Float=0
     )
    )
    Noise Strength=&MaskFloat(
     Mask Source=0
     Mask Parameter=?
     Mask=8
     Masked Float=&RandomRangeFloat(Minimum=200, Maximum=350)
     Passthrough Float=&MaskFloat(
      Mask Source=0
      Mask Parameter=?
      Mask=9
      Masked Float=&RandomRangeFloat(Minimum=100, Maximum=125)
      Passthrough Float=0
     )
    )
    Pan Noise=false
   )
   /SetVariables(Particles.Mesh Update Color:LinearColor=&MultiplyLinearColor(
     Linear Color=$Particles.Color
     Scale Factor=&MakeVector4_FromLinearColor(LinearColor=&MakeLinearColorFromVectorAndFloat(
       "Vector (RGB)"=&VectorFromFloat(Value=&Lerp_Float(
         A=1
         B=20
         Alpha=&FloatFromCurve(
          FloatCurve=DataInterface() {propertyValues {Curve {keys [{arriveTangent=-3.0621, interpMode=RCIM_Cubic, leaveTangent=-3.0621, tangentMode=RCTM_User, time=0, value=1}, <at 0.4 0 RCIM_Cubic>]}}}
          CurveIndex=$Emitter.Age
         )
        ))
       "Float (Alpha)"=1
      ))
    )) {isSetParameters=true}
   /SetVariables(Particles.Mesh Dynamic Param:Vector4f=&MakeVector4(X=0.2, Y=0, Z=0, W=0)) {isSetParameters=true}
   /SetVariables(
    Particles.Inner Sphere Glow Color:LinearColor=&MakeCustomLinearColorFromBool(
     Bool=&SetBoolByIntComparison(A=$Engine.QualityLevel, B=0, Comparison Type=::NewEnumerator4)
     True Color=&ColorFromCurve(ColorCurve=DataInterface() {propertyValues {
        AlphaCurve {keys [
          <at 0 0.5>
          {arriveTangent=1.2175, interpMode=RCIM_Cubic, leaveTangent=1.2175, time=0.0841, value=1}
          <at 0.4107 1>
          <at 0.5573 0.3>
          <at 0.7643 0.1>
          <at 1 0 RCIM_Cubic>
         ]}
        BlueCurve {keys [<at 0.066 0>,<at 0.2079 0>,<at 0.5124 0>,<at 0.9995 0>]}
        GreenCurve {keys [<at 0.066 5>,<at 0.2079 2.5>,<at 0.5124 2.5>,<at 0.9995 0.6948>]}
        RedCurve {keys [<at 0.066 0.2178>,<at 0.2079 0.1089>,<at 0.5124 0.1089>,<at 0.9995 0.0303>]}
       }})
     False Color=&ColorFromCurve(ColorCurve=DataInterface() {propertyValues {
        AlphaCurve {keys [
          <at 0 0.5>
          {arriveTangent=1.2175, interpMode=RCIM_Cubic, leaveTangent=1.2175, time=0.0841, value=1}
          <at 0.4107 1>
          <at 0.5573 0.3>
          <at 0.7643 0.1>
          <at 1 0 RCIM_Cubic>
         ]}
        BlueCurve {keys [
          <at 0.108 0>
          {arriveTangent=0.5878, interpMode=RCIM_Cubic, leaveTangent=0.5878, time=0.4596, value=2.5}
          <at 0.8683 0.4469>
          <at 1 0 RCIM_Cubic>
         ]}
        GreenCurve {keys [
          <at 0.108 5>
          {arriveTangent=-5.2614, interpMode=RCIM_Cubic, leaveTangent=-5.2614, time=0.4596, value=2.5}
          <at 0.8683 1>
          <at 1 0.224 RCIM_Cubic>
         ]}
        RedCurve {keys [
          <at 0.108 2.6303>
          {arriveTangent=-2.7449, interpMode=RCIM_Cubic, leaveTangent=-2.7449, time=0.4596, value=2.5}
          <at 0.8683 0.5435>
          <at 1 0.0391 RCIM_Cubic>
         ]}
       }})
    )
    Particles.Inner Sphere Glow Size:Vector3f=&Multiply_VectorByFloat(Vector=$Particles.Scale, Float=0.88)
   ) {isSetParameters=true}
   &Light_Attributes(
    Color=1
    Diffuse Scale=1
    Enabled=1
    Falloff Exponent=1
    Light Enabled:NiagaraBool=$Emitter.isFirstTransition
    Light Exponent=1.5
    Light Position Offset:Vector3f=<0 0 145>
    Light Radius:NiagaraFloat=&MakeCustomFloatFromBool(Bool=&Boolean_OrOperation(Bool1=$User.bIsFrontEnd, Bool2=$User.bIsFrontEndPreview), True Float=0, False Float=200)
    Position=1
    Radius=1
    Scale Alpha:NiagaraFloat=&FloatFromCurve(FloatCurve=DataInterface() {propertyValues {Curve {keys [
         {arriveTangent=0.0635, interpMode=RCIM_Cubic, leaveTangent=0.0635, tangentMode=RCTM_User, time=0.0246, value=0}
         {arriveTangent=0.7973, interpMode=RCIM_Cubic, leaveTangent=0.7973, time=0.3038, value=1}
         {arriveTangent=-1.4364, interpMode=RCIM_Cubic, leaveTangent=-1.4364, time=0.7264, value=0.5595}
         <at 1 0 RCIM_Cubic>
        ]}}})
    Scale RGB:Vector3f=&MakeVectorFromLinearColorRGB(Color=&ColorFromCurve(ColorCurve=DataInterface() {propertyValues {
        AlphaCurve {keys []}
        BlueCurve {keys [<at 0.1848 0.1186 RCIM_Cubic>, {arriveTangent=-0.2362, interpMode=RCIM_Cubic, time=0.4201, value=0}, <at 0.6454 0>, <at 0.9291 0>]}
        GreenCurve {keys [<at 0.1848 5 RCIM_Cubic>, {arriveTangent=-5.1433, interpMode=RCIM_Cubic, time=0.4201, value=4.8964}, <at 0.6454 2.4186>, <at 0.9291 0.9434>]}
        RedCurve {keys [<at 0.1848 1.6682 RCIM_Cubic>, {arriveTangent=8.3299, interpMode=RCIM_Cubic, time=0.4201, value=7}, <at 0.6454 5.8489>, <at 0.9291 5>]}
       }}))
    ScaleA=true
    ScaleRGB=true
    Specular Scale=1
    UsePositionOffset=true
   )
   &Light_Attributes(
    Color=1
    Diffuse Scale=1
    Enabled=1
    Falloff Exponent=1
    Light Position Offset:Vector3f=<0 0 50>
    Light Radius=150
    Position=1
    Radius=1
    Scale Alpha:NiagaraFloat=&FloatFromCurve(
     FloatCurve=DataInterface() {propertyValues {Curve {
        defaultValue=0
        keys [
         <at 0.25 1 RCIM_Cubic>
         {arriveTangent=-0.437, interpMode=RCIM_Cubic, leaveTangent=-0.437, tangentMode=RCTM_User, time=0.4434, value=0.9688}
         {arriveTangent=-3.2939, interpMode=RCIM_Cubic, leaveTangent=-3.2939, tangentMode=RCTM_User, time=0.8159, value=0.3913}
         {arriveTangent=0.2694, interpMode=RCIM_Cubic, leaveTangent=0.2694, tangentMode=RCTM_User, time=1, value=0}
        ]
       }}}
     Scale Curve=&MakeCustomFloatFromBool(Bool=&Boolean_OrOperation(Bool1=$User.bIsFrontEnd, Bool2=$User.bIsFrontEndPreview), True Float=0, False Float=1)
    )
    Scale RGB:Vector3f=&MakeVectorFromLinearColorRGB(Color=&ColorFromCurve(ColorCurve=DataInterface() {propertyValues {
        AlphaCurve {keys []}
        BlueCurve {keys [<at 0.0251 0 RCIM_Cubic>,<at 0.1435 0>,<at 0.3231 0>,<at 0.614 0 RCIM_Cubic>,<at 0.9947 0>]}
        GreenCurve {keys [
          <at 0.0251 20 RCIM_Cubic>
          <at 0.1435 5>
          <at 0.3231 5>
          {arriveTangent=-7.3119, interpMode=RCIM_Cubic, leaveTangent=-7.3119, time=0.614, value=2}
          <at 0.9947 0.0896>
         ]}
        RedCurve {keys [
          <at 0.0251 7.8779 RCIM_Cubic>
          <at 0.1435 2.8748>
          <at 0.3231 0.2878>
          {arriveTangent=-0.4205, interpMode=RCIM_Cubic, leaveTangent=-0.4205, time=0.614, value=0.1209}
          <at 0.9947 0.0054>
         ]}
       }}))
    ScaleA=true
    ScaleRGB=true
    Specular Scale=1
    UsePositionOffset=true
   ) {module=Light_Attributes001}
   /SetVariables(
    Particles.Streaks Color:LinearColor=&MultiplyLinearColor(
     Linear Color=&MultiplyLinearColor(
      Linear Color=&MultiplyLinearColor(
       Linear Color=<rgba 0 1 0.4471 1>
       Scale Factor=&MakeVector4_FromLinearColor(LinearColor=<rgba 0 1 0.1521 1>)
      )
      Scale Factor=&MakeVector4_FromLinearColor(LinearColor=<rgba 1 1 1 1>)
     )
     Scale Factor=&MakeFloat4FromVectorAndFloat(
      "Vector (RGB)"=&Multiply_Vector(
       A=&Multiply_VectorByFloat(
        Vector=&MakeVectorFromLinearColorRGB(Color=&ColorFromCurve(ColorCurve=DataInterface() {propertyValues {
            AlphaCurve {keys [<at 0 1 RCIM_Cubic>]}
            BlueCurve {keys [
              <at 0.045 19.1833>
              {arriveTangent=-55.4456, interpMode=RCIM_Cubic, leaveTangent=-55.4456, time=0.1045, value=10}
              {arriveTangent=-13.1418, interpMode=RCIM_Cubic, leaveTangent=-13.1418, time=0.3459, value=2.5}
              <at 0.4628 1>
              <at 0.8197 0.6>
             ]}
            GreenCurve {keys [
              <at 0.045 30>
              {arriveTangent=-91.3937, interpMode=RCIM_Cubic, leaveTangent=-91.3937, time=0.1045, value=10}
              {arriveTangent=-13.1418, interpMode=RCIM_Cubic, leaveTangent=-13.1418, time=0.3459, value=2.5}
              <at 0.4628 1>
              <at 0.8197 0.6>
             ]}
            RedCurve {keys [
              <at 0.045 28.0963>
              {arriveTangent=-85.0669, interpMode=RCIM_Cubic, leaveTangent=-85.0669, time=0.1045, value=10}
              {arriveTangent=-13.1418, interpMode=RCIM_Cubic, leaveTangent=-13.1418, time=0.3459, value=2.5}
              <at 0.4628 1>
              <at 0.8197 0.6>
             ]}
           }}))
        Float=1
       )
       B=&MakeVectorFromLinearColorRGB(Color=&FNTimeOfDayColor(Night Color=<rgba 1 1 1 1>, Morning Color=<rgba 0 0.5388 1 1>, Day Color=<rgba 1 1 1 1>, Evening Color=<rgba 0 0.5388 1 1>))
      )
      "Float (Alpha)"=&FloatFromCurve(FloatCurve=DataInterface() {propertyValues {Curve {keys [
           {arriveTangent=-2.2233, interpMode=RCIM_Cubic, leaveTangent=-2.2233, tangentMode=RCTM_User, time=0.1, value=1}
           {arriveTangent=-0.0075, interpMode=RCIM_Cubic, leaveTangent=-0.0075, tangentMode=RCTM_User, time=1, value=0}
          ]}}})
     )
    )
    Particles.Streaks Size:Vector2f=&Multiply_Vector2D(
     A=&Multiply_Vector2D_ByFloat(
      Float=&FloatFromCurve(FloatCurve=DataInterface() {propertyValues {Curve {keys [
           {arriveTangent=2.0201, interpMode=RCIM_Cubic, leaveTangent=2.0201, tangentMode=RCTM_User, time=0, value=0}
           {arriveTangent=-0.026, interpMode=RCIM_Cubic, leaveTangent=-0.026, tangentMode=RCTM_User, time=0.6, value=1}
          ]}}})
      Vector2D=&Multiply_Vector2D_ByFloat(Float=0.5, Vector2D=&RandomRangeVector2D(Minimum=<100 300>, Maximum=<200 400>))
     )
     B=&Vector2DFromCurve(
      Vector2Curve=DataInterface() {propertyValues {
        XCurve {keys [
          {arriveTangent=-1.6611, interpMode=RCIM_Cubic, leaveTangent=-1.6611, tangentMode=RCTM_User, time=0.2, value=1.2}
          {arriveTangent=-0.1324, interpMode=RCIM_Cubic, leaveTangent=-0.1324, tangentMode=RCTM_User, time=1, value=0.7}
         ]}
        YCurve {keys [{interpMode=RCIM_Cubic, tangentMode=RCTM_User, time=0.4, value=1}]}
       }}
      ScaleCurve=<1 1>
     )
    )
   ) {isSetParameters=true}
   &SolveForcesAndVelocity(Clamp Velocity=false, Limit Acceleration=false)
   &ScaleSpriteSizeBySpeed(
    Max Scale Factor=&MakeVector2D(
     X=&MaskFloat(
      Mask Source=0
      Mask Parameter=?
      Mask=7
      Masked Float=&FloatFromCurve(FloatCurve=DataInterface() {propertyValues {Curve {
          defaultValue=1
          keys [
           <at 0.1013 0.0114 RCIM_Cubic>
           {arriveTangent=1.2969, interpMode=RCIM_Cubic, leaveTangent=1.2969, time=0.1899, value=0.25}
           {arriveTangent=-0.483, interpMode=RCIM_Cubic, leaveTangent=-0.483, time=0.4468, value=0.4595}
           {arriveTangent=0.1291, interpMode=RCIM_Cubic, leaveTangent=0.1291, time=0.5803, value=0.0614}
           {arriveTangent=4.2571, interpMode=RCIM_Cubic, leaveTangent=4.2571, time=0.64, value=0.4845}
           {arriveTangent=1.432, interpMode=RCIM_Cubic, leaveTangent=1.432, time=0.7663, value=0.8531}
           {arriveTangent=-0.2466, interpMode=RCIM_Cubic, leaveTangent=-0.2466, tangentMode=RCTM_User, time=1, value=1}
          ]
         }}})
      Passthrough Float=&MaskFloat(
       Mask Source=0
       Mask Parameter=?
       Mask=8
       Masked Float=1
       Passthrough Float=&MaskFloat(Mask Source=0, Mask Parameter=?, Mask=9, Masked Float=1, Passthrough Float=1)
      )
     )
     Y=&MaskFloat(
      Mask Source=0
      Mask Parameter=?
      Mask=7
      Masked Float=&FloatFromCurve(FloatCurve=DataInterface() {propertyValues {Curve {
          defaultValue=1
          keys [
           <at 0.0346 2.2238 RCIM_Cubic>
           {arriveTangent=-3.0875, interpMode=RCIM_Cubic, leaveTangent=-3.0875, time=0.078, value=1.2351}
           {arriveTangent=1.0676, interpMode=RCIM_Cubic, leaveTangent=1.0676, time=0.3871, value=1.1353}
           {arriveTangent=-1.6321, interpMode=RCIM_Cubic, leaveTangent=-1.6321, time=0.5966, value=1.7888}
           {arriveTangent=-3.5833, interpMode=RCIM_Cubic, leaveTangent=-3.5833, time=0.6461, value=0.7127}
           <at 1 0.3435 RCIM_Cubic>
          ]
         }}})
      Passthrough Float=&MaskFloat(
       Mask Source=0
       Mask Parameter=?
       Mask=8
       Masked Float=5
       Passthrough Float=&MaskFloat(Mask Source=0, Mask Parameter=?, Mask=9, Masked Float=5, Passthrough Float=1)
      )
     )
    )
    Min Scale Factor=&MakeVector2D(
     X=&MaskFloat(
      Mask Source=0
      Mask Parameter=?
      Mask=7
      Masked Float=&FloatFromCurve(FloatCurve=DataInterface() {propertyValues {Curve {keys [<at 0 0.5>,<at 1 1>]}}})
      Passthrough Float=&MaskFloat(
       Mask Source=0
       Mask Parameter=?
       Mask=8
       Masked Float=0.7
       Passthrough Float=&MaskFloat(Mask Source=0, Mask Parameter=?, Mask=9, Masked Float=0.7, Passthrough Float=1)
      )
     )
     Y=&MaskFloat(
      Mask Source=0
      Mask Parameter=?
      Mask=7
      Masked Float=&FloatFromCurve(FloatCurve=DataInterface() {propertyValues {Curve {keys [<at 0 1>,<at 1 0.75>]}}})
      Passthrough Float=&MaskFloat(
       Mask Source=0
       Mask Parameter=?
       Mask=8
       Masked Float=1
       Passthrough Float=&MaskFloat(Mask Source=0, Mask Parameter=?, Mask=9, Masked Float=1, Passthrough Float=1)
      )
     )
    )
   )
   /SetVariables(
    Particles.Dots Outline Color:LinearColor=&MakeLinearColorFromVectorAndFloat(
     "Vector (RGB)"=&MakeVectorFromLinearColorRGB(Color=<rgba 0 0.3365 0.0261 1>)
     "Float (Alpha)"=&FloatFromCurve(
      FloatCurve=DataInterface() {propertyValues {Curve {
         defaultValue=1
         keys [
          {interpMode=RCIM_Cubic, tangentMode=RCTM_User, time=0, value=1}
          {arriveTangent=-1.8986, interpMode=RCIM_Cubic, leaveTangent=-1.8986, tangentMode=RCTM_User, time=1, value=0}
         ]
        }}}
      Scale Curve=&Waveform(
       "Waveform [1]"=::NewEnumerator0
       "[1] Frequency"=&RandomRangeFloat(Minimum=1.5, Maximum=3.5)
       "[1] Amplitude Min/Max"=<0.65 0.85>
      )
     )
    )
    Particles.Dots Outline Size:Vector2f=&Multiply_Vector2D_ByFloat(Float=1.1, Vector2D=$Particles.SpriteSize)
   ) {isSetParameters=true}
   /SetVariables(
    Particles.Mini Outline Color:LinearColor=&MakeLinearColorFromVectorAndFloat(
     "Vector (RGB)"=&MakeVectorFromLinearColorRGB(Color=<rgba 0 0.776 0.043 1>)
     "Float (Alpha)"=&FloatFromCurve(FloatCurve=DataInterface() {propertyValues {Curve {
         defaultValue=1
         keys [
          {interpMode=RCIM_Cubic, tangentMode=RCTM_User, time=0.07, value=1}
          {arriveTangent=-0.9689, interpMode=RCIM_Cubic, leaveTangent=-0.9689, tangentMode=RCTM_User, time=0.3651, value=0.3956}
          {interpMode=RCIM_Cubic, tangentMode=RCTM_User, time=1, value=0}
         ]
        }}})
    )
    Particles.Mini Outline Size:Vector2f=&Multiply_Vector2D_ByFloat(Float=1.75, Vector2D=$Particles.SpriteSize)
   ) {isSetParameters=true}
   /SetVariables(
    Particles.Mini Glow Color:LinearColor=&MakeLinearColorFromVectorAndFloat(
     "Vector (RGB)"=&MakeVectorFromLinearColorRGB(Color=<rgba 0 0.375 0.1536 1>)
     "Float (Alpha)"=&FloatFromCurve(
      FloatCurve=DataInterface() {propertyValues {Curve {
         defaultValue=1
         keys [
          {interpMode=RCIM_Cubic, tangentMode=RCTM_User, time=0.07, value=1}
          {arriveTangent=-0.9689, interpMode=RCIM_Cubic, leaveTangent=-0.9689, tangentMode=RCTM_User, time=0.4513, value=0.7133}
          {interpMode=RCIM_Cubic, tangentMode=RCTM_User, time=1, value=0}
         ]
        }}}
      Scale Curve=0.7
     )
    )
    Particles.Mini Glow Size:Vector2f=&Multiply_Vector2D_ByFloat(Float=3, Vector2D=$Particles.SpriteSize)
   ) {isSetParameters=true}
   /SetVariables(Particles.FlarePosition:NiagaraPosition=&ConvertVectorToPosition(
     Input Position=&Add_Vector(
      A=&ConvertPositionToVector(Input Position=$Particles.Position, Convert To Absolute Large World Vector=false)
      B=&MakeVector(
       X=0
       Y=&FloatFromCurve(
        FloatCurve=DataInterface() {propertyValues {Curve {keys [{tangentMode=RCTM_User, time=0, value=0}, {tangentMode=RCTM_User, time=1, value=1}]}}}
        Scale Curve=&Multiply_Float(A=25, B=-1)
       )
       Z=&FloatFromCurve(
        FloatCurve=DataInterface() {propertyValues {Curve {keys [{tangentMode=RCTM_User, time=0, value=0}, {tangentMode=RCTM_User, time=1, value=1}]}}}
        Scale Curve=-35
       )
      )
     )
     Vector Is In Absolute Large World Space:NiagaraBool=false
    )) {isSetParameters=true}
   /SetVariables(
    Particles.Linger Mini Glow Color:LinearColor=&MakeLinearColorFromVectorAndFloat(
     "Vector (RGB)"=&MakeVectorFromLinearColorRGB(Color=<rgba 0 0.375 0.1536 1>)
     "Float (Alpha)"=&FloatFromCurve(
      FloatCurve=DataInterface() {propertyValues {Curve {
         defaultValue=1
         keys [
          {interpMode=RCIM_Cubic, tangentMode=RCTM_User, time=0.07, value=1}
          {arriveTangent=-0.9689, interpMode=RCIM_Cubic, leaveTangent=-0.9689, tangentMode=RCTM_User, time=0.4513, value=0.7133}
          {interpMode=RCIM_Cubic, tangentMode=RCTM_User, time=1, value=0}
         ]
        }}}
      Scale Curve=0.7
     )
    )
    Particles.Linger Mini Glow Size:Vector2f=&Multiply_Vector2D_ByFloat(Float=3, Vector2D=$Particles.SpriteSize)
   ) {isSetParameters=true}
   /SetVariables(
    Particles.Linger Mini Outline Color:LinearColor=&MakeLinearColorFromVectorAndFloat(
     "Vector (RGB)"=&MakeVectorFromLinearColorRGB(Color=<rgba 0 0.1656 0.0092 1>)
     "Float (Alpha)"=&FloatFromCurve(
      FloatCurve=DataInterface() {propertyValues {Curve {
         defaultValue=1
         keys [
          {interpMode=RCIM_Cubic, tangentMode=RCTM_User, time=0.07, value=1}
          {arriveTangent=-0.9689, interpMode=RCIM_Cubic, leaveTangent=-0.9689, tangentMode=RCTM_User, time=0.3651, value=0.3956}
          {interpMode=RCIM_Cubic, tangentMode=RCTM_User, time=1, value=0}
         ]
        }}}
      Scale Curve=3
     )
    )
    Particles.Linger Mini Outline Size:Vector2f=&Multiply_Vector2D_ByFloat(Float=1.25, Vector2D=$Particles.SpriteSize)
   ) {isSetParameters=true}
   /SetVariables(Particles.LOW - Outline Color:LinearColor=&MakeLinearColorFromVectorAndFloat(
     "Vector (RGB)"=&MakeVectorFromLinearColorRGB(Color=&ColorFromCurve(ColorCurve=DataInterface() {propertyValues {
         AlphaCurve {keys []}
         BlueCurve {keys [<at 0.0042 0.1672>,<at 0.1048 0.064>,<at 0.2267 0>,<at 0.4209 0.0279>,<at 0.576 0>,<at 0.7314 0>,<at 0.9143 0>]}
         GreenCurve {keys [<at 0.0042 5>,<at 0.1048 5>,<at 0.2267 20>,<at 0.4209 4>,<at 0.576 1>,<at 0.7314 0.199>,<at 0.9143 0.0781>]}
         RedCurve {keys [<at 0.0042 0.4678>,<at 0.1048 0.4566>,<at 0.2267 5.1729>,<at 0.4209 0.5865>,<at 0.576 0.219>,<at 0.7314 0.0385>,<at 0.9143 0.0151>]}
        }}))
     "Float (Alpha)"=&FloatFromCurve(
      FloatCurve=DataInterface() {propertyValues {Curve {keys [
          {arriveTangent=0.6916, interpMode=RCIM_Cubic, leaveTangent=0.6916, tangentMode=RCTM_User, time=0.2, value=0.9812}
          {arriveTangent=-2.3514, interpMode=RCIM_Cubic, leaveTangent=-2.3514, tangentMode=RCTM_User, time=0.5699, value=0.2916}
          {arriveTangent=-0.2805, interpMode=RCIM_Cubic, leaveTangent=-0.2805, tangentMode=RCTM_User, time=0.8334, value=-0.0381}
         ]}}}
      Scale Curve=0.25
     )
    )) {isSetParameters=true}
   /SetVariables(
    Particles.Glow Adjust:Vector3f=&Multiply_VectorByFloat(
     Vector=$Particles.Inner Sphere Glow Size
     Float=&MakeCustomFloatFromBool(
      Bool=&Boolean_OrOperation(
       Bool1=&SetBoolByIntComparison(A=$Engine.QualityLevel, B=0, Comparison Type=::NewEnumerator4)
       Bool2=&SetBoolByIntComparison(A=$Engine.QualityLevel, B=1, Comparison Type=::NewEnumerator4)
      )
      True Float=1.05
      False Float=1
     )
    )
    Particles.LOW Outline Adjust:Vector3f=&Multiply_VectorByFloat(Vector=$Particles.Scale, Float=0.72)
    Particles.Sphere Adjust:Vector3f=&Multiply_VectorByFloat(
     Vector=$Particles.Scale
     Float=&MakeCustomFloatFromBool(
      Bool=&SetBoolByIntComparison(A=$Engine.QualityLevel, B=0, Comparison Type=::NewEnumerator4)
      True Float=0.8
      False Float=1
     )
    )
   ) {isSetParameters=true}
   /SetVariables(
    Particles.LOW - Energy Color:LinearColor=&ColorFromCurve(ColorCurve=DataInterface() {propertyValues {
       AlphaCurve {keys [<at 0.1235 2.5 RCIM_Cubic>,<at 1 0 RCIM_Cubic>]}
       BlueCurve {keys [<at 0.097 0>,<at 0.3524 0>]}
       GreenCurve {keys [<at 0.097 1>,<at 0.3524 0.2198>]}
       RedCurve {keys [<at 0.097 0.0934>,<at 0.3524 0.0002>]}
      }})
    Particles.LOW - Energy Orientation:Quat4f=&MakeQuaternion(Quaternion From=0, Axis=<1 0 0>, Angle=0.5, From=?, To=?, Angle Type=::NewEnumerator0, Normalize Vectors=false)
   ) {isSetParameters=true}
  ]
  rendererClasses [
   &NiagaraMeshRendererProperties
   &NiagaraSpriteRendererProperties
   &NiagaraLightRendererProperties
  ]
  renderers [
   {rendererClass=&NiagaraMeshRendererProperties, rendererIndex=0}
   {rendererClass=&NiagaraMeshRendererProperties, rendererIndex=1}
   {rendererClass=&NiagaraSpriteRendererProperties, rendererIndex=2}
   {rendererClass=&NiagaraLightRendererProperties, rendererIndex=3}
   {rendererClass=&NiagaraSpriteRendererProperties, rendererIndex=4}
   {rendererClass=&NiagaraSpriteRendererProperties, rendererIndex=5}
   {rendererClass=&NiagaraSpriteRendererProperties, rendererIndex=6}
   {rendererClass=&NiagaraSpriteRendererProperties, rendererIndex=7}
   {rendererClass=&NiagaraSpriteRendererProperties, rendererIndex=8}
   {rendererClass=&NiagaraSpriteRendererProperties, rendererIndex=9}
   {rendererClass=&NiagaraSpriteRendererProperties, rendererIndex=10}
   {rendererClass=&NiagaraSpriteRendererProperties, rendererIndex=11}
   {rendererClass=&NiagaraSpriteRendererProperties, rendererIndex=12}
   {rendererClass=&NiagaraSpriteRendererProperties, rendererIndex=13}
   {rendererClass=&NiagaraSpriteRendererProperties, rendererIndex=14}
   {rendererClass=&NiagaraLightRendererProperties, rendererIndex=15}
   {rendererClass=&NiagaraMeshRendererProperties, rendererIndex=16}
   {rendererClass=&NiagaraMeshRendererProperties, rendererIndex=17}
  ]
 }
 FE_Ground {rendererClasses [], renderers []}
 Glow_Decals {
  emitterUpdateScript [
   &EmitterState()
   &SpawnBurst_Instantaneous(Spawn Count=14, Use Spawn Probability=false, UseLoopCountLimit=false)
  ]
  particleSpawnScript [
   /SetVariables(Particles.BoneIndex:NiagaraInt32=&ReturnExecIndex()) {isSetParameters=true}
   &InitializeParticle(Lifetime=1.9, Sprite Size Mode=::NewEnumerator3, Uniform Sprite Size=70, UsePositionOffset=true, Write Material Random=true)
   &CameraOffset(Camera Offset Amount=30)
  ]
  particleUpdateScript [
   &ParticleState()
   &SocketLocation(
    Position Sampling=::NewEnumerator1
    SamplePosOffset=false
    Sampled Position Offset=&MaskVector(
     Mask Source=1
     Mask Parameter=$Particles.BoneIndex
     Mask=14
     Masked Vector=<-5 0 -8>
     Passthrough Vector=&MaskVector(Mask Source=1, Mask Parameter=$Particles.BoneIndex, Mask=15, Masked Vector=<-5 0 -8>, Passthrough Vector=<0 0 0>)
    )
    Socket Index=$Particles.BoneIndex
    Socket Reader=$System.Socket Reader
    Socket Sampling Mode=::NewEnumerator3
    VelocitySpeedLimit=false
   )
   &ScaleColor(
    Scale Alpha=&FloatFromCurve(
     FloatCurve=DataInterface() {propertyValues {Curve {keys [
         {arriveTangent=18.008, interpMode=RCIM_Cubic, leaveTangent=18.008, tangentMode=RCTM_User, time=0, value=0}
         {arriveTangent=0.0259, interpMode=RCIM_Cubic, leaveTangent=0.0259, tangentMode=RCTM_User, time=0.1, value=1}
         {arriveTangent=-1.6808, interpMode=RCIM_Cubic, leaveTangent=-1.6808, tangentMode=RCTM_User, time=0.4043, value=0.4939}
         {arriveTangent=-0.4088, interpMode=RCIM_Cubic, leaveTangent=-0.4088, tangentMode=RCTM_User, time=1, value=0}
        ]}}}
     Scale Curve=&MakeCustomFloatFromBool(Bool=$Particles.isHelmetFeather, True Float=0.2, False Float=0.1)
    )
    Scale RGB=&Multiply_VectorByFloat(
     Vector=&MakeVectorFromLinearColorRGB(Color=&ColorFromCurve(ColorCurve=DataInterface() {propertyValues {
         AlphaCurve {keys [<at 0 1 RCIM_Cubic>]}
         BlueCurve {keys [<at 0 1.2031>,<at 0.0242 0.1529>,<at 0.0418 0.0595>,<at 0.0526 0.0673>,<at 0.0661 0.4569>,<at 0.0891 0.1529>,<at 0.3708 0.0035>]}
         GreenCurve {keys [<at 0 3>,<at 0.0242 1>,<at 0.0418 0.451>,<at 0.0526 0.451>,<at 0.0661 3>,<at 0.0891 1>,<at 0.3708 0.0781>]}
         RedCurve {keys [<at 0 1.5385>,<at 0.0242 0.3095>,<at 0.0418 0>,<at 0.0526 0.1169>,<at 0.0661 0.9231>,<at 0.0891 0.3095>,<at 0.3708 0>]}
        }}))
     Float=1
    )
    ScaleA=true
    ScaleRGB=true
   ) {module=ScaleColor001}
   &DynamicMaterialParameters(
    Index 0 Param 1=&FloatFromCurve(FloatCurve=DataInterface() {propertyValues {Curve {keys [<at 0 -0.5 RCIM_Cubic>, {arriveTangent=1.5, interpMode=RCIM_Cubic, leaveTangent=1.5, time=0.3, value=0.5}, <at 1 1 RCIM_Cubic>]}}})
    Module.Index0Float0Write=true
    Module.Index0Float1Write=true
    Module.Index0Float2Write=true
    Module.Index0Float3Write=true
   )
   /SetVariables(Particles.Position=$Output.SocketLocation.SampledPosition) {isSetParameters=true}
  ]
  rendererClasses [&NiagaraSpriteRendererProperties]
  renderers [{rendererClass=&NiagaraSpriteRendererProperties}]
 }
 Lightning {
  emitterUpdateScript [
   &EmitterState(Life Cycle Mode=::NewEnumerator1, Loop Behavior=::NewEnumerator1, Loop Duration=1.2, UseLoopDelay=false)
   &SpawnRate(
    SpawnRate=&FloatFromCurve(
     FloatCurve=DataInterface() {propertyValues {Curve {keys [<at 0.037 1.1157>,<at 0.3476 0.5652>,<at 0.7477 0.3472>,<at 0.92 0>]}}}
     CurveIndex=$Emitter.NormalizedLoopAge
     Scale Curve=90
    )
    Use Spawn Probability=false
   )
  ]
  particleSpawnScript [
   /SetVariables(Particles.VisibilityTag=&RandomRangeInt(Minimum=0, Maximum=3)) {isSetParameters=true}
   &InitializeParticle(
    Color=<rgba 5 5 5 1>
    Lifetime=&RandomRangeFloat(Minimum=0.3, Maximum=0.5)
    Mesh Scale Mode=::NewEnumerator3
    Mesh Uniform Scale=&Multiply_Float(A=0.25, B=&RandomRangeFloat(Minimum=0.7, Maximum=1))
    UsePositionOffset=false
    Write Material Random=true
   )
   &InitialMeshOrientation(Use Rotation Vector=false) {module=InitialMeshOrientation001}
   &ShapeLocation(
    MaskSpawnGroup=false
    Non Uniform Scale=<0.3 0.3 0.8>
    Offset=<0 0 110>
    Offset Mode=::NewEnumerator0
    Override Seed=false
    Sphere Radius=82
    UseOwnerScale=true
   )
  ]
  particleUpdateScript [
   /SetVariables(Particles.Flicker=1) {isSetParameters=true}
   &ParticleState()
   &ScaleMeshSize(Scale Factor=&VectorFromCurve(
     VectorCurve=DataInterface() {propertyValues {
       XCurve {keys [{arriveTangent=0.4178, interpMode=RCIM_Cubic, leaveTangent=0.4178, tangentMode=RCTM_User, time=0, value=1}, <at 1 1.2 RCIM_Cubic>]}
       YCurve {keys [{arriveTangent=1.8574, interpMode=RCIM_Cubic, leaveTangent=1.8574, tangentMode=RCTM_User, time=0, value=1}, <at 1 2 RCIM_Cubic>]}
       ZCurve {keys [<at 0 1>,<at 1 1>]}
      }}
     Scale Curve=&VectorFromFloat(Value=1.6)
    )) {module=ScaleMeshSize001}
   &ScaleMeshSize(Scale Factor=&Multiply_VectorByFloat(Vector=&RandomRangeVector(Minimum=<1 1 4>, Maximum=<1.7 1.7 9>), Float=1))
   &DynamicMaterialParameters(
    Index 0 Param 2=&RandomRangeFloat(Minimum=0, Maximum=0)
    Index 0 Param 4=13
    Index 1 Param 1=&FloatFromCurve(FloatCurve=DataInterface() {propertyValues {Curve {keys [<at 0 0.5>,<at 1 1>]}}})
    Index 1 Param 2=2
    Index 1 Param 3=&FloatFromCurve(FloatCurve=DataInterface() {propertyValues {Curve {keys [
         {interpMode=RCIM_Cubic, tangentMode=RCTM_Break, time=0.5, value=-0.5}
         {arriveTangent=1.2, interpMode=RCIM_Cubic, leaveTangent=1.2, time=0.6931, value=-0.0671}
         <at 1 0.1>
        ]}}})
    Index 2 Param 1=&Multiply_Float(
     A=&FloatFromCurve(
      FloatCurve=DataInterface() {propertyValues {Curve {keys [
          {interpMode=RCIM_Cubic, leaveTangent=8.1085, tangentMode=RCTM_Break, time=0.7695, value=1}
          {arriveTangent=-4.3381, interpMode=RCIM_Cubic, leaveTangent=-4.3381, time=0.9048, value=0.8212}
          {tangentMode=RCTM_User, time=1, value=0}
         ]}}}
      Scale Curve=0.8
     )
     B=&Divide_Float(
      A=&MakeFloatFromVector(Vector=$Particles.Scale, Channel=::NewEnumerator0)
      B=&RandomRangeFloat(Minimum=10, Maximum=15)
     )
    )
    Index 2 Param 2=&FloatFromCurve(
     FloatCurve=DataInterface() {propertyValues {Curve {keys [
         {interpMode=RCIM_Cubic, tangentMode=RCTM_Break, time=0.7, value=0}
         {arriveTangent=-0.0859, interpMode=RCIM_Cubic, tangentMode=RCTM_Break, time=1, value=-1}
        ]}}}
     Scale Curve=50
    )
    Index 2 Param 3=&Waveform(
     Global Input Position=&RandomRangeFloat(Minimum=0, Maximum=15)
     Global Phase=$Particles.Age
     "Waveform [1]"=::NewEnumerator12
     "[1] Phase Scale"=2
     "[1] Frequency"=1
     "[1] Amplitude Min/Max"=<0 1>
    )
    Index 2 Param 4=25
    Index 3 Param 1=&FloatFromCurve(
     FloatCurve=DataInterface() {propertyValues {Curve {keys [
         {interpMode=RCIM_Cubic, leaveTangent=-0.0025, tangentMode=RCTM_Break, time=0, value=0}
         {arriveTangent=1.8528, interpMode=RCIM_Cubic, tangentMode=RCTM_Break, time=1, value=1}
        ]}}}
     Scale Curve=&RandomRangeFloat(Minimum=0, Maximum=0)
    )
    Index 3 Param 2=&FloatFromCurve(
     FloatCurve=DataInterface() {propertyValues {Curve {keys [
         {interpMode=RCIM_Cubic, leaveTangent=-0.0025, tangentMode=RCTM_Break, time=0, value=0.5}
         {arriveTangent=-0.0316, interpMode=RCIM_Cubic, tangentMode=RCTM_Break, time=1, value=-0.2}
        ]}}}
     Scale Curve=0
    )
    Index 3 Param 3=&FloatFromCurve(FloatCurve=DataInterface() {propertyValues {Curve {keys [
         {interpMode=RCIM_Cubic, leaveTangent=1.1034, tangentMode=RCTM_Break, time=0, value=-1}
         {arriveTangent=2.2399, tangentMode=RCTM_User, time=0.5, value=0.125}
        ]}}})
    Index 3 Param 4=&FloatFromCurve(FloatCurve=DataInterface() {propertyValues {Curve {keys [
         {interpMode=RCIM_Cubic, leaveTangent=1.8089, tangentMode=RCTM_Break, time=0.1837, value=-0.5}
         {arriveTangent=2.0941, interpMode=RCIM_Cubic, leaveTangent=2.0941, time=0.5787, value=0.4567}
         {arriveTangent=2.0757, interpMode=RCIM_Cubic, tangentMode=RCTM_Break, time=0.9, value=1}
        ]}}})
    Module.Index0Param0Write=true
    Module.Index0Param1Write=true
    Module.Index0Param2Write=true
    Module.Index0Param3Write=true
    Module.Index1Param0Write=true
    Module.Index1Param1Write=true
    Module.Index1Param2Write=true
    Module.Index1Param3Write=true
    Module.Index2Param0Write=true
    Module.Index2Param1Write=true
    Module.Index2Param2Write=true
    Module.Index2Param3Write=true
    Module.Index3Param0Write=true
    Module.Index3Param1Write=true
    Module.Index3Param2Write=true
    Module.Index3Param3Write=true
    Write Parameter Index 1=true
    Write Parameter Index 2=true
    Write Parameter Index 3=true
   ) {module=DynamicMaterialParameters001}
   &Drag(Drag=3, Use Linear Drag=true, Use Rotational Drag=true)
   &CurlNoiseForce(Bias Noise=false, Noise Strength=200, Pan Noise=false)
   &SolveForcesAndVelocity(Clamp Velocity=false, Limit Acceleration=false)
   &ScaleColor(
    Scale Alpha=&Multiply_Float(
     A=&Lerp_Float(A=0.3, B=1, Alpha=$Particles.Flicker)
     B=&FloatFromCurve(
      FloatCurve=DataInterface() {propertyValues {Curve {keys [
          <at 0.3435 1 RCIM_Cubic>
          {arriveTangent=-1.5621, interpMode=RCIM_Cubic, leaveTangent=-1.5621, time=0.5723, value=0.179}
          <at 0.9837 0 RCIM_Cubic>
         ]}}}
      CurveIndex=$Emitter.NormalizedLoopAge
     )
    )
    Scale RGB=&VectorFromFloat(Value=&Multiply_Float(
      A=&Lerp_Float(A=1, B=3, Alpha=$Particles.Flicker)
      B=&FloatFromCurve(
       FloatCurve=DataInterface() {propertyValues {Curve {keys [
           {interpMode=RCIM_Cubic, tangentMode=RCTM_User, time=0.0013, value=0.6894}
           {arriveTangent=-1.769, interpMode=RCIM_Cubic, leaveTangent=-1.769, tangentMode=RCTM_User, time=0.5113, value=0.293}
           {interpMode=RCIM_Cubic, tangentMode=RCTM_User, time=1, value=0}
          ]}}}
       Scale Curve=0.1
      )
     ))
    ScaleA=true
    ScaleRGB=true
   ) {module=ScaleColor001}
  ]
  rendererClasses [&NiagaraMeshRendererProperties]
  renderers [
   {rendererClass=&NiagaraMeshRendererProperties, rendererIndex=0}
   {rendererClass=&NiagaraMeshRendererProperties, rendererIndex=1}
   {rendererClass=&NiagaraMeshRendererProperties, rendererIndex=2}
   {rendererClass=&NiagaraMeshRendererProperties, rendererIndex=3}
  ]
 }
}
summary {
 systemName=NS_Emote_BiomeHose_TransformGlow
 emitters [
  {
   emitterName=EnergyExplosion
   rendererClasses [&NiagaraMeshRendererProperties, &NiagaraSpriteRendererProperties]
  }
  {emitterName=Glow_Decals, rendererClasses [&NiagaraSpriteRendererProperties]}
  {emitterName=Lightning, rendererClasses [&NiagaraMeshRendererProperties]}
  {
   emitterName=Explosion
   rendererClasses [
    &NiagaraMeshRendererProperties
    &NiagaraSpriteRendererProperties
    &NiagaraLightRendererProperties
   ]
  }
  {emitterName=FE_Ground, rendererClasses []}
 ]
 userVariables [
  User.bIsFrontEnd {type=NiagaraBool}
  User.bIsFrontEndPreview {type=NiagaraBool}
  User.BurstIndex {type=NiagaraFloat, defaultValue=0, description=""}
 ]
}
