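# Niagara effect for the watch in the BiomeHose emote: a green, hologram-like light projection showing alien sprites.
# Aliens draws a sprite with its SubUV frame driven by User.FX_Lerp, fading out when the camera vector nears vertical.
# Light layers timed meshes, sprites and a light through spawn-group switches, ending in short flashes.
# User.FX_Control scales key opacities and beam size; front-end flags zero the light radius.
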NiagaraSystem /BRCosmetics/Animation/Game/MainPlayer/Emotes/BiomeHose/VFX/NS_Emote_BiomeHose_Watch
dependencies {
 usedDataInterfaces [
  &NiagaraDataInterfaceSpriteRendererInfo
  &NiagaraDataInterfaceCurve
  &NiagaraDataInterfaceCamera
  &NiagaraDataInterfaceVectorCurve
  &NiagaraDataInterfaceColorCurve
 ]
 usedDynamicInputs []
 usedModules [&SystemState]
}
emitters {
 Aliens {
  emitterUpdateScript [
   &EmitterState()
   &SpawnBurst_Instantaneous(Spawn Time=0.75, Use Spawn Probability=false, UseLoopCountLimit=false)
   /SetVariables(Emitter.Flicker Speed=9, Emitter.Noise Distortion Amount=5) {isSetParameters=true}
  ]
  particleSpawnScript [
   &InitializeParticle(
    Color=<rgba 3.449 5 3.3177 1>
    Lifetime=2
    Position Offset=<0 0 16>
    Sprite Size Mode=::NewEnumerator3
    Uniform Sprite Size=18
    UsePositionOffset=true
    Write Material Random=true
   )
   &CameraOffset(Camera Offset Amount=2)
  ]
  particleUpdateScript [
   &ParticleState()
   &SolveForcesAndVelocity(Clamp Velocity=false, Limit Acceleration=false)
   &SubUVAnimation(
    Direct Frame Index:NiagaraFloat=$User.FX_Lerp
    "Sprite Renderer "=DataInterface() {propertyValues {SpriteRenderer=&self:NiagaraEmitter_1.NiagaraSpriteRendererProperties_1}}
    SubUV Animation Mode:ENiagara_SubUVLookupModeV2=::NewEnumerator5
   )
   &ScaleColor(
    Scale Alpha=&Lerp_Float(
     A=&FloatFromCurve(
      FloatCurve=DataInterface() {propertyValues {Curve {defaultValue=1, keys [{tangentMode=RCTM_User, time=0, value=0}, {tangentMode=RCTM_User, time=0.05, value=1}]}}}
      CurveIndex=$Emitter.Age
      Scale Curve=$User.FX_Control
     )
     B=0
     Alpha=&ClampFloat(Float=&RemapRange(
       Input Value To Remap=&DotProduct(Vector A=&ReturnCameraVector(Camera Vectors=::NewEnumerator1, Coordinate Space=::World), Vector B=<0 0 -1>)
       Input Min=0.9
       Input Max=1
       Output Min=0
       Output Max=1
      ))
    )
    ScaleA=true
    ScaleRGB=true
   )
   /SetVariables(Particles.Alien Alignment:Vector3f=<0 0 1>) {isSetParameters=true}
  ]
  rendererClasses [&NiagaraSpriteRendererProperties]
  renderers [{rendererClass=&NiagaraSpriteRendererProperties}]
 }
 Light {
  emitterUpdateScript [
   &EmitterState()
   /SetVariables(Emitter.Flicker Speed=9, Emitter.Noise Distortion Amount=5) {isSetParameters=true}
   &SpawnBurst_Instantaneous(Spawn Time=0.45, Use Spawn Probability=false, UseLoopCountLimit=false)
   &SpawnBurst_Instantaneous(Spawn Count=2, Spawn Time=0.45, Use Spawn Probability=false, UseLoopCountLimit=false) {module=SpawnBurst_Instantaneous001}
   &SpawnRate(
    SpawnGroup=1
    SpawnRate=&Multiply_Float(
     A=7
     B=&FloatFromCurve(
      FloatCurve=DataInterface() {propertyValues {Curve {
         defaultValue=1
         keys [{tangentMode=RCTM_User, time=0.4225, value=0}, {tangentMode=RCTM_User, time=0.4247, value=1}, <at 2.581 1>, <at 2.5925 0>]
        }}}
      CurveIndex=$Emitter.Age
     )
    )
    Use Spawn Probability=false
   )
   &SpawnBurst_Instantaneous(Spawn Group=2, Spawn Time=0.45, Use Spawn Probability=false, UseLoopCountLimit=false) {module=SpawnBurst_Instantaneous002}
   &SpawnBurst_Instantaneous(Spawn Group=3, Spawn Time=0.45, Use Spawn Probability=false, UseLoopCountLimit=false) {module=SpawnBurst_Instantaneous003}
   &SpawnBurst_Instantaneous(Spawn Group=4, Spawn Time=2.8, Use Spawn Probability=false, UseLoopCountLimit=false) {module=SpawnBurst_Instantaneous004}
   &SpawnBurst_Instantaneous(Spawn Group=5, Spawn Time=2.8, Use Spawn Probability=false, UseLoopCountLimit=false) {module=SpawnBurst_Instantaneous005}
   &SpawnBurst_Instantaneous(Spawn Group=6, Spawn Time=0.45, Use Spawn Probability=false, UseLoopCountLimit=false) {module=SpawnBurst_Instantaneous006}
   &SpawnBurst_Instantaneous(Spawn Count=3, Spawn Group=7, Spawn Time=0.85, Use Spawn Probability=false, UseLoopCountLimit=false) {module=SpawnBurst_Instantaneous007}
   &SpawnBurst_Instantaneous(Spawn Count=5, Spawn Group=7, Spawn Time=0.99, Use Spawn Probability=false, UseLoopCountLimit=false) {module=SpawnBurst_Instantaneous008}
   &SpawnBurst_Instantaneous(Spawn Count=5, Spawn Group=7, Spawn Time=1.28, Use Spawn Probability=false, UseLoopCountLimit=false) {module=SpawnBurst_Instantaneous009}
   &SpawnBurst_Instantaneous(Spawn Count=5, Spawn Group=7, Spawn Time=1.46, Use Spawn Probability=false, UseLoopCountLimit=false) {module=SpawnBurst_Instantaneous010}
   &SpawnBurst_Instantaneous(Spawn Group=8, Spawn Time=2.79, Use Spawn Probability=false, UseLoopCountLimit=false) {module=SpawnBurst_Instantaneous011}
   &SpawnBurst_Instantaneous(Spawn Group=9, Spawn Time=2.79, Use Spawn Probability=false, UseLoopCountLimit=false) {module=SpawnBurst_Instantaneous012}
  ]
  particleSpawnScript [
   /SetVariables(Particles.VisibilityTag=&SpawnGroup()) {isSetParameters=true}
   &InitializeParticle(
    Color=&MergeHelper_ColorSwitch(
     0=<rgba 1.8621 3 0.4281 3>
     1=<rgba 1 1 1 1>
     2=<rgba 0.5139 1.5396 0 0.1861>
     3=<rgba 0.0833 0.9843 0 0.6661>
     4=<rgba 0.8084 2 0 0.3043>
     5=<rgba 0.0847 1 0 0.5409>
     6=<rgba 1 1 1 1>
     7=<rgba 1 1 1 1>
     8=<rgba 1 1 1 1>
     9=<rgba 0.4681 1 0.2797 1>
     Index=&SpawnGroup()
     Num=::NewEnumerator8
    )
    Lifetime=&MergeHelper_FloatSwitch(
     0=3.8
     1=&RandomRangeFloat(Minimum=1, Maximum=1.5)
     2=3
     3=3
     4=0.11
     5=0.11
     6=4
     7=&RandomRangeFloat(Minimum=0.1, Maximum=0.15)
     8=0.2
     9=0.1
     Index=&SpawnGroup()
     Num=::NewEnumerator8
    )
    Mesh Scale=&MergeHelper_VectorSwitch(
     0=<0.17 0.17 0.7>
     1=<0 0 0>
     2=<0.55 0.4 0.55>
     3=<0.55 0.4 0.55>
     4=<0.55 0.55 0.55>
     5=<0.55 0.55 0.55>
     6=<0 0 0>
     7=<0 0 0>
     Index=&SpawnGroup()
     Num=::NewEnumerator6
    )
    Mesh Scale Mode=::NewEnumerator4
    Position Offset=&MergeHelper_VectorSwitch(
     0=<0 -1 4>
     1=<0 0 6>
     2=<0 -1 5>
     3=<0 -1 5>
     4=<0 -1 5>
     5=<0 -1 5>
     6=<0 0 0>
     7=<0 0 0>
     8=<0 0 5>
     9=<0 0 1>
     Index=&SpawnGroup()
     Num=::NewEnumerator8
    )
    Sprite Size=&MergeHelper_Vector2Switch(
     0=<0 0>
     1=&RandomRangeVector2D(Minimum=<40 50>, Maximum=<60 60>)
     2=<0 0>
     3=<0 0>
     4=<0 0>
     5=<0 0>
     6=<0 0>
     7=&Vector2DFromFloat(Value=50)
     8=&Vector2DFromFloat(Value=45)
     9=&Vector2DFromFloat(Value=300)
     Index=&SpawnGroup()
     Num=::NewEnumerator8
    )
    Sprite Size Mode=::NewEnumerator4
    UsePositionOffset=true
    Write Material Random=true
   )
   &InitialMeshOrientation(
    Mesh Axis To Orient=&MergeHelper_VectorSwitch(
     0=<1 0 0>
     1=<0 0 0>
     2=<0 0 2>
     3=<0 0 2>
     4=<0 0 2>
     5=<0 0 2>
     6=<0 0 0>
     7=<1 0 0>
     Index=&SpawnGroup()
     Num=::NewEnumerator6
    )
    Mesh Orientation Mode=::NewEnumerator2
    Orientation Vector=&MergeHelper_VectorSwitch(
     0=<1 0 0>
     1=<0 0 0>
     2=<0 1 0>
     3=<0 1 0>
     4=<0 1 0>
     5=<0 1 0>
     6=<0 0 0>
     7=<1 0 0>
     Index=&SpawnGroup()
     Num=::NewEnumerator6
    )
    Use Rotation Vector=false
   )
   /SetVariables(Particles.Long Glitch Size:Vector3f=&RandomRangeVector(Minimum=<0.25 2.5 0.25>, Maximum=<0.5 5 1>)) {isSetParameters=true}
   &SubUVAnimation(
    Output Alternate Renderer Binding=true
    "Sprite Renderer "=DataInterface() {propertyValues {SpriteRenderer=&self:Light_0.NiagaraSpriteRendererProperties_0}}
    SubUV Animation Mode:ENiagara_SubUVLookupModeV2=::NewEnumerator2
    UseEndFrame=false
    UseStartFrame=false
   )
   &CameraOffset(Camera Offset Amount=&MergeHelper_FloatSwitch(0=0, 1=6, 2=0, 3=0, 4=0, 5=0, 6=0, 7=0, 8=15, 9=15, Index=&SpawnGroup(), Num=::NewEnumerator8))
   /SetVariables(Particles.Surface Glow Dynamic Param:Vector4f=&MakeVector4(X=1, Y=0.7, Z=0, W=0)) {isSetParameters=true}
   &ShapeLocation(
    MaskSpawnGroup=false
    Non Uniform Scale=&MaskVector(Mask Source=0, Mask Parameter=?, Mask=7, Masked Vector=<0.65 0.65 1.2>, Passthrough Vector=<0 0 0>)
    Offset=&MaskVector(Mask Source=0, Mask Parameter=?, Mask=7, Masked Vector=<-1.5 0 15>, Passthrough Vector=<0 0 0>)
    Offset Mode=::NewEnumerator0
    Override Seed=false
    Sphere Radius=&MaskFloat(Mask Source=0, Mask Parameter=?, Mask=7, Masked Float=7, Passthrough Float=0)
    UseOwnerScale=true
   )
   /SetVariables(Particles.Linear Color=<rgba 50 50 50 1>) {isSetParameters=true}
  ]
  particleUpdateScript [
   &ParticleState()
   /SetVariables(
    Particles.Long Glitch Color Update:LinearColor=&RandomRangeLinearColor(Minimum=<rgba 0.8407 1 0.4066 1>, Maximum=<rgba 0 1 0.1992 1>)
    Particles.Long Glitch Dynam 1:Vector4f=&MakeVector4(
     X=0
     Y=&RandomRangeFloat(Minimum=-1, Maximum=1)
     Z=&FloatFromCurve(FloatCurve=DataInterface() {propertyValues {Curve {keys [
          <at 0 2 RCIM_Cubic>
          {arriveTangent=-6.2285, interpMode=RCIM_Cubic, leaveTangent=-6.2285, tangentMode=RCTM_User, time=0.5612, value=1.0841}
          {arriveTangent=-0.8021, interpMode=RCIM_Cubic, leaveTangent=-0.8021, tangentMode=RCTM_User, time=1, value=0}
         ]}}})
     W=0
    )
    Particles.Long Glitch Dynam 2:Vector4f=&MakeVector4(
     X=0
     Y=0
     Z=0
     W=&FloatFromCurve(FloatCurve=DataInterface() {propertyValues {Curve {keys [
          {arriveTangent=4.1785, interpMode=RCIM_Cubic, leaveTangent=4.1785, tangentMode=RCTM_User, time=0, value=0}
          {arriveTangent=0.9995, interpMode=RCIM_Cubic, leaveTangent=0.9995, tangentMode=RCTM_User, time=0.2223, value=0.6825}
          {arriveTangent=-0.0405, interpMode=RCIM_Cubic, leaveTangent=-0.0405, tangentMode=RCTM_User, time=1, value=1}
         ]}}})
    )
    Particles.Long Glitch Scale:Vector3f=&Multiply_Vector(
     A=$Particles.Initial.Long Glitch Size
     B=&Multiply_VectorByFloat(
      Vector=<0.5 0.5 0.2>
      Float=&Multiply_Float(
       A=0.08
       B=&FloatFromCurve(
        FloatCurve=DataInterface() {propertyValues {Curve {keys [
            <at 0.0024 4.7657 RCIM_Cubic>
            {arriveTangent=14.525, interpMode=RCIM_Cubic, leaveTangent=14.525, tangentMode=RCTM_Break, time=0.0625, value=7.4517}
            {arriveTangent=-14.3776, interpMode=RCIM_Cubic, leaveTangent=-14.3776, tangentMode=RCTM_Break, time=0.15, value=6.2624}
            {arriveTangent=0.1532, interpMode=RCIM_Cubic, leaveTangent=0.1532, tangentMode=RCTM_Break, time=0.5, value=2.875}
            {interpMode=RCIM_Cubic, tangentMode=RCTM_Break, time=0.8185, value=2.3106}
            {arriveTangent=-0.0405, interpMode=RCIM_Cubic, leaveTangent=-0.0405, tangentMode=RCTM_Break, time=0.9185, value=0.8106}
           ]}}}
        Scale Curve=&Multiply_Float(
         A=0.5
         B=&ClampFloat(
          Float=&FN_LOD_Distance_Remap(Start Distance=200, End Distance=400, Minimum Output=0.5, Maximum Output=1)
          Min=0.5
         )
        )
       )
      )
     )
    )
   ) {isSetParameters=true}
   &ScaleSpriteSize(
    Scale Sprite Size Mode=::NewEnumerator0
    Uniform Scale Factor=&MergeHelper_FloatSwitch(
     0=0
     1=0.65
     2=1
     3=1
     4=1
     5=0
     6=1
     7=0
     8=&FloatFromCurve(FloatCurve=DataInterface() {propertyValues {Curve {keys [
          {arriveTangent=7.2928, interpMode=RCIM_Cubic, leaveTangent=7.2928, tangentMode=RCTM_User, time=0, value=0.5}
          {arriveTangent=4.5, interpMode=RCIM_Cubic, leaveTangent=4.5, time=0.5096, value=3.5141}
          <at 1 5 RCIM_Cubic>
         ]}}})
     9=&FloatFromCurve(FloatCurve=DataInterface() {propertyValues {Curve {
         defaultValue=0
         keys [
          {interpMode=RCIM_Cubic, tangentMode=RCTM_User, time=0, value=1}
          {arriveTangent=-0.7492, interpMode=RCIM_Cubic, leaveTangent=-0.7492, time=0.4743, value=0.5772}
          {arriveTangent=0.1433, interpMode=RCIM_Cubic, leaveTangent=0.1433, tangentMode=RCTM_User, time=0.7756, value=0.4189}
          {arriveTangent=2.8825, interpMode=RCIM_Cubic, leaveTangent=56.4021, tangentMode=RCTM_Break, time=0.889, value=0.5723}
          {interpMode=RCIM_Cubic, tangentMode=RCTM_User, time=1, value=0}
         ]
        }}})
     Index=&SpawnGroup()
     Num=::NewEnumerator8
    )
   )
   &ScaleMeshSize(Scale Factor=&MergeHelper_VectorSwitch(
     0=&MakeVector(
      X=1
      Y=1
      Z=&FloatFromCurve(FloatCurve=DataInterface() {propertyValues {Curve {keys [<at 0.5699 1>,<at 0.9985 0.2353>]}}})
     )
     1=<0 0 0>
     2=&VectorFromCurve(
      VectorCurve=DataInterface() {propertyValues {
        XCurve {keys [<at 0 1>]}
        YCurve {keys [
          {arriveTangent=0.3778, interpMode=RCIM_Cubic, leaveTangent=0.3778, tangentMode=RCTM_User, time=0, value=0.5}
          {arriveTangent=-5.8915, interpMode=RCIM_Cubic, leaveTangent=-5.8915, tangentMode=RCTM_User, time=0.0245, value=1.169}
          {
           arriveTangent=-30.1978
           arriveTangentWeight=0.0172
           interpMode=RCIM_Cubic
           leaveTangent=-30.1978
           leaveTangentWeight=0.3242
           tangentMode=RCTM_User
           time=0.0282
           value=1.0884
          }
          {interpMode=RCIM_Cubic, leaveTangent=-4, time=0.0405, value=1}
         ]}
        ZCurve {keys [<at 0 1>]}
       }}
      Scale Curve=&MakeVector(X=1, Y=&Multiply_Float(A=$User.FX_Control, B=1.2), Z=1)
     )
     3=&VectorFromCurve(
      VectorCurve=DataInterface() {propertyValues {
        XCurve {keys [<at 0 1>]}
        YCurve {keys [
          {arriveTangent=0.3778, interpMode=RCIM_Cubic, leaveTangent=0.3778, tangentMode=RCTM_User, time=0, value=0.5}
          {arriveTangent=-5.8915, interpMode=RCIM_Cubic, leaveTangent=-5.8915, tangentMode=RCTM_User, time=0.0245, value=1.169}
          {
           arriveTangent=-30.1978
           arriveTangentWeight=0.0172
           interpMode=RCIM_Cubic
           leaveTangent=-30.1978
           leaveTangentWeight=0.3242
           tangentMode=RCTM_User
           time=0.0282
           value=1.0884
          }
          {interpMode=RCIM_Cubic, leaveTangent=-4, time=0.0405, value=1}
         ]}
        ZCurve {keys [<at 0 1>]}
       }}
      Scale Curve=&MakeVector(X=1, Y=&Multiply_Float(A=$User.FX_Control, B=1.2), Z=1)
     )
     4=&VectorFromCurve(VectorCurve=DataInterface() {propertyValues {
        XCurve {keys [<at 0 1>]}
        YCurve {keys [
          {arriveTangent=-8.1015, interpMode=RCIM_Cubic, leaveTangent=-8.1015, tangentMode=RCTM_User, time=0, value=0.5}
          {arriveTangent=-186.4003, interpMode=RCIM_Cubic, leaveTangent=-1.1522, tangentMode=RCTM_Break, time=0.0104, value=1.034}
          {interpMode=RCIM_Cubic, leaveTangent=-4, time=0.0405, value=1}
         ]}
        ZCurve {keys [<at 0 1>]}
       }})
     5=&VectorFromCurve(VectorCurve=DataInterface() {propertyValues {
        XCurve {keys [<at 0 1>]}
        YCurve {keys [
          {arriveTangent=-8.1015, interpMode=RCIM_Cubic, leaveTangent=-8.1015, tangentMode=RCTM_User, time=0, value=0.5}
          {arriveTangent=-186.4003, interpMode=RCIM_Cubic, leaveTangent=-1.1522, tangentMode=RCTM_Break, time=0.0104, value=1.034}
          {interpMode=RCIM_Cubic, leaveTangent=-4, time=0.0405, value=1}
         ]}
        ZCurve {keys [<at 0 1>]}
       }})
     6=<0 0 0>
     7=<0 0 0>
     Index=&SpawnGroup()
     Num=::NewEnumerator6
    ))
   &ScaleColor(
    Scale Alpha=&MergeHelper_FloatSwitch(
     0=1
     1=&FloatFromCurve(
      FloatCurve=DataInterface() {propertyValues {Curve {keys [{interpMode=RCIM_Cubic, tangentMode=RCTM_User, time=0, value=0}, {interpMode=RCIM_Cubic, tangentMode=RCTM_User, time=0.3, value=1}, <at 1 0>]}}}
      Scale Curve=&Multiply_Float(A=$User.FX_Control, B=0.6)
     )
     2=&FloatFromCurve(
      FloatCurve=DataInterface() {propertyValues {Curve {
         defaultValue=1
         keys [
          {interpMode=RCIM_Cubic, tangentMode=RCTM_User, time=0, value=0}
          {arriveTangentWeight=0.0226, interpMode=RCIM_Cubic, leaveTangentWeight=0.2337, tangentMode=RCTM_User, time=0.0679, value=1}
          {arriveTangentWeight=0.3427, interpMode=RCIM_Cubic, leaveTangentWeight=0.1129, tangentMode=RCTM_User, time=0.769, value=1}
          {tangentMode=RCTM_User, time=1, value=0}
         ]
        }}}
      Scale Curve=&Multiply_Float(A=$User.FX_Control, B=0.3)
     )
     3=&FloatFromCurve(
      FloatCurve=DataInterface() {propertyValues {Curve {
         defaultValue=1
         keys [
          {interpMode=RCIM_Cubic, tangentMode=RCTM_User, time=0, value=0}
          {arriveTangentWeight=0.0226, interpMode=RCIM_Cubic, leaveTangentWeight=0.2337, tangentMode=RCTM_User, time=0.0679, value=1}
          {arriveTangentWeight=0.3427, interpMode=RCIM_Cubic, leaveTangentWeight=0.1129, tangentMode=RCTM_User, time=0.769, value=1}
          {tangentMode=RCTM_User, time=1, value=0}
         ]
        }}}
      Scale Curve=&Multiply_Float(A=$User.FX_Control, B=0.3)
     )
     4=&FloatFromCurve(
      FloatCurve=DataInterface() {propertyValues {Curve {
         defaultValue=1
         keys [
          {interpMode=RCIM_Cubic, tangentMode=RCTM_User, time=0, value=0}
          {arriveTangentWeight=0.0226, interpMode=RCIM_Cubic, leaveTangentWeight=0.2337, tangentMode=RCTM_User, time=0.0679, value=1}
          {
           arriveTangent=-2.8987
           arriveTangentWeight=0.3427
           interpMode=RCIM_Cubic
           leaveTangent=-2.8987
           leaveTangentWeight=0.1129
           tangentMode=RCTM_User
           time=0.4048
           value=0.7433
          }
          {tangentMode=RCTM_User, time=1, value=0}
         ]
        }}}
      Scale Curve=0.35
     )
     5=&FloatFromCurve(
      FloatCurve=DataInterface() {propertyValues {Curve {
         defaultValue=1
         keys [
          {interpMode=RCIM_Cubic, tangentMode=RCTM_User, time=0, value=0}
          {arriveTangentWeight=0.0226, interpMode=RCIM_Cubic, leaveTangentWeight=0.2337, tangentMode=RCTM_User, time=0.0679, value=1}
          {
           arriveTangent=-2.8987
           arriveTangentWeight=0.3427
           interpMode=RCIM_Cubic
           leaveTangent=-2.8987
           leaveTangentWeight=0.1129
           tangentMode=RCTM_User
           time=0.4048
           value=0.7433
          }
          {tangentMode=RCTM_User, time=1, value=0}
         ]
        }}}
      Scale Curve=0.35
     )
     6=1
     7=&FloatFromCurve(FloatCurve=DataInterface() {propertyValues {Curve {keys [<at -0.0046 1.0113 RCIM_Cubic>,<at 1.0011 0.002 RCIM_Cubic>]}}})
     8=&FloatFromCurve(
      FloatCurve=DataInterface() {propertyValues {Curve {keys [
          <at -0.0046 1.0113 RCIM_Cubic>
          {arriveTangent=-1.0349, interpMode=RCIM_Cubic, leaveTangent=-1.0349, time=0.435, value=0.834}
          {arriveTangent=-1.2597, interpMode=RCIM_Cubic, leaveTangent=-1.2597, tangentMode=RCTM_User, time=0.6814, value=0.3013}
          <at 1.0011 0.002 RCIM_Cubic>
         ]}}}
      Scale Curve=0.25
     )
     9=&FloatFromCurve(FloatCurve=DataInterface() {propertyValues {Curve {keys [<at -0.0046 1.0113 RCIM_Cubic>,<at 1.0011 0.002 RCIM_Cubic>]}}})
     Index=&SpawnGroup()
     Num=::NewEnumerator8
    )
    Scale RGB=&MergeHelper_VectorSwitch(
     0=&VectorFromFloat(Value=&Lerp_Float(
       A=1
       B=&Waveform(Global Amplitude Scale=0.5, "Waveform [1]"=::NewEnumerator8, "[1] Frequency"=15, "[1] Amplitude Min/Max"=<0.5 1>, "[1] Bias"=0.7)
       Alpha=0.4
      ))
     1=&MakeVectorFromLinearColorRGB(Color=&ColorFromCurve(
       ColorCurve=DataInterface() {propertyValues {
         AlphaCurve {keys [<at 0 1 RCIM_Cubic>,<at 1 1 RCIM_Cubic>]}
         BlueCurve {keys [
           <at 0.0292 0.041>
           {arriveTangent=-0.0463, interpMode=RCIM_Cubic, leaveTangent=-0.0463, time=0.1777, value=0}
           <at 0.2827 0.0293>
           <at 0.3698 0>
           <at 0.4927 0>
           <at 0.5809 0.0153>
           <at 0.8914 0>
           <at 0.9461 0.1741>
           <at 1 0.4795>
          ]}
         GreenCurve {keys [
           <at 0.0292 0.399>
           {arriveTangent=1.9195, interpMode=RCIM_Cubic, leaveTangent=1.9195, time=0.1777, value=1}
           <at 0.2827 0.8856>
           <at 0.3698 0.6781>
           <at 0.4927 1>
           <at 0.5809 0.5281>
           <at 0.8914 1>
           <at 0.9461 1>
           <at 1 1>
          ]}
         RedCurve {keys [<at 0.0292 0>,<at 0.1777 0.1798 RCIM_Cubic>,<at 0.2827 0>,<at 0.3698 0.0171>,<at 0.4927 0.1798>,<at 0.5809 0>,<at 0.8914 0.1798>,<at 0.9461 0.4778>,<at 1 0.6222>]}
        }}
       CurveIndex=$Emitter.NormalizedLoopAge
      ))
     2=<1 1 1>
     3=<1 1 1>
     4=&MakeVectorFromLinearColorRGB(Color=&ColorFromCurve(ColorCurve=DataInterface() {propertyValues {
         AlphaCurve {keys []}
         BlueCurve {keys [
           <at 0.1687 5>
           <at 0.2535 1>
           {arriveTangent=-0.6991, interpMode=RCIM_Cubic, leaveTangent=-0.6991, time=0.5393, value=1}
           <at 1 0.4781 RCIM_Cubic>
          ]}
         GreenCurve {keys [
           <at 0.1687 5>
           <at 0.2535 1>
           {arriveTangent=-0.6991, interpMode=RCIM_Cubic, leaveTangent=-0.6991, time=0.5393, value=1}
           <at 1 0.4781 RCIM_Cubic>
          ]}
         RedCurve {keys [
           <at 0.1687 5>
           <at 0.2535 1>
           {arriveTangent=-0.6991, interpMode=RCIM_Cubic, leaveTangent=-0.6991, time=0.5393, value=1}
           <at 1 0.4781 RCIM_Cubic>
          ]}
        }}))
     5=&MakeVectorFromLinearColorRGB(Color=&ColorFromCurve(ColorCurve=DataInterface() {propertyValues {
         AlphaCurve {keys []}
         BlueCurve {keys [
           <at 0.1687 5>
           <at 0.2535 1>
           {arriveTangent=-0.6991, interpMode=RCIM_Cubic, leaveTangent=-0.6991, time=0.5393, value=1}
           <at 1 0.4781 RCIM_Cubic>
          ]}
         GreenCurve {keys [
           <at 0.1687 5>
           <at 0.2535 1>
           {arriveTangent=-0.6991, interpMode=RCIM_Cubic, leaveTangent=-0.6991, time=0.5393, value=1}
           <at 1 0.4781 RCIM_Cubic>
          ]}
         RedCurve {keys [
           <at 0.1687 5>
           <at 0.2535 1>
           {arriveTangent=-0.6991, interpMode=RCIM_Cubic, leaveTangent=-0.6991, time=0.5393, value=1}
           <at 1 0.4781 RCIM_Cubic>
          ]}
        }}))
     6=<1 1 1>
     7=<1 1 1>
     8=&MakeVectorFromLinearColorRGB(Color=&ColorFromCurve(ColorCurve=DataInterface() {propertyValues {
         AlphaCurve {keys []}
         BlueCurve {keys [<at 0.0189 0>,<at 0.1109 0>,<at 0.3081 0>,<at 0.6538 0 RCIM_Cubic>,<at 0.7747 0 RCIM_Cubic>,<at 1 0>]}
         GreenCurve {keys [
           <at 0.0189 1>
           <at 0.1109 1.5>
           <at 0.3081 1>
           {arriveTangent=-1.4398, interpMode=RCIM_Cubic, leaveTangent=-1.4398, time=0.6538, value=0.8906}
           {arriveTangent=-2.1785, interpMode=RCIM_Cubic, leaveTangent=-2.1785, time=0.7747, value=0.3281}
           <at 1 0.1365>
          ]}
         RedCurve {keys [
           <at 0.0189 0.0354>
           <at 0.1109 0.3057>
           <at 0.3081 0.2038>
           {arriveTangent=-0.4182, interpMode=RCIM_Cubic, leaveTangent=-0.4182, time=0.6538, value=0.1405}
           {arriveTangent=-0.3956, interpMode=RCIM_Cubic, leaveTangent=-0.3956, time=0.7747, value=0.0086}
           <at 1 0.0036>
          ]}
        }}))
     9=<1 1 1>
     Index=&SpawnGroup()
     Num=::NewEnumerator8
    )
    ScaleA=true
    ScaleRGB=true
   )
   &SolveForcesAndVelocity(Clamp Velocity=false, Limit Acceleration=false)
   /SetVariables(Particles.Sprite Ray Alignment:Vector3f=<0 0 1>) {isSetParameters=true}
   &Light_Attributes(
    Color=1
    Diffuse Scale=1
    Falloff Exponent=1
    Light Radius:NiagaraFloat=&MakeCustomFloatFromBool(Bool=&Boolean_OrOperation(Bool1=$User.bIsFrontEnd, Bool2=$User.bIsFrontEndPreview), True Float=0, False Float=130)
    Position=1
    Radius=1
    Scale Alpha:NiagaraFloat=&FloatFromCurve(
     FloatCurve=DataInterface() {propertyValues {Curve {keys [
         {arriveTangent=21.0496, interpMode=RCIM_Cubic, leaveTangent=21.0496, tangentMode=RCTM_User, time=0.0057, value=0.0096}
         {arriveTangentWeight=0.0465, interpMode=RCIM_Cubic, leaveTangentWeight=0.3801, tangentMode=RCTM_User, time=0.1014, value=1}
         {arriveTangentWeight=0.3899, interpMode=RCIM_Cubic, leaveTangentWeight=0.0582, tangentMode=RCTM_User, time=0.8833, value=1}
         {arriveTangent=-7.4625, interpMode=RCIM_Cubic, leaveTangent=-7.4625, tangentMode=RCTM_User, time=1, value=0}
        ]}}}
     Scale Curve=&Add_Float(A=$User.FX_Control, B=0.5)
    )
    Scale RGB:Vector3f=&MakeVectorFromLinearColorRGB(Color=<rgba 0 1 0.0109 1>)
    ScaleA=true
    ScaleRGB=true
    Specular Scale=1
    UsePositionOffset=true
   )
  ]
  rendererClasses [
   &NiagaraMeshRendererProperties
   &NiagaraSpriteRendererProperties
   &NiagaraLightRendererProperties
  ]
  renderers [
   {rendererClass=&NiagaraMeshRendererProperties, rendererIndex=0}
   {rendererClass=&NiagaraSpriteRendererProperties, rendererIndex=1}
   {rendererClass=&NiagaraSpriteRendererProperties, rendererIndex=2}
   {rendererClass=&NiagaraMeshRendererProperties, rendererIndex=3}
   {rendererClass=&NiagaraMeshRendererProperties, rendererIndex=4}
   {rendererClass=&NiagaraMeshRendererProperties, rendererIndex=5}
   {rendererClass=&NiagaraMeshRendererProperties, rendererIndex=6}
   {rendererClass=&NiagaraMeshRendererProperties, rendererIndex=7}
   {rendererClass=&NiagaraSpriteRendererProperties, rendererIndex=8}
   {rendererClass=&NiagaraLightRendererProperties, rendererIndex=9}
   {rendererClass=&NiagaraSpriteRendererProperties, rendererIndex=10}
  ]
 }
}
summary {
 systemName=NS_Emote_BiomeHose_Watch
 emitters [
  {emitterName=Aliens, rendererClasses [&NiagaraSpriteRendererProperties]}
  {
   emitterName=Light
   rendererClasses [
    &NiagaraMeshRendererProperties
    &NiagaraSpriteRendererProperties
    &NiagaraLightRendererProperties
   ]
  }
 ]
 userVariables [
  User.bIsFrontEnd {type=NiagaraBool}
  User.bIsFrontEndPreview {type=NiagaraBool}
  User.BurstIndex {type=NiagaraFloat, defaultValue=0, description=""}
  User.FX_Control {type=NiagaraFloat, defaultValue=0, description=""}
  User.FX_Lerp {type=NiagaraFloat, defaultValue=0, description=""}
 ]
}
